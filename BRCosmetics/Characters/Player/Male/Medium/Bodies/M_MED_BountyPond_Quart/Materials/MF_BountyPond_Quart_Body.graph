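# Material function for the BountyPond Quart male medium body that outputs NoiseMask, a scrolling grayscale noise mask.
# It takes the back-projection UV from the WiryPerk glider's projection function, with size and Z center set by parameters.
# Those UVs are scaled by a tiling parameter, panned by the NoiseUV_Speed vector, then used to sample a low-res blurred noise texture.
# The scalar NoiseUV_SpeedX and NoiseUV_SpeedY parameters are left unconnected.
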MaterialFunction /BRCosmetics/Characters/Player/Male/Medium/Bodies/M_MED_BountyPond_Quart/Materials/MF_BountyPond_Quart_Body
graph {nodes [
  n1 = Constant()
  n2 = ScalarParameter() {defaultValue=51.1618, group=BountyPond_Quart, parameterName=Noise_Uvsize}
  n3 = Constant()
  n4 = Constant()
  n5 = ScalarParameter() {defaultValue=156.7523, group=BountyPond_Quart, parameterName=Noise_CenterZ}
  n6 = Multiply(
   A=NamedRerouteDeclaration(None=MaterialFunctionCall!(Root_Position=@n1, UseClamp?, FlipUV, UVSize=@n2, CenterOffsetX=@n3, CenterOffsetY=@n4, CenterOffsetZ=@n5) {materialFunction=/BRCosmetics/Gadgets/Assets/VinderTech_GliderChute/Glider_WiryPerk/Materials/MF_ProjectUV_WiryPerk}.UV_BackProjection) {name=FrontProjectUV_1, nodeColor=<rgba 0.2422 0 1 1>}
   B=ScalarParameter() {defaultValue=3.4678, group=BountyPond_Quart, parameterName=NoiseUV_tiling}
  ) {constB=3}
  MaterialExpressionVectorParameter_0 = VectorParameter() {group=BountyPond_Quart, parameterName=NoiseUV_Speed, defaultValue=<rgba 0.08 0.08 0 1>}
  n7 = Convert(None=@MaterialExpressionVectorParameter_0.RGB) {
   nodeName=Break Float2
   convertInputs [{
     type=Vector2
     expressionInput {
      expression=&self:MaterialExpressionVectorParameter_0
      inputName=None
      mask=1
      maskA=0
      maskB=1
      maskG=1
      maskR=1
      outputIndex=0
     }
    }]
   convertOutputs [{type=Scalar}, {type=Scalar}]
  }
  n8 = AppendVector(A=@n7, B=@n7)
  FunctionOutput(None=TextureSample(UVs=Panner(Coordinate=@n6, Speed=@n8) {speedX=0.08, speedY=0.08}) {samplerType=SAMPLERTYPE_Grayscale, texture=/Game/Global/Textures/Default/Noise/LowResBlurredNoiseNormalized}.RGB) {outputName=NoiseMask}
  ScalarParameter() {parameterName=NoiseUV_SpeedX}
  ScalarParameter() {parameterName=NoiseUV_SpeedY}
 ]}
