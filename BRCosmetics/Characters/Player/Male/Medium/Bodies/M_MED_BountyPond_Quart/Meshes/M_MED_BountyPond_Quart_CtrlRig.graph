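# Control Rig for the BountyPond Quart male body mesh, procedurally correcting skirt, belt-string, sleeve and body deform bones.
# After the FK and shared deform-rig functions, skirt roots follow the pelvis, then spherical pose readers and calf angles weight thigh-driven skirt offsets.
# Smoothed movement, slide and skydive states add further poses, with crouch blending some targets.
# A sprint-driven frame counter sets the Alpha RBAN curve, and right-wrist rotation lifts sleeve bones.
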ControlRigBlueprint /BRCosmetics/Characters/Player/Male/Medium/Bodies/M_MED_BountyPond_Quart/Meshes/M_MED_BountyPond_Quart_CtrlRig
graphs [
 New Function_ContainedGraph {nodes [
   RigVMFunctionEntryNode()
   > &RigUnit_ModifyTransforms(
    ItemToModify=((Item=(Type=Bone,Name="None"),Transform=(Rotation=(X=0,Y=0,Z=0,W=1),Translation=(X=0,Y=0,Z=0),Scale3D=(X=1,Y=1,Z=1))),(Item=(Type=Bone,Name="None"),Transform=(Rotation=(X=0,Y=0,Z=0,W=1),Translation=(X=0,Y=0,Z=0),Scale3D=(X=1,Y=1,Z=1))))
    WeightMinimum=0
    Weight=RigVMDispatchNode(BoneName=root, Default=0, Name=deformRig_l_upperarm_fr_alpha) {template="DISPATCH_RigDispatch_GetAnimAttribute(in Name,in BoneName,in Default,out Value,out Found)"}.Value
   )
   > RigVMFunctionReturnNode()
  ]}
 New Function_ContainedGraph {nodes [
   RigVMFunctionEntryNode()
   RigVMFunction_MathBoolToInteger = &RigVMFunction_MathBoolToInteger(Value=RigVMVariableNode() {variable=onlyTwist}.Value)
   RigVMFunction_MathQuaternionSwingTwist = &RigVMFunction_MathQuaternionSwingTwist(
    Input=&RigUnit_GetRelativeTransformForItem(
     bChildInitial=False
     bParentInitial=False
     Child=RigVMVariableNode() {variable=child}.Value
     Parent=RigVMVariableNode() {variable=parent}.Value
    ).RelativeTransform.Rotation
    TwistAxis=RigVMVariableNode() {variable=twistAxis}.Value
   )
   n1 = RigVMDispatchNode(
    Values=((),())
    Index=@RigVMFunction_MathBoolToInteger.Result
    Values.0=@RigVMFunction_MathQuaternionSwingTwist.Swing
    Values.1=@RigVMFunction_MathQuaternionSwingTwist.Twist
   ) {template="DISPATCH_RigVMDispatch_SelectInt32(in Index,in Values,out Result)"}
   > RigVMVariableNode(Value=@n1.Result) {variable=entry_outputQuat}
   RigVMFunction_MathQuaternionToEuler = &RigVMFunction_MathQuaternionToEuler(Value=@n1.Result, RotationOrder=RigVMVariableNode() {variable=Rotation Order}.Value)
   > DISPATCH_RigVMDispatch_Print = RigVMDispatchNode(
    ScreenColor=(R=1,G=1,B=1,A=1)
    ScreenDuration=1.5
    Value=@RigVMFunction_MathQuaternionToEuler.Result
    Enabled=RigVMVariableNode() {variable=debugRotation}.Value
   ) {template="DISPATCH_RigVMDispatch_Print(in Prefix,in Value,in Enabled,in ScreenDuration,in ScreenColor)"}
   Cast = RigVMDispatchNode(Value=RigVMVariableNode() {variable=triggerAxis}.Value) {template="DISPATCH_RigVMDispatch_CastEnumToInt(in Value,out Result)"}
   DISPATCH_RigVMDispatch_SelectInt32 = RigVMDispatchNode(
    Values=(0,0,0)
    Index=@Cast.Result
    Values.0=@RigVMFunction_MathQuaternionToEuler.Result.X
    Values.1=@RigVMFunction_MathQuaternionToEuler.Result.Y
    Values.2=@RigVMFunction_MathQuaternionToEuler.Result.Z
   ) {template="DISPATCH_RigVMDispatch_SelectInt32(in Index,in Values,out Result)"}
   n2 = RigVMVariableNode() {variable=rotationMinMax}
   n3 = RigVMVariableNode() {variable=outputMinMax}
   Remap = &RigVMFunction_MathDoubleRemap(
    bClamp=False
    Value=@DISPATCH_RigVMDispatch_SelectInt32.Result
    SourceMinimum=@n2.Value.X
    SourceMaximum=@n2.Value.Y
    TargetMinimum=@n3.Value.X
    TargetMaximum=@n3.Value.Y
   )
   Absolute = &RigVMFunction_MathDoubleAbs(Value=@Remap.Result)
   n4 = RigVMVariableNode() {variable=absoluteValue}
   Multiply = &RigVMFunction_MathDoubleMul(
    A=&RigVMFunction_MathDoubleClamp(Value=RigVMDispatchNode(True=@Absolute.Result, False=@Remap.Result, Condition=@n4.Value) {template="DISPATCH_RigVMDispatch_If(in Condition,in True,in False,out Result)"}.Result).Result
    B=RigVMVariableNode() {variable=premult}.Value
   )
   > RigVMFunctionReturnNode(alpha=@Multiply.Result, rawQuat=RigVMVariableNode() {variable=entry_outputQuat}.Value)
  ]}
 RigVMFunctionLibrary {nodes [
   GetAlphaFromQuatDelta = RigVMCollapseNode(
    ExecuteContext=()
    Rotation Order=ZYX
    child=(Type=Bone,Name="None")
    outputMinMax=(X=0,Y=1)
    parent=(Type=Bone,Name="None")
    premult=1
    rotationMinMax=(X=0,Y=0)
    twistAxis=(X=1,Y=0,Z=0)
   )
   Wrist Sleeves = RigVMCollapseNode(ExecuteContext=())
  ]}
 RigVMFunction_Sequence_2_ContainedGraph {nodes [
   n1 = RigVMFunctionEntryNode()
   n2 = &RigVMFunction_Sequence(ExecuteContext=@n1.ExecuteContext)
   n3 = &RigVMFunction_Sequence(ExecuteContext=@n2.B)
   RigVMFunctionReturnNode(A=@n2.A, B=@n3.A, C=@n3.B)
  ]}
 RigVMModel {nodes [
   n1 = &RigUnit_PrepareForExecution()
   addFKCtrl = RigVMFunctionReferenceNode(
    Custom_FKCtrl=()
    Options_FKCtrl="(pelvis_1_#1=true,spine_3_#2=true,clavicle_5_#3=False,arm_7_#4=False,leg_9_#5=true,head_11_46DC15EF465CD29 ...(truncated)"
    Shape_Settings="(pelvis_2_#6=(bVisible=True,Name=\"Hexagon_Thin\",Color=(R=0.0117,G=0.8281,B=1,A=1),Transform=(Rotation=(X=0,Y=-0.7071,Z=0,W=0.7071),Translation=(X=2,Y=0,Z=0),Scale3D=(X=5,Y=5,Z=7))),spine_4_#7=(bVi ...(truncated)"
    ExecuteContext=@n1.ExecutePin
   )
   RigUnit_BeginExecution = &RigUnit_BeginExecution()
   RigVMFunction_Sequence_1 = RigVMAggregateNode(ExecuteContext=@RigUnit_BeginExecution.ExecutePin)
   setFKCtrl = RigVMFunctionReferenceNode(setFK_Ctrl=false, ExecuteContext=@RigVMFunction_Sequence_1.A)
   > FN_DeformRig_FNC = RigVMFunctionReferenceNode(
    DeformRig_ Curve="(deformRig_l_upperarm_up_alpha_72_#8=False,deformRig_r_upperarm_up_alpha_73_#9=False,deformRig_l_upperarm_fr_alpha_74_#10=False,deformRig_r_upperarm_fr_alpha_75_CB51B8 ...(truncated)"
    DeformRig_Bool="(upperarm_1_#11=True,elbow_3_#12=True,wrist_7_#13=True,groin_8_#14=True,glute_17_#15=True,knee_9_40C408324CA261 ...(truncated)"
    DeformRig_Debug=(upperarm_debug_18_#11=False,wrist_debug_20_#13=False,lat_debug_26_#16=False,groinGlute_debug_27_#14=False)
    DeformRig_Value="(upperarm_up_18_#17=(Rotation=(X=0.0,Y=-0.0,Z=0.0,W=1.0),Translation=(X=0.0,Y=0.0,Z=8.0),Scale3D=(X=0.70,Y=1.0,Z=1.0)),upperarm_fwd_fr_19_#18=(Rotation=(X=0.0,Y=-0.0,Z=0.0,W=1.0),Translation=(X=0.0, ...(truncated)"
   )
   > &RigUnit_ParentConstraint(
    AdvancedSettings=(InterpolationType=Average,RotationOrderForFilter=XZY)
    Child=(Type=Bone,Name="beltStrings_Master")
    Parents="((Item=(Type=Bone,Name=\"spine_02\"),Weight=0.3279),(Item=(Type=Bone,Name=\"spine_03\"),Weight=0.2776),(Item=(Type=Bone,Name=\"spine_01\"),Weight=0.1893),(Item=(Type=Bone,Name=\"pelvis\"),Weight=0.1557),(Item=(Type=Bone,Name=\"spine_04\"),Weight=0.0452),(Item=(Type= ...(truncated)"
    Weight=1
   )
   > &RigUnit_ParentConstraint(
    AdvancedSettings=(InterpolationType=Average,RotationOrderForFilter=XZY)
    Child=(Type=Bone,Name="base_skirt_Front_r_01")
    Parents=((Item=(Type=Bone,Name="pelvis"),Weight=1))
    Weight=1
   )
   > &RigUnit_ParentConstraint(
    AdvancedSettings=(InterpolationType=Average,RotationOrderForFilter=XZY)
    Child=(Type=Bone,Name="base_skirt_Front_l_01")
    Parents=((Item=(Type=Bone,Name="pelvis"),Weight=1))
    Weight=1
   )
   > &RigUnit_ParentConstraint(
    AdvancedSettings=(InterpolationType=Average,RotationOrderForFilter=XZY)
    Child=(Type=Bone,Name="base_skirt_SideFront_l_01")
    Parents=((Item=(Type=Bone,Name="pelvis"),Weight=1))
    Weight=1
   )
   > &RigUnit_ParentConstraint(
    AdvancedSettings=(InterpolationType=Average,RotationOrderForFilter=XZY)
    Child=(Type=Bone,Name="base_skirt_SideFront_r_01")
    Parents=((Item=(Type=Bone,Name="pelvis"),Weight=1))
    Weight=1
   )
   > &RigUnit_ParentConstraint(
    AdvancedSettings=(InterpolationType=Average,RotationOrderForFilter=XZY)
    Child=(Type=Bone,Name="base_skirt_Back_l_01")
    Parents=((Item=(Type=Bone,Name="pelvis"),Weight=1))
    Weight=1
   )
   > &RigUnit_ParentConstraint(
    AdvancedSettings=(InterpolationType=Average,RotationOrderForFilter=XZY)
    Child=(Type=Bone,Name="base_skirt_Back_r_01")
    Parents=((Item=(Type=Bone,Name="pelvis"),Weight=1))
    Weight=1
   )
   > &RigUnit_ParentConstraint(
    AdvancedSettings=(InterpolationType=Average,RotationOrderForFilter=XZY)
    Child=(Type=Bone,Name="base_skirt_SideBack_l_01")
    Parents=((Item=(Type=Bone,Name="pelvis"),Weight=1))
    Weight=1
   )
   > n2 = &RigUnit_ParentConstraint(
    AdvancedSettings=(InterpolationType=Average,RotationOrderForFilter=XZY)
    Child=(Type=Bone,Name="base_skirt_SideBack_r_01")
    Parents=((Item=(Type=Bone,Name="pelvis"),Weight=1))
    Weight=1
   )
   RigVMFunction_Sequence = &RigVMFunction_Sequence(ExecuteContext=@n2.ExecutePin)
   n3 = &RigUnit_SphericalPoseReader(
    ActiveRegionScaleFactors=(PositiveWidth=3.5,NegativeWidth=0.5,PositiveHeight=3,NegativeHeight=3)
    ActiveRegionSize=0.1
    Debug=(bDrawDebug=false,bDraw2D=False,bDrawLocalAxes=False,DebugScale=25,DebugSegments=20,DebugThickness=0.25)
    DriverAxis=(X=-1,Y=0,Z=0)
    DriverItem=(Type=Bone,Name="thigh_l")
    FalloffRegionScaleFactors=(PositiveWidth=1,NegativeWidth=1,PositiveHeight=1,NegativeHeight=1)
    FalloffSize=0.45
    FlipHeightScaling=False
    FlipWidthScaling=False
    OptionalParentItem=(Type=Bone,Name="None")
    RotationOffset=(X=90,Y=-6,Z=0)
    ExecutePin=@RigVMFunction_Sequence.A
   )
   > RigVMVariableNode(Value=@n3.OutputParam) {variable=Thigh L Forward}
   > n4 = &RigUnit_SphericalPoseReader(
    ActiveRegionScaleFactors=(PositiveWidth=3.5,NegativeWidth=0.5,PositiveHeight=3,NegativeHeight=3)
    ActiveRegionSize=0.1
    Debug=(bDrawDebug=false,bDraw2D=False,bDrawLocalAxes=False,DebugScale=25,DebugSegments=20,DebugThickness=0.25)
    DriverAxis=(X=-1,Y=0,Z=0)
    DriverItem=(Type=Bone,Name="thigh_l")
    FalloffRegionScaleFactors=(PositiveWidth=0.5,NegativeWidth=1,PositiveHeight=1,NegativeHeight=1)
    FalloffSize=0.49
    FlipHeightScaling=False
    FlipWidthScaling=False
    OptionalParentItem=(Type=Bone,Name="None")
    RotationOffset=(X=180,Y=-4,Z=0)
   )
   > RigVMVariableNode(Value=@n4.OutputParam) {variable=Thigh L Side}
   > n5 = &RigUnit_SphericalPoseReader(
    ActiveRegionScaleFactors=(PositiveWidth=3.5,NegativeWidth=0.5,PositiveHeight=3,NegativeHeight=3)
    ActiveRegionSize=0.1
    Debug=(bDrawDebug=false,bDraw2D=False,bDrawLocalAxes=False,DebugScale=25,DebugSegments=20,DebugThickness=0.25)
    DriverAxis=(X=-1,Y=0,Z=0)
    DriverItem=(Type=Bone,Name="thigh_l")
    FalloffRegionScaleFactors=(PositiveWidth=0.5,NegativeWidth=1,PositiveHeight=1,NegativeHeight=1)
    FalloffSize=0.49
    FlipHeightScaling=False
    FlipWidthScaling=False
    OptionalParentItem=(Type=Bone,Name="None")
    RotationOffset=(X=-90,Y=-6,Z=0)
   )
   > RigVMVariableNode(Value=@n5.OutputParam) {variable=Thigh L Backwards}
   > n6 = &RigUnit_SphericalPoseReader(
    ActiveRegionScaleFactors=(PositiveWidth=2,NegativeWidth=2.7,PositiveHeight=2,NegativeHeight=2)
    ActiveRegionSize=0.1
    Debug=(bDrawDebug=false,bDraw2D=False,bDrawLocalAxes=False,DebugScale=25,DebugSegments=20,DebugThickness=0.25)
    DriverAxis=(X=-1,Y=0,Z=0)
    DriverItem=(Type=Bone,Name="thigh_l")
    FalloffRegionScaleFactors=(PositiveWidth=1,NegativeWidth=1,PositiveHeight=1,NegativeHeight=1)
    FalloffSize=0.35
    FlipHeightScaling=False
    FlipWidthScaling=False
    OptionalParentItem=(Type=Bone,Name="None")
    RotationOffset=(X=90,Y=-6,Z=-55)
   )
   > RigVMVariableNode(Value=@n6.OutputParam) {variable=Thigh L Forward Extreme}
   > n7 = &RigUnit_SphericalPoseReader(
    ActiveRegionScaleFactors=(PositiveWidth=3.5,NegativeWidth=0.5,PositiveHeight=3,NegativeHeight=3)
    ActiveRegionSize=0.1
    Debug=(bDrawDebug=false,bDraw2D=False,bDrawLocalAxes=False,DebugScale=25,DebugSegments=20,DebugThickness=0.25)
    DriverAxis=(X=1,Y=0,Z=0)
    DriverItem=(Type=Bone,Name="thigh_r")
    FalloffRegionScaleFactors=(PositiveWidth=1,NegativeWidth=1,PositiveHeight=1,NegativeHeight=1)
    FalloffSize=0.45
    FlipHeightScaling=False
    FlipWidthScaling=False
    OptionalParentItem=(Type=Bone,Name="None")
    RotationOffset=(X=-90,Y=174,Z=0)
   )
   > RigVMVariableNode(Value=@n7.OutputParam) {variable=Thigh R Forward}
   > n8 = &RigUnit_SphericalPoseReader(
    ActiveRegionScaleFactors=(PositiveWidth=3.5,NegativeWidth=0.5,PositiveHeight=3,NegativeHeight=3)
    ActiveRegionSize=0.1
    Debug=(bDrawDebug=false,bDraw2D=False,bDrawLocalAxes=False,DebugScale=25,DebugSegments=20,DebugThickness=0.25)
    DriverAxis=(X=1,Y=0,Z=0)
    DriverItem=(Type=Bone,Name="thigh_r")
    FalloffRegionScaleFactors=(PositiveWidth=0.5,NegativeWidth=1,PositiveHeight=1,NegativeHeight=1)
    FalloffSize=0.49
    FlipHeightScaling=False
    FlipWidthScaling=False
    OptionalParentItem=(Type=Bone,Name="None")
    RotationOffset=(X=180,Y=174,Z=0)
   )
   > RigVMVariableNode(Value=@n8.OutputParam) {variable=Thigh R Side}
   > n9 = &RigUnit_SphericalPoseReader(
    ActiveRegionScaleFactors=(PositiveWidth=3.5,NegativeWidth=0.5,PositiveHeight=3,NegativeHeight=3)
    ActiveRegionSize=0.1
    Debug=(bDrawDebug=false,bDraw2D=False,bDrawLocalAxes=False,DebugScale=25,DebugSegments=20,DebugThickness=0.25)
    DriverAxis=(X=1,Y=0,Z=0)
    DriverItem=(Type=Bone,Name="thigh_r")
    FalloffRegionScaleFactors=(PositiveWidth=0.5,NegativeWidth=1,PositiveHeight=1,NegativeHeight=1)
    FalloffSize=0.49
    FlipHeightScaling=False
    FlipWidthScaling=False
    OptionalParentItem=(Type=Bone,Name="None")
    RotationOffset=(X=90,Y=174,Z=0)
   )
   > RigVMVariableNode(Value=@n9.OutputParam) {variable=Thigh R Backwards}
   n10 = &RigUnit_GetRelativeTransformForItem(Child=(Type=Bone,Name="calf_l"), Parent=(Type=Bone,Name="thigh_l"), bChildInitial=False, bParentInitial=False)
   > RigVMVariableNode(Value=&RigVMFunction_MathDoubleRemap(
     SourceMaximum=140
     SourceMinimum=23
     TargetMaximum=1
     TargetMinimum=0
     Value=&RigVMFunction_MathQuaternionToEuler(RotationOrder=ZXY, Value=@n10.RelativeTransform.Rotation).Result.Z
    ).Result) {variable=Calf L Backwards}
   n11 = &RigUnit_GetRelativeTransformForItem(Child=(Type=Bone,Name="calf_r"), Parent=(Type=Bone,Name="thigh_r"), bChildInitial=False, bParentInitial=False)
   > RigVMVariableNode(Value=&RigVMFunction_MathDoubleRemap(
     SourceMaximum=140
     SourceMinimum=23
     TargetMaximum=1
     TargetMinimum=0
     Value=&RigVMFunction_MathQuaternionToEuler(RotationOrder=ZXY, Value=@n11.RelativeTransform.Rotation).Result.Z
    ).Result) {variable=Calf R Backwards}
   n12 = RigVMVariableNode() {variable="is Crouching "}
   Interpolate = /Script/RigVM.RigVMFunction_MathQuaternionSlerp(A=(X=0,Y=0,Z=-0.5736,W=0.8192), B=(X=0.4299,Y=0.0049,Z=-0.3924,W=0.8131), T=@n12.Value) {is=struct RigVMUnitNode, template="Interpolate::Execute(in A,in B,in T,out Result)"}
   &RigUnit_ModifyTransforms(
    ItemToModify="((Item=(Type=Bone,Name=\"base_skirt_Front_l_01\"),Transform=(Rotation=(X=0.4299,Y=0.0049,Z=-0.3924,W=0.8131),Translation=(X=5,Y=0,Z=0),Scale3D=(X=1,Y=1,Z=1))),(Item=(Type=Bone,Name=\"base_skirt_SideFront_l_01\"),Transform=(Rotation=(X=0.0398,Y=-0.0093,Z=0.0527 ...(truncated)"
    WeightMinimum=0
    ItemToModify.0.Transform.Rotation=@Interpolate.Result
    ItemToModify.1.Transform.Rotation=&RigVMFunction_MathQuaternionSlerp(A=(X=-0.0338,Y=-0.2566,Z=-0.1261,W=0.9577), B=(X=0.0398,Y=-0.0093,Z=0.0527,W=0.9978), T=@n12.Value) {template="Interpolate::Execute(in A,in B,in T,out Result)"}.Result
    ItemToModify.2.Transform.Rotation=&RigVMFunction_MathQuaternionSlerp(A=(X=0,Y=-0.1305,Z=0,W=0.9914), B=(X=0.0017,Y=0.0636,Z=-0.0261,W=0.9976), T=@n12.Value) {template="Interpolate::Execute(in A,in B,in T,out Result)"}.Result
    ExecutePin=@RigVMFunction_Sequence.B
    Weight=RigVMVariableNode() {variable=Thigh L Forward}.Value
   )
   n13 = RigVMVariableNode() {variable="is Crouching "}
   > &RigUnit_ModifyTransforms(
    ItemToModify="((Item=(Type=Bone,Name=\"base_skirt_SideFront_l_01\"),Transform=(Rotation=(X=0.0802,Y=0.154,Z=-0.4547,W=0.8735),Translation=(X=0,Y=0,Z=0),Scale3D=(X=1,Y=1,Z=1))),(Item=(Type=Bone,Name=\"base_skirt_SideBack_l_01\"),Transform=(Rotation=(X=-0.0262,Y=-0.0831,Z=-0. ...(truncated)"
    WeightMinimum=0
    ItemToModify.0.Transform.Rotation=&RigVMFunction_MathQuaternionSlerp(A=(X=-0.025,Y=-0.0357,Z=-0.573,W=0.8184), B=(X=0.0802,Y=0.154,Z=-0.4547,W=0.8735), T=@n13.Value) {template="Interpolate::Execute(in A,in B,in T,out Result)"}.Result
    ItemToModify.1.Transform.Rotation=&RigVMFunction_MathQuaternionSlerp(A=(X=-0.0632,Y=-0.0991,Z=-0.5336,W=0.8375), B=(X=-0.0262,Y=-0.0831,Z=-0.2996,W=0.9501), T=@n13.Value) {template="Interpolate::Execute(in A,in B,in T,out Result)"}.Result
    ItemToModify.2.Transform.Rotation=&RigVMFunction_MathQuaternionSlerp(A=(X=0.036,Y=-0.2733,Z=0.1255,W=0.953), B=(X=-0.0227,Y=-0.1722,Z=-0.1285,W=0.9764), T=@n13.Value) {template="Interpolate::Execute(in A,in B,in T,out Result)"}.Result
    Weight=RigVMVariableNode() {variable=Thigh L Side}.Value
   )
   Subtract_1 = RigVMAggregateNode(
    A=RigVMVariableNode() {variable=Calf L Backwards}.Value
    B=RigVMVariableNode() {variable=Thigh L Forward}.Value
    C=RigVMVariableNode() {variable=is Knee Sliding}.Value
   )
   n14 = &RigVMFunction_AlphaInterp(
    InterpSpeedDecreasing=10
    InterpSpeedIncreasing=10
    Scale=1
    bClampResult=true
    Value=&RigVMFunction_MathDoubleMax(B=@Subtract_1.Result, A=RigVMVariableNode() {variable=Thigh L Backwards}.Value).Result
   )
   > &RigUnit_ModifyTransforms(
    ItemToModify="((Item=(Type=Bone,Name=\"base_skirt_Back_l_01\"),Transform=(Rotation=(X=0,Y=0,Z=-0.5736,W=0.8192),Translation=(X=0,Y=0,Z=0),Scale3D=(X=1,Y=1,Z=1))),(Item=(Type=Bone,Name=\"base_skirt_Back_r_01\"),Transform=(Rotation=(X=-0.0436,Y=-0.0755,Z=-0.4981,W=0.8627),Tra ...(truncated)"
    WeightMinimum=0
    Weight=@n14.Result
   )
   n15 = RigVMVariableNode() {variable="is Crouching "}
   > &RigUnit_ModifyTransforms(
    ItemToModify="((Item=(Type=Bone,Name=\"base_skirt_Front_r_01\"),Transform=(Rotation=(X=-0.0376,Y=-0.2132,Z=-0.1695,W=0.9615),Translation=(X=5,Y=0,Z=0),Scale3D=(X=1,Y=1,Z=1))),(Item=(Type=Bone,Name=\"base_skirt_SideFront_r_01\"),Transform=(Rotation=(X=-0.0114,Y=-0.13,Z=-0.08 ...(truncated)"
    WeightMinimum=0
    ItemToModify.0.Transform.Rotation=&RigVMFunction_MathQuaternionSlerp(A=(X=0,Y=0,Z=-0.5736,W=0.8192), B=(X=-0.0376,Y=-0.2132,Z=-0.1695,W=0.9615), T=@n15.Value) {template="Interpolate::Execute(in A,in B,in T,out Result)"}.Result
    ItemToModify.1.Transform.Rotation=&RigVMFunction_MathQuaternionSlerp(A=(X=0.0338,Y=0.2566,Z=-0.1261,W=0.9577), B=(X=-0.0114,Y=-0.13,Z=-0.0864,W=0.9877), T=@n15.Value) {template="Interpolate::Execute(in A,in B,in T,out Result)"}.Result
    ItemToModify.2.Transform.Rotation=&RigVMFunction_MathQuaternionSlerp(A=(X=0,Y=0.1305,Z=0,W=0.9914), B=(X=0,Y=-0.061,Z=0,W=0.9981), T=@n15.Value) {template="Interpolate::Execute(in A,in B,in T,out Result)"}.Result
    Weight=RigVMVariableNode() {variable=Thigh R Forward}.Value
   )
   > &RigUnit_ModifyTransforms(
    ItemToModify="((Item=(Type=Bone,Name=\"base_skirt_SideFront_r_01\"),Transform=(Rotation=(X=0.025,Y=0.0357,Z=-0.573,W=0.8184),Translation=(X=0,Y=0,Z=0),Scale3D=(X=1,Y=1,Z=1))),(Item=(Type=Bone,Name=\"base_skirt_SideBack_r_01\"),Transform=(Rotation=(X=0.0933,Y=0.1465,Z=-0.529 ...(truncated)"
    WeightMinimum=0
    Weight=RigVMVariableNode() {variable=Thigh R Side}.Value
   )
   Multiply = &RigVMFunction_MathDoubleMul(A=RigVMVariableNode() {variable=Thigh R Side}.Value, B=RigVMVariableNode() {variable=is Standing Moving}.Value)
   > &RigUnit_ModifyTransforms(
    ItemToModify="((Item=(Type=Bone,Name=\"base_skirt_SideFront_r_01\"),Transform=(Rotation=(X=0,Y=-0.1305,Z=0,W=0.9914),Translation=(X=0,Y=0,Z=0),Scale3D=(X=1,Y=1,Z=1))),(Item=(Type=Bone,Name=\"base_skirt_SideBack_r_01\"),Transform=(Rotation=(X=0,Y=0.1865,Z=0,W=0.9825),Transla ...(truncated)"
    WeightMinimum=0
    Weight=@Multiply.Result
   )
   Subtract = RigVMAggregateNode(
    A=RigVMVariableNode() {variable=Calf R Backwards}.Value
    B=RigVMVariableNode() {variable=Thigh R Forward}.Value
    C=RigVMVariableNode() {variable=is Knee Sliding}.Value
   )
   Maximum = &RigVMFunction_MathDoubleMax(B=@Subtract.Result, A=RigVMVariableNode() {variable=Thigh R Backwards}.Value)
   n16 = &RigVMFunction_AlphaInterp(InterpSpeedDecreasing=10, InterpSpeedIncreasing=10, Scale=1, bClampResult=true, Value=@Maximum.Result)
   > &RigUnit_ModifyTransforms(
    ItemToModify="((Item=(Type=Bone,Name=\"base_skirt_Back_r_01\"),Transform=(Rotation=(X=0,Y=0,Z=-0.5736,W=0.8192),Translation=(X=0,Y=0,Z=0),Scale3D=(X=1,Y=1,Z=1))),(Item=(Type=Bone,Name=\"base_skirt_Back_l_01\"),Transform=(Rotation=(X=0.0436,Y=0.0755,Z=-0.4981,W=0.8627),Trans ...(truncated)"
    WeightMinimum=0
    Weight=@n16.Result
   )
   AlphaInterp = &RigVMFunction_AlphaInterp(InterpSpeedDecreasing=3, InterpSpeedIncreasing=10, Scale=1, bClampResult=true, Value=RigVMVariableNode() {variable="is Moving 2D "}.Value)
   > &RigUnit_ModifyTransforms(
    ItemToModify="((Item=(Type=Bone,Name=\"base_skirt_Back_r_01\"),Transform=(Rotation=(X=0,Y=0,Z=-0.1478,W=0.989),Translation=(X=0,Y=0,Z=0),Scale3D=(X=1,Y=1,Z=1))),(Item=(Type=Bone,Name=\"base_skirt_Back_l_01\"),Transform=(Rotation=(X=0,Y=0,Z=-0.1478,W=0.989),Translation=(X=0, ...(truncated)"
    WeightMinimum=0
    Weight=@AlphaInterp.Result
   )
   n17 = &RigVMFunction_AlphaInterp(InterpSpeedDecreasing=15, InterpSpeedIncreasing=7, Scale=1, bClampResult=true, Value=RigVMVariableNode() {variable=is Knee Sliding}.Value)
   RigVMFunction_AnimEvalRichCurve = &RigVMFunction_AnimEvalRichCurve(
    Curve="(EditorCurveData=(Keys=((InterpMode=RCIM_Linear,TangentMode=RCTM_Auto,TangentWeightMode=RCTWM_WeightedNone,Time=0,Value=0,ArriveTangent=0,ArriveTangentWeight=0,LeaveTangent=0,LeaveTangentWeight=0),(InterpMode=RCIM_Linear,TangentMode=RCTM_Auto,TangentWeight ...(truncated)"
    SourceMinimum=0
    TargetMinimum=0
    Value=@n17.Result
   )
   > &RigUnit_ModifyTransforms(
    ItemToModify="((Item=(Type=Bone,Name=\"base_skirt_Front_r_01\"),Transform=(Rotation=(X=-0.0091,Y=-0.1302,Z=-0.0692,W=0.989),Translation=(X=0,Y=0,Z=0),Scale3D=(X=1,Y=1,Z=1))),(Item=(Type=Bone,Name=\"base_skirt_SideFront_r_01\"),Transform=(Rotation=(X=-0.0302,Y=-0.171,Z=-0.17 ...(truncated)"
    WeightMinimum=0
    Weight=@RigVMFunction_AnimEvalRichCurve.Result
   )
   n18 = &RigVMFunction_AlphaInterp(
    InterpSpeedDecreasing=10
    InterpSpeedIncreasing=5
    Scale=1
    bClampResult=true
    Value=RigVMVariableNode() {variable=is Baseball Sliding}.Value
   )
   n19 = &RigVMFunction_AnimEvalRichCurve(
    Curve="(EditorCurveData=(Keys=((InterpMode=RCIM_Linear,TangentMode=RCTM_Auto,TangentWeightMode=RCTWM_WeightedNone,Time=0,Value=0,ArriveTangent=0,ArriveTangentWeight=0,LeaveTangent=0,LeaveTangentWeight=0),(InterpMode=RCIM_Linear,TangentMode=RCTM_Auto,TangentWeight ...(truncated)"
    SourceMinimum=0
    TargetMinimum=0
    Value=@n18.Result
   )
   > &RigUnit_ModifyTransforms(
    ItemToModify="((Item=(Type=Bone,Name=\"base_skirt_Front_r_01\"),Transform=(Rotation=(X=-0.0827,Y=-0.3319,Z=-0.2273,W=0.9118),Translation=(X=0,Y=8,Z=0),Scale3D=(X=1,Y=1,Z=1))),(Item=(Type=Bone,Name=\"base_skirt_SideFront_r_01\"),Transform=(Rotation=(X=-0.0531,Y=-0.0691,Z=-0. ...(truncated)"
    WeightMinimum=0
    Weight=@n19.Result
   )
   n20 = &RigVMFunction_AlphaInterp(InterpSpeedDecreasing=10, InterpSpeedIncreasing=10, Scale=1, bClampResult=true, Value=RigVMVariableNode() {variable=is Sky Dive}.Value)
   n21 = &RigVMFunction_AnimEvalRichCurve(
    Curve="(EditorCurveData=(Keys=((InterpMode=RCIM_Linear,TangentMode=RCTM_Auto,TangentWeightMode=RCTWM_WeightedNone,Time=0,Value=0,ArriveTangent=0,ArriveTangentWeight=0,LeaveTangent=0,LeaveTangentWeight=0),(InterpMode=RCIM_Linear,TangentMode=RCTM_Auto,TangentWeight ...(truncated)"
    SourceMinimum=0
    TargetMinimum=0
    Value=@n20.Result
   )
   > &RigUnit_ModifyTransforms(
    ItemToModify="((Item=(Type=Bone,Name=\"base_skirt_Back_r_01\"),Transform=(Rotation=(X=0,Y=0,Z=-0.0872,W=0.9962),Translation=(X=0,Y=0,Z=0),Scale3D=(X=1,Y=1,Z=1))),(Item=(Type=Bone,Name=\"base_skirt_Back_l_01\"),Transform=(Rotation=(X=0,Y=0,Z=-0.0872,W=0.9962),Translation=(X= ...(truncated)"
    WeightMinimum=0
    Weight=@n21.Result
   )
   Remap = &RigVMFunction_MathDoubleRemap(SourceMaximum=1, SourceMinimum=0, TargetMaximum=0, TargetMinimum=1, bClamp=False, Value=RigVMVariableNode() {variable="is Crouching "}.Value)
   > &RigUnit_ParentConstraint(
    AdvancedSettings=(InterpolationType=Average,RotationOrderForFilter=XZY)
    Child=(Type=Bone,Name="beltStrings_Sub")
    Filter=(TranslationFilter=(bX=false,bY=false,bZ=false),RotationFilter=(bX=True,bY=True,bZ=True),ScaleFilter=(bX=True,bY=True,bZ=True))
    Parents=((Item=(Type=Bone,Name="base_skirt_Front_l_01"),Weight=1))
    Weight=1
    Parents.0.Weight=@Remap.Result
   )
   n22 = RigVMVariableNode() {variable=is Sprinting}
   n23 = &RigVMFunction_ControlFlowBranch(ExecuteContext=@RigVMFunction_Sequence_1.B, Condition=@n22.Value)
   n24 = RigVMDispatchNode(
    True=1
    Condition=/Script/RigVM.RigVMFunction_MathDoubleGreaterEqual(B=20, A=RigVMVariableNode() {variable=Timer}.Value) {is=struct RigVMUnitNode, template="GreaterEqual::Execute(in A,in B,out Result)"}.Result
    False=RigVMDispatchNode(False=1, True=0.5, Condition=@n22.Value) {template="DISPATCH_RigVMDispatch_If(in Condition,in True,in False,out Result)"}.Result
   ) {template="DISPATCH_RigVMDispatch_If(in Condition,in True,in False,out Result)"}
   RigVMVariableNode(Value=@n24.Result, ExecuteContext=@n23.Completed) {variable=Alpha RBAN}
   Absolute = &RigVMFunction_MathDoubleAbs(Value=&RigVMFunction_MathDoubleSub(B=1, A=RigVMVariableNode() {variable=Alpha RBAN}.Value).Result)
   > /Script/ControlRig.RigUnit_SetCurveValue(Curve=Alpha RBAN, Value=@Absolute.Result) {is=struct RigVMUnitNode}
   RigVMVariableNode(ExecuteContext=@n23.False) {variable=Timer}
   RigVMFunction_ForLoopCount = /Script/RigVM.RigVMFunction_ForLoopCount(Count=1, ExecutePin=@n23.True) {is=struct RigVMUnitNode}
   RigVMFunction_MathDoubleToInt = /Script/RigVM.RigVMFunction_MathDoubleToInt(Value=RigVMVariableNode() {variable=Timer}.Value) {is=struct RigVMUnitNode, template="Cast::Execute(in Value,out Result)"}
   RigVMFunction_MathIntToDouble = /Script/RigVM.RigVMFunction_MathIntToFloat(Value=/Script/RigVM.RigVMFunction_MathIntAdd(B=1, A=@RigVMFunction_MathDoubleToInt.Result) {is=struct RigVMUnitNode, template="Add::Execute(in A,in B,out Result)"}.Result) {is=struct RigVMUnitNode, template="Cast::Execute(in Value,out Result)"}
   > RigVMVariableNode(Value=@RigVMFunction_MathIntToDouble.Result) {variable=Timer}
   GetAlphaFromQuatDelta = RigVMFunctionReferenceNode(
    Rotation Order=ZYX
    child=(Type=Bone,Name="hand_r")
    debugRotation=false
    outputMinMax=(X=0,Y=1)
    parent=(Type=Null,Name="hand_r_null")
    premult=1
    rotationMinMax=(X=0,Y=45)
    triggerAxis=Y
    twistAxis=(X=1,Y=0,Z=0)
    ExecuteContext=@RigVMFunction_Sequence_1.C
   )
   n25 = &RigVMFunction_AnimEvalRichCurve(
    Curve="(EditorCurveData=(Keys=((InterpMode=RCIM_Cubic,TangentMode=RCTM_User,ArriveTangent=1.4036,LeaveTangent=1.4036),(InterpMode=RCIM_Cubic,TangentMode=RCTM_User,Time=0.5,Value=0.65,ArriveTangent=1.1037,LeaveTangent=1.1037),(InterpMode=RCIM_Cubic,TangentMode=RCT ...(truncated)"
    SourceMinimum=0
    TargetMinimum=0
    Value=@GetAlphaFromQuatDelta.alpha
   )
   > &RigUnit_ModifyTransforms(
    ItemToModify="((Item=(Type=Bone,Name=\"base_sleeve_03_dn_r\"),Transform=(Rotation=(X=0,Y=0,Z=-0.284,W=0.9588),Translation=(X=0,Y=0,Z=0),Scale3D=(X=1,Y=1,Z=1))),(Item=(Type=Bone,Name=\"base_sleeve_03_r\"),Transform=(Rotation=(X=0,Y=0,Z=-0.0872,W=0.9962),Translation=(X=0,Y=0, ...(truncated)"
    WeightMinimum=0
    Weight=@n25.Result
   )
   Control Stack at Item = RigVMFunctionReferenceNode(
    Add Bottom Null=False
    Add Null=true
    Add Secondary Control=False
    Control Suffix=_ctrl
    ExecuteContext=()
    Item=(Name="head",Type=Bone)
    Name=ref_head
    Null Suffix=_null
    Parent=(Name="neck_02",Type=Bone)
    Secondary Settings="(Shape=(bVisible=True,Name=\"Default\",Color=(R=1,G=0,B=0,A=1),Transform=(Rotation=(X=0,Y=0,Z=0,W=1),Translation=(X=0,Y=0,Z=0),Scale3D=(X=1,Y=1,Z=1))),Proxy=(bIsProxy=False,ShapeVisibility=BasedOnSelection),DisplayName=\"None\",InitialSpace=LocalSpace,bUsePref ...(truncated)"
    Settings="(Shape=(bVisible=True,Name=\"Pyramid_Thin\",Color=(R=1,G=0,B=0,A=1),Transform=(Rotation=(X=0,Y=0,Z=0,W=1),Translation=(X=0,Y=0,Z=0),Scale3D=(X=1,Y=1,Z=1))),Proxy=(bIsProxy=False,ShapeVisibility=BasedOnSelection),DisplayName=\"None\",InitialSpace=LocalSpace,bUs ...(truncated)"
   )
   &RigVMFunction_MathQuaternionSlerp(A=(X=-0.0338,Y=-0.2566,Z=-0.1261,W=0.9577), B=(X=-0.0016,Y=-0.0261,Z=-0.061,W=0.9978), T=1) {template="Interpolate::Execute(in A,in B,in T,out Result)"}
   &RigVMFunction_MathQuaternionSlerp(A=(X=0,Y=-0.1305,Z=0,W=0.9914), B=(X=0.0007,Y=0.0262,Z=-0.0262,W=0.9993), T=1) {template="Interpolate::Execute(in A,in B,in T,out Result)"}
   &RigVMFunction_MathQuaternionSlerp(A=(X=0,Y=0,Z=-0.5736,W=0.8192), B=(X=0.0449,Y=0.2549,Z=-0.1677,W=0.9513), T=1) {template="Interpolate::Execute(in A,in B,in T,out Result)"}
   &RigUnit_ModifyTransforms(
    ExecutePin=()
    ItemToModify="((Item=(Type=Bone,Name=\"base_skirt_SideFront_r_01\"),Transform=(Rotation=(X=-0.025,Y=-0.0357,Z=-0.573,W=0.8184),Translation=(X=0,Y=0,Z=0),Scale3D=(X=1,Y=1,Z=1))),(Item=(Type=Bone,Name=\"base_skirt_SideBack_r_01\"),Transform=(Rotation=(X=0.0632,Y=0.0991,Z=-0.5 ...(truncated)"
    Weight=1
    WeightMinimum=0
   )
   RigVMDispatchNode(Enabled=True, ExecuteContext=(), ScreenColor=(R=1,G=1,B=1,A=1), ScreenDuration=0.05, Value=@n24.Result) {template="DISPATCH_RigVMDispatch_Print(in Prefix,in Value,in Enabled,in ScreenDuration,in ScreenColor)"}
  ]}
 Subtract_2_ContainedGraph {nodes [
   n1 = RigVMFunctionEntryNode()
   RigVMFunctionReturnNode(Result=&RigVMFunction_MathDoubleSub(B=@n1.C, A=&RigVMFunction_MathDoubleSub(A=@n1.A, B=@n1.B).Result).Result)
  ]}
 Subtract_2_ContainedGraph {nodes [
   n1 = RigVMFunctionEntryNode()
   Subtract = &RigVMFunction_MathDoubleSub(A=@n1.A, B=@n1.B)
   RigVMFunctionReturnNode(Result=&RigVMFunction_MathDoubleSub(B=@n1.C, A=@Subtract.Result).Result)
  ]}
]
hierarchy [
 root {loc=<0 0 0>, rot=<0 0 0 1>}
 pelvis {parent=root, loc=<0 -0.7071 93.1191>, rot=<0.0127 0.707 -0.0127 -0.707>}
 spine_01 {parent=pelvis, loc=<6.7494 -0.6388 0>, rot=<0 0 0.034 -0.9994>}
 spine_02 {parent=spine_01, loc=<7.2518 -0.035 0>, rot=<0 0 0.0238 -0.9997>}
 spine_03 {parent=spine_02, loc=<6.0104 0 0>, rot=<0 0 0.0149 -0.9999>}
 spine_04 {parent=spine_03, loc=<8.1024 0 0>, rot=<0 0 -0.0132 -0.9999>}
 spine_05 {parent=spine_04, loc=<10.0146 0 0>, rot=<0 0 -0.0015 -1>}
 neck_01 {parent=spine_05, loc=<11.3138 0.1529 0>, rot=<0 0 -0.2003 -0.9797>}
 neck_02 {parent=neck_01, loc=<4.5627 0 0>, rot=<0 0 0.0504 -0.9987>}
 head {parent=neck_02, loc=<5.3914 0 0>, rot=<0 0 0.115 -0.9934>}
 attach_cape {parent=spine_05, loc=<9.017 -11.714 0>, rot=<0.0616 -0.7044 -0.0616 -0.7044>}
 cape {parent=attach_cape, type=Null, loc=<0 4.0021 0>, rot=<0 0 0 1>}
 attach_backpack {parent=spine_05, loc=<-0.123 -13.473 0>, rot=<-0.0093 -0.707 0.0093 -0.707>}
 Backpack_BR {parent=attach_backpack, type=Null, loc=<0 3.4315 0>, rot=<0 0 0 1>}
 clavicle_l {parent=spine_05, loc=<4.3575 -0.1466 -2.3223>, rot=<0.0465 -0.7289 -0.009 -0.6829>}
 upperarm_l {parent=clavicle_l, loc=<15.5111 0 0>, rot=<0.0226 -0.3234 -0.0088 -0.9459>}
 lowerarm_l {parent=upperarm_l, loc=<26.6514 0 0>, rot=<0.0391 -0.0099 -0.2453 -0.9686>}
 hand_l {parent=lowerarm_l, loc=<22.3491 0 0>, rot=<-0.7081 -0.0507 0.0975 -0.6975>}
 thumb_01_l {parent=hand_l, loc=<1.3118 -0.3191 -2.459>, rot=<0.6546 -0.4042 0.0664 -0.6354>}
 thumb_02_l {parent=thumb_01_l, loc=<5.9336 -0.0448 0.0401>, rot=<-0.0037 0.0213 0.298 -0.9543>}
 thumb_03_l {parent=thumb_02_l, loc=<2.2223 0.3682 -0.0907>, rot=<0.0118 -0.013 0.0066 -0.9998>}
 index_metacarpal_l {parent=hand_l, loc=<1.6535 0.0183 -2.96>, rot=<0.0158 -0.0124 -0.0173 -0.9996>}
 index_01_l {parent=index_metacarpal_l, loc=<8.3008 0.0357 -0.0524>, rot=<-0.0044 0.0485 0.0924 -0.9945>}
 index_02_l {parent=index_01_l, loc=<3.3587 -0.3004 0.0452>, rot=<-0.002 0.0027 0.3225 -0.9466>}
 index_03_l {parent=index_02_l, loc=<2.5578 -0.123 -0.0252>, rot=<-0.002 0.0067 0.2558 -0.9667>}
 middle_metacarpal_l {parent=hand_l, loc=<2.1856 -0.1725 -0.5984>, rot=<0.006 0.0168 -0.0329 -0.9993>}
 middle_01_l {parent=middle_metacarpal_l, loc=<7.3275 -0.2447 0.0129>, rot=<-0.0208 0.0042 0.0858 -0.9961>}
 middle_02_l {parent=middle_01_l, loc=<4.0123 0.1401 0.1175>, rot=<-0.0218 0.0258 0.3512 -0.9357>}
 middle_03_l {parent=middle_02_l, loc=<2.6077 0.0534 0.1373>, rot=<0.0082 0.0188 0.1708 -0.9851>}
 ring_metacarpal_l {parent=hand_l, loc=<2.2074 -0.3625 1.3638>, rot=<-0.0284 0.0722 -0.0458 -0.9959>}
 ring_01_l {parent=ring_metacarpal_l, loc=<7.0295 -0.6197 -0.1446>, rot=<-0.0489 -0.0252 0.1892 -0.9804>}
 ring_02_l {parent=ring_01_l, loc=<3.6174 0.4647 -0.1027>, rot=<0 0 0.2806 -0.9598>}
 ring_03_l {parent=ring_02_l, loc=<2.4152 0.114 0.0199>, rot=<0 0 0.1189 -0.9929>}
 pinky_metacarpal_l {parent=hand_l, loc=<1.7616 -0.9753 3.3814>, rot=<-0.0397 0.0814 -0.069 -0.9935>}
 pinky_01_l {parent=pinky_metacarpal_l, loc=<7.1709 -0.4831 -0.1094>, rot=<-0.0654 -0.0035 0.2242 -0.9723>}
 pinky_02_l {parent=pinky_01_l, loc=<2.8274 0.0893 -0.1043>, rot=<0.0037 0.0011 0.2854 -0.9584>}
 pinky_03_l {parent=pinky_02_l, loc=<2.2046 0.0307 -0.0078>, rot=<-0.0002 0.0098 0.1445 -0.9895>}
 deform_wrist_dn_base_l {parent=hand_l, loc=<0 -2.2342 0>, rot=<0 0 0 -1>}
 deform_wrist_dn_l {parent=deform_wrist_dn_base_l, loc=<0 0 0>, rot=<0 0 0 1>}
 deform_wrist_up_base_l {parent=hand_l, loc=<0 2.2682 0>, rot=<0 0 0 -1>}
 deform_wrist_up_l {parent=deform_wrist_up_base_l, loc=<0 0 0>, rot=<0 0 0 1>}
 lowerarm_twist_02_l {parent=lowerarm_l, loc=<9.6199 0 0>, rot=<0 0 0 -1>}
 base_sleeve_01_l {parent=lowerarm_twist_02_l, loc=<3.9055 0.8402 3.8173>, rot=<-0.0597 -0.586 0.8057 0.0614>}
 dyn_sleeve_01_l {parent=base_sleeve_01_l, loc=<0 0 0>, rot=<0 0 0 -1>}
 base_sleeve_02_l {parent=lowerarm_twist_02_l, loc=<3.4144 -4.6071 -1.7269>, rot=<-0.1034 -0.9912 0.0702 0.0446>}
 dyn_sleeve_02_l {parent=base_sleeve_02_l, loc=<0 0 0>, rot=<0 0 0 1>}
 base_sleeve_03_l {parent=lowerarm_twist_02_l, loc=<5.2871 0.9873 -4.7514>, rot=<0.1079 0.4178 0.9013 0.0389>}
 dyn_sleeve_03_l {parent=base_sleeve_03_l, loc=<0 0 0>, rot=<0 0 0 1>}
 lowerarm_twist_01_l {parent=lowerarm_l, loc=<17.8549 0 0>, rot=<0 0 0 -1>}
 deform_elbow_in_base_l {parent=lowerarm_l, loc=<0.2112 5.259 0.8274>, rot=<0 0 0 1>}
 deform_elbow_in_l {parent=deform_elbow_in_base_l, loc=<0 0 0>, rot=<0 0 0 1>}
 deform_elbow_out_base_l {parent=lowerarm_l, loc=<0 -4 0>, rot=<0 0 0 1>}
 deform_elbow_out_l {parent=deform_elbow_out_base_l, loc=<0 0 0>, rot=<0 0 0 1>}
 upperarm_twist_01_l {parent=upperarm_l, loc=<0.0041 0 0>, rot=<0 0 0 -1>}
 deform_upperarm_bk_base_l {parent=upperarm_twist_01_l, loc=<1.0813 -7.2193 -1.9453>, rot=<0 0 0 -1>}
 deform_upperarm_bk_l {parent=deform_upperarm_bk_base_l, loc=<0 0 0>, rot=<0 0 0 1>}
 deform_upperarm_fr_base_l {parent=upperarm_twist_01_l, loc=<-0.3691 7.0506 -2.765>, rot=<0 0 0 -1>}
 deform_upperarm_fr_l {parent=deform_upperarm_fr_base_l, loc=<0 0 0>, rot=<0 0 0 1>}
 deform_upperarm_up_base_l {parent=upperarm_twist_01_l, loc=<-3.2216 -1.1391 3.3311>, rot=<0 0 0 1>}
 deform_upperarm_up_l {parent=deform_upperarm_up_base_l, loc=<0 0 0>, rot=<0 0 0 1>}
 upperarm_twist_02_l {parent=upperarm_l, loc=<13.7941 0 0>, rot=<0 0 0 -1>}
 clavicle_r {parent=spine_05, loc=<4.3575 -0.1466 2.3223>, rot=<0.7289 0.0465 -0.6829 0.009>}
 upperarm_r {parent=clavicle_r, loc=<-15.5111 0 0>, rot=<0.0226 -0.3234 -0.0088 -0.9459>}
 lowerarm_r {parent=upperarm_r, loc=<-26.6514 0 0>, rot=<0.0391 -0.0099 -0.2453 -0.9686>}
 hand_r {parent=lowerarm_r, loc=<-22.3491 0 0>, rot=<-0.7081 -0.0507 0.0975 -0.6975>}
 thumb_01_r {parent=hand_r, loc=<-1.3118 0.3191 2.459>, rot=<0.6546 -0.4042 0.0664 -0.6354>}
 thumb_02_r {parent=thumb_01_r, loc=<-5.9336 0.0448 -0.0401>, rot=<-0.0037 0.0213 0.298 -0.9543>}
 thumb_03_r {parent=thumb_02_r, loc=<-2.2223 -0.3682 0.0907>, rot=<0.0118 -0.013 0.0066 -0.9998>}
 index_metacarpal_r {parent=hand_r, loc=<-1.6535 -0.0183 2.96>, rot=<0.0158 -0.0124 -0.0173 -0.9996>}
 index_01_r {parent=index_metacarpal_r, loc=<-8.3008 -0.0357 0.0524>, rot=<-0.0044 0.0485 0.0924 -0.9945>}
 index_02_r {parent=index_01_r, loc=<-3.3587 0.3004 -0.0452>, rot=<-0.002 0.0027 0.3225 -0.9466>}
 index_03_r {parent=index_02_r, loc=<-2.5578 0.123 0.0252>, rot=<-0.002 0.0067 0.2558 -0.9667>}
 middle_metacarpal_r {parent=hand_r, loc=<-2.1856 0.1725 0.5984>, rot=<0.006 0.0168 -0.0329 -0.9993>}
 middle_01_r {parent=middle_metacarpal_r, loc=<-7.3275 0.2447 -0.0129>, rot=<-0.0208 0.0042 0.0858 -0.9961>}
 middle_02_r {parent=middle_01_r, loc=<-4.0123 -0.1401 -0.1175>, rot=<-0.0218 0.0258 0.3512 -0.9357>}
 middle_03_r {parent=middle_02_r, loc=<-2.6077 -0.0534 -0.1373>, rot=<0.0082 0.0188 0.1708 -0.9851>}
 ring_metacarpal_r {parent=hand_r, loc=<-2.2074 0.3625 -1.3638>, rot=<-0.0284 0.0722 -0.0458 -0.9959>}
 ring_01_r {parent=ring_metacarpal_r, loc=<-7.0295 0.6197 0.1446>, rot=<-0.0489 -0.0252 0.1892 -0.9804>}
 ring_02_r {parent=ring_01_r, loc=<-3.6174 -0.4647 0.1027>, rot=<0 0 0.2806 -0.9598>}
 ring_03_r {parent=ring_02_r, loc=<-2.4152 -0.114 -0.0199>, rot=<0 0 0.1189 -0.9929>}
 pinky_metacarpal_r {parent=hand_r, loc=<-1.7616 0.9753 -3.3814>, rot=<-0.0397 0.0814 -0.069 -0.9935>}
 pinky_01_r {parent=pinky_metacarpal_r, loc=<-7.1709 0.4831 0.1094>, rot=<-0.0654 -0.0035 0.2242 -0.9723>}
 pinky_02_r {parent=pinky_01_r, loc=<-2.8274 -0.0893 0.1043>, rot=<0.0037 0.0011 0.2854 -0.9584>}
 pinky_03_r {parent=pinky_02_r, loc=<-2.2046 -0.0307 0.0078>, rot=<-0.0002 0.0098 0.1445 -0.9895>}
 deform_wrist_dn_base_r {parent=hand_r, loc=<-0.0004 2.2345 -0.0001>, rot=<0 0 0 -1>}
 deform_wrist_dn_r {parent=deform_wrist_dn_base_r, loc=<0 0 0>, rot=<0 0 0 1>}
 deform_wrist_up_base_r {parent=hand_r, loc=<0.0003 -2.2684 0.0001>, rot=<0 0 0 -1>}
 deform_wrist_up_r {parent=deform_wrist_up_base_r, loc=<0 0 0>, rot=<0 0 0 1>}
 lowerarm_twist_02_r {parent=lowerarm_r, loc=<-9.6199 0 0>, rot=<0 0 0 -1>}
 base_sleeve_01_r {parent=lowerarm_twist_02_r, loc=<-3.1127 1.7432 -4.2253>, rot=<0.0689 0.6173 0.7806 0.0696>}
 dyn_sleeve_01_r {parent=base_sleeve_01_r, loc=<0 0 0>, rot=<0 0 0 1>}
 base_sleeve_02_r {parent=lowerarm_twist_02_r, loc=<-2.4921 4.5314 1.0447>, rot=<0.0325 -0.0604 0.993 0.0962>}
 dyn_sleeve_02_r {parent=base_sleeve_02_r, loc=<0 0 0>, rot=<0 0 0 1>}
 base_sleeve_03_r {parent=lowerarm_twist_02_r, loc=<-5.5611 -4.4998 0.8281>, rot=<0.0358 -0.9991 -0.0074 0.0224>}
 dyn_sleeve_03_r {parent=base_sleeve_03_r, loc=<0 0 0>, rot=<0 0 0 1>}
 base_sleeve_03_dn_r {parent=dyn_sleeve_03_r, loc=<2.8591 0 0>, rot=<0 0 0 1>}
 dyn_sleeve_03_dn_r {parent=base_sleeve_03_dn_r, loc=<0 0 0>, rot=<0 0 0 1>}
 hand_r_null {parent=lowerarm_twist_02_r, type=Null, loc=<-12.7293 0 0>, rot=<0.7081 0.0507 -0.0975 0.6975>}
 lowerarm_twist_01_r {parent=lowerarm_r, loc=<-17.8549 0 0>, rot=<0 0 0 -1>}
 deform_elbow_in_base_r {parent=lowerarm_r, loc=<-0.2111 -5.2591 -0.8275>, rot=<0 0 0 -1>}
 deform_elbow_in_r {parent=deform_elbow_in_base_r, loc=<-0.0001 0 0.0002>, rot=<0 0 0 1>}
 deform_elbow_out_base_r {parent=lowerarm_r, loc=<0 4 -0.0001>, rot=<0 0 0 -1>}
 deform_elbow_out_r {parent=deform_elbow_out_base_r, loc=<0 0 0>, rot=<0 0 0 1>}
 upperarm_twist_01_r {parent=upperarm_r, loc=<-0.0041 0 0>, rot=<0 0 0 -1>}
 deform_upperarm_bk_base_r {parent=upperarm_twist_01_r, loc=<-1.0814 7.2192 1.9455>, rot=<0 0 0 1>}
 deform_upperarm_bk_r {parent=deform_upperarm_bk_base_r, loc=<0 0 0>, rot=<0 0 0 1>}
 deform_upperarm_fr_base_r {parent=upperarm_twist_01_r, loc=<0.3689 -7.0506 2.7652>, rot=<0 0 0 1>}
 deform_upperarm_fr_r {parent=deform_upperarm_fr_base_r, loc=<0 0 0>, rot=<0 0 0 1>}
 deform_upperarm_up_base_r {parent=upperarm_twist_01_r, loc=<3.2214 1.1391 -3.3308>, rot=<0 0 0 1>}
 deform_upperarm_up_r {parent=deform_upperarm_up_base_r, loc=<0 0 0>, rot=<0 0 0 -1>}
 upperarm_twist_02_r {parent=upperarm_r, loc=<-13.7941 0 0>, rot=<0 0 0 -1>}
 deform_pec_aim_base_l {parent=spine_04, loc=<7.8732 12.6388 -2.4123>, rot=<0.0102 -0.4602 0.3979 -0.7936>}
 deform_pec_aim_l {parent=deform_pec_aim_base_l, loc=<0 0 0>, rot=<0 0 0 1>}
 deform_pec_base_l {parent=deform_pec_aim_l, loc=<10.766 0 0>, rot=<0 0 0 1>}
 deform_pec_l {parent=deform_pec_base_l, loc=<0.055 0 0>, rot=<0 0 0 1>}
 deform_pec_aim_base_r {parent=spine_04, loc=<7.8727 12.6388 2.4123>, rot=<-0.4602 -0.0102 0.7936 0.3979>}
 deform_pec_aim_r {parent=deform_pec_aim_base_r, loc=<0 0 0>, rot=<0 0 0 1>}
 deform_pec_base_r {parent=deform_pec_aim_r, loc=<-10.7662 -0.0006 -0.0007>, rot=<0 0 0 1>}
 deform_pec_r {parent=deform_pec_base_r, loc=<0 0 0>, rot=<0 0 0 1>}
 beltStrings_Master {parent=spine_03, loc=<-3.8292 9.5929 -4.5917>, rot=<-0.0617 -0.9935 0.0908 -0.0305>}
 beltStrings_Sub {parent=beltStrings_Master, loc=<0 0 0>, rot=<0 0 0 1>}
 base_beltString_2_01 {parent=beltStrings_Sub, loc=<0.5212 0.128 0.2939>, rot=<0 0 0 1>}
 dyn_beltString_2_01 {parent=base_beltString_2_01, loc=<0 0 0>, rot=<0 0 0 1>}
 dyn_beltString_2_02 {parent=dyn_beltString_2_01, loc=<14.8064 0 0>, rot=<0 0 0.0679 -0.9977>}
 dyn_beltString_2_03 {parent=dyn_beltString_2_02, loc=<15.331 0.0081 -0.1165>, rot=<-0.0346 -0.0002 0.0062 -0.9994>}
 dyn_beltString_2_04 {parent=dyn_beltString_2_03, loc=<14.0343 0.0025 0.0302>, rot=<-0.0143 -0.023 -0.0003 -0.9996>}
 base_beltString_1_01 {parent=beltStrings_Sub, loc=<0.4272 -0.1991 -0.2169>, rot=<-0.0094 -0.0185 0.0272 -0.9994>}
 dyn_beltString_1_01 {parent=base_beltString_1_01, loc=<0 0 0>, rot=<0 0 0 1>}
 dyn_beltString_1_02 {parent=dyn_beltString_1_01, loc=<16.8741 0 0>, rot=<0 0 0.0638 -0.998>}
 dyn_beltString_1_03 {parent=dyn_beltString_1_02, loc=<16.6675 0 0>, rot=<0.1001 -0.0001 -0.0006 -0.995>}
 dyn_beltString_1_04 {parent=dyn_beltString_1_03, loc=<16.6843 0 0>, rot=<0 0 0.0042 -1>}
 Skirt_Master {parent=spine_02, loc=<6.0104 0 0>, rot=<0 0 0.0149 -0.9999>}
 Skirt_MasterOffset {parent=Skirt_Master, loc=<0 0 0>, rot=<0 0 0 1>}
 base_skirt_Back_l_01 {parent=Skirt_MasterOffset, loc=<-0.942 -5.1698 -3.565>, rot=<0.1092 -0.9852 -0.1321 -0.0072>}
 dyn_skirt_Back_l_01 {parent=base_skirt_Back_l_01, loc=<0 0 0>, rot=<0 0 0 1>}
 dyn_skirt_Back_l_02 {parent=dyn_skirt_Back_l_01, loc=<19.3916 0 0>, rot=<0 0 0 1>}
 dyn_skirt_Back_l_03 {parent=dyn_skirt_Back_l_02, loc=<18.6381 0 0>, rot=<0 0 0 1>}
 dyn_skirt_Back_l_04 {parent=dyn_skirt_Back_l_03, loc=<18.957 0 0>, rot=<0 0 0 1>}
 dyn_skirt_Back_l_05 {parent=dyn_skirt_Back_l_04, loc=<18.957 0 0>, rot=<0 0 0 1>}
 base_skirt_SideBack_l_01 {parent=Skirt_MasterOffset, loc=<-1.3755 -1.8114 -8.2799>, rot=<0.1204 -0.8348 -0.537 -0.0181>}
 dyn_skirt_SideBack_l_01 {parent=base_skirt_SideBack_l_01, loc=<0 0 0>, rot=<0 0 0 1>}
 dyn_skirt_SideBack_l_02 {parent=dyn_skirt_SideBack_l_01, loc=<19.3916 0 0>, rot=<0 0 0 1>}
 dyn_skirt_SideBack_l_03 {parent=dyn_skirt_SideBack_l_02, loc=<18.6381 0 0>, rot=<0 0 0 1>}
 dyn_skirt_SideBack_l_04 {parent=dyn_skirt_SideBack_l_03, loc=<18.957 0 0>, rot=<0 0 0 1>}
 dyn_skirt_SideBack_l_05 {parent=dyn_skirt_SideBack_l_04, loc=<18.957 0 0>, rot=<0 0 0 1>}
 base_skirt_SideBack_r_01 {parent=Skirt_MasterOffset, loc=<-1.3696 -1.8643 8.1131>, rot=<0.1204 -0.8348 0.537 0.0181>}
 dyn_skirt_SideBack_r_01 {parent=base_skirt_SideBack_r_01, loc=<0 0 0>, rot=<0 0 0 1>}
 dyn_skirt_SideBack_r_02 {parent=dyn_skirt_SideBack_r_01, loc=<19.3916 0 0>, rot=<0 0 0 1>}
 dyn_skirt_SideBack_r_03 {parent=dyn_skirt_SideBack_r_02, loc=<18.6381 0 0>, rot=<0 0 0 1>}
 dyn_skirt_SideBack_r_04 {parent=dyn_skirt_SideBack_r_03, loc=<18.957 0 0>, rot=<0 0 0 1>}
 dyn_skirt_SideBack_r_05 {parent=dyn_skirt_SideBack_r_04, loc=<18.957 0 0>, rot=<0 0 0 1>}
 base_skirt_Back_r_01 {parent=Skirt_MasterOffset, loc=<-0.942 -5.1698 3.1532>, rot=<0.1092 -0.9852 0.1321 0.0072>}
 dyn_skirt_Back_r_01 {parent=base_skirt_Back_r_01, loc=<0 0 0>, rot=<0 0 0 1>}
 dyn_skirt_Back_r_02 {parent=dyn_skirt_Back_r_01, loc=<19.3916 0 0>, rot=<0 0 0 1>}
 dyn_skirt_Back_r_03 {parent=dyn_skirt_Back_r_02, loc=<18.6381 0 0>, rot=<0 0 0 1>}
 dyn_skirt_Back_r_04 {parent=dyn_skirt_Back_r_03, loc=<18.957 0 0>, rot=<0 0 0 1>}
 dyn_skirt_Back_r_05 {parent=dyn_skirt_Back_r_04, loc=<18.957 0 0>, rot=<0 0 0 1>}
 base_skirt_SideFront_l_01 {parent=Skirt_MasterOffset, loc=<-1.9535 4.0289 -8.9924>, rot=<-0.088 0.6036 0.7923 0.0117>}
 dyn_skirt_SideFront_l_01 {parent=base_skirt_SideFront_l_01, loc=<0 0 0>, rot=<0 0 0 1>}
 dyn_skirt_SideFront_l_02 {parent=dyn_skirt_SideFront_l_01, loc=<19.3916 0 0>, rot=<0 0 0 1>}
 dyn_skirt_SideFront_l_03 {parent=dyn_skirt_SideFront_l_02, loc=<18.6381 0 0>, rot=<0 0 0 1>}
 dyn_skirt_SideFront_l_04 {parent=dyn_skirt_SideFront_l_03, loc=<18.957 0 0>, rot=<0 0 0 1>}
 dyn_skirt_SideFront_l_05 {parent=dyn_skirt_SideFront_l_04, loc=<18.957 0 0>, rot=<0 0 0 1>}
 base_skirt_SideFront_r_01 {parent=Skirt_MasterOffset, loc=<-1.9545 4.0384 9.2938>, rot=<0.088 -0.6036 0.7923 0.0117>}
 dyn_skirt_SideFront_r_01 {parent=base_skirt_SideFront_r_01, loc=<0 0 0>, rot=<0 0 0 1>}
 dyn_skirt_SideFront_r_02 {parent=dyn_skirt_SideFront_r_01, loc=<19.3916 0 0>, rot=<0 0 0 1>}
 dyn_skirt_SideFront_r_03 {parent=dyn_skirt_SideFront_r_02, loc=<18.6381 0 0>, rot=<0 0 0 1>}
 dyn_skirt_SideFront_r_04 {parent=dyn_skirt_SideFront_r_03, loc=<18.957 0 0>, rot=<0 0 0 1>}
 dyn_skirt_SideFront_r_05 {parent=dyn_skirt_SideFront_r_04, loc=<18.957 0 0>, rot=<0 0 0 1>}
 base_skirt_Front_r_01 {parent=Skirt_MasterOffset, loc=<-2.4524 8.566 4.722>, rot=<0.0459 -0.0832 0.9953 -0.0191>}
 dyn_skirt_Front_r_01 {parent=base_skirt_Front_r_01, loc=<0 0 0>, rot=<0 0 0 1>}
 dyn_skirt_Front_r_02 {parent=dyn_skirt_Front_r_01, loc=<19.3916 0 0>, rot=<0 0 0 1>}
 dyn_skirt_Front_r_03 {parent=dyn_skirt_Front_r_02, loc=<18.6381 0 0>, rot=<0 0 0 1>}
 dyn_skirt_Front_r_04 {parent=dyn_skirt_Front_r_03, loc=<18.957 0 0>, rot=<0 0 0 1>}
 dyn_skirt_Front_r_05 {parent=dyn_skirt_Front_r_04, loc=<18.957 0 0>, rot=<0 0 0 1>}
 base_skirt_Front_l_01 {parent=Skirt_MasterOffset, loc=<-2.4154 8.2294 -4.7217>, rot=<-0.0459 0.0832 0.9953 -0.0191>}
 dyn_skirt_Front_l_01 {parent=base_skirt_Front_l_01, loc=<0 0 0>, rot=<0 0 0 1>}
 dyn_skirt_Front_l_02 {parent=dyn_skirt_Front_l_01, loc=<19.3916 0 0>, rot=<0 0 0 1>}
 dyn_skirt_Front_l_03 {parent=dyn_skirt_Front_l_02, loc=<18.6381 0 0>, rot=<0 0 0 1>}
 dyn_skirt_Front_l_04 {parent=dyn_skirt_Front_l_03, loc=<18.957 0 0>, rot=<0 0 0 1>}
 dyn_skirt_Front_l_05 {parent=dyn_skirt_Front_l_04, loc=<18.957 0 0>, rot=<0 0 0 1>}
 thigh_l {parent=pelvis, loc=<-2.5693 -0.0109 -7.7501>, rot=<0.0012 0.0572 0.0155 -0.9982>}
 calf_l {parent=thigh_l, loc=<-39.0484 0 0>, rot=<0.0006 -0.0161 -0.0391 -0.9991>}
 foot_l {parent=calf_l, loc=<-42.561 0.1077 0.4292>, rot=<0.0418 -0.0455 0.0343 -0.9975>}
 ball_l {parent=foot_l, loc=<-6.1599 13.8264 0.6806>, rot=<0.0001 -0.0001 -0.7063 -0.7079>}
 calf_twist_02_l {parent=calf_l, loc=<-18.8064 -0.0967 0.0397>, rot=<0.0237 0.0095 -0.0035 -0.9997>}
 calf_twist_01_l {parent=calf_l, loc=<-35.813 -0.0967 0.0397>, rot=<0.0237 0.0095 -0.0035 -0.9997>}
 deform_knee_in_base_l {parent=calf_l, loc=<1.298 -5.2957 -0.0496>, rot=<0 0 0 -1>}
 deform_knee_in_l {parent=deform_knee_in_base_l, loc=<0 0 0>, rot=<0 0 0 1>}
 deform_knee_out_base_l {parent=calf_l, loc=<0 4 0>, rot=<0 0 0 -1>}
 deform_knee_out_l {parent=deform_knee_out_base_l, loc=<0 0 0>, rot=<0 0 0 1>}
 thigh_twist_01_l {parent=thigh_l, loc=<-21.6303 0 0>, rot=<0.0013 0 0 -1>}
 deform_groin_base_l {parent=thigh_l, loc=<-2.1908 8.1142 -1.7442>, rot=<0 0 0 -1>}
 deform_groin_l {parent=deform_groin_base_l, loc=<0 0 0>, rot=<0 0 0 1>}
 thigh_r {parent=pelvis, loc=<-2.5693 -0.0109 7.7501>, rot=<0.0572 -0.0012 0.9982 0.0155>}
 calf_r {parent=thigh_r, loc=<39.0484 0 0>, rot=<0.0006 -0.0161 -0.0391 -0.9991>}
 foot_r {parent=calf_r, loc=<42.561 -0.1077 -0.4292>, rot=<0.0418 -0.0455 0.0343 -0.9975>}
 ball_r {parent=foot_r, loc=<6.1599 -13.8264 -0.6806>, rot=<0.0001 -0.0001 -0.7063 -0.7079>}
 calf_twist_02_r {parent=calf_r, loc=<18.8064 0.0967 -0.0397>, rot=<0.0237 0.0095 -0.0035 -0.9997>}
 calf_twist_01_r {parent=calf_r, loc=<35.813 0.0967 -0.0397>, rot=<0.0237 0.0095 -0.0035 -0.9997>}
 deform_knee_in_base_r {parent=calf_r, loc=<-1.298 5.2957 0.0496>, rot=<0 0 0 1>}
 deform_knee_in_r {parent=deform_knee_in_base_r, loc=<0 0 0>, rot=<0 0 0 1>}
 deform_knee_out_base_r {parent=calf_r, loc=<0 -4 0>, rot=<0 0 0 -1>}
 deform_knee_out_r {parent=deform_knee_out_base_r, loc=<0 0 0>, rot=<0 0 0 1>}
 thigh_twist_01_r {parent=thigh_r, loc=<21.6303 0 0>, rot=<0.0013 0 0 -1>}
 deform_groin_base_r {parent=thigh_r, loc=<2.1908 -8.1142 1.7442>, rot=<0 0 0 -1>}
 deform_groin_r {parent=deform_groin_base_r, loc=<0 0 0>, rot=<0 0 0 1>}
 deform_glute_base_l {parent=pelvis, loc=<-8.8064 -8.8222 -8.4287>, rot=<0.0275 0.0501 -0.4472 -0.8926>}
 deform_glute_l {parent=deform_glute_base_l, loc=<0 0 0>, rot=<0 0 0 1>}
 deform_glute_base_r {parent=pelvis, loc=<-8.8062 -8.8222 8.4287>, rot=<0.0501 -0.0275 0.8926 -0.4472>}
 deform_glute_r {parent=deform_glute_base_r, loc=<0 0 0>, rot=<0 0 0 1>}
 ground_mover {parent=root, loc=<0 0 0>, rot=<0 0 0 1>}
 Alpha RBAN {type=Curve, loc=<0 0 0>, rot=<0 0 0 1>}
]
variables [
 Alpha RBAN {type=double, default=0}
 Calf L Backwards {type=double, default=0}
 Calf R Backwards {type=double, default=0}
 Forward Start {type=double, default=0}
 Thigh L Backwards {type=double, default=0}
 Thigh L Forward {type=double, default=0}
 Thigh L Forward Extreme {type=double, default=0}
 Thigh L Side {type=double, default=0}
 Thigh R Backwards {type=double, default=0}
 Thigh R Forward {type=double, default=0}
 Thigh R Inwards {type=double, default=0}
 Thigh R Side {type=double, default=0}
 Timer {type=double, default=0}
 is Baseball Sliding {type=double, default=0}
 "is Crouching " {type=double, default=0}
 is Knee Sliding {type=double, default=0}
 "is Moving 2D " {type=double, default=0}
 is Sky Dive {type=double, default=0}
 is Sprinting {type=bool, default=False}
 is Standing Moving {type=double, default=0}
]
