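# Physics asset for the M_MED_BountyPond_Quart male body cosmetic, setting up a simulated long skirt or coat.
# Eight five-segment capsule chains hang from base bodies linked to spine_02, with tight twist and swing limits that generally loosen downward and at the front.
# A few free-rotation links join neighboring chains.
# Pelvis, spine and leg capsules provide body shapes, and a huge ground_mover capsule appears to approximate the floor.
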PhysicsAsset /BRCosmetics/Characters/Player/Male/Medium/Bodies/M_MED_BountyPond_Quart/Meshes/M_MED_BountyPond_Quart_Physics
bodies {
 base_skirt_Back_l_01 [{shapeType=Capsule, length=0.505, radius=0.505, shapeName=base_skirt_Back_l_01_capsule, rotation=<rot 0 -90 0>}]
 base_skirt_Back_r_01 [{shapeType=Capsule, length=0.505, radius=0.505, shapeName=base_skirt_Back_r_01_capsule, rotation=<rot 0 -90 0>}]
 base_skirt_Front_l_01 [{shapeType=Capsule, length=0.505, radius=0.505, shapeName=base_skirt_Front_l_01_capsule, rotation=<rot 0 -90 0>}]
 base_skirt_Front_r_01 [{shapeType=Capsule, length=0.505, radius=0.505, shapeName=base_skirt_Front_r_01_capsule, rotation=<rot 0 -90 0>}]
 base_skirt_SideBack_l_01 [{shapeType=Capsule, length=0.505, radius=0.505, shapeName=base_skirt_SideBack_l_01_capsule, rotation=<rot 0 -90 0>}]
 base_skirt_SideBack_r_01 [{shapeType=Capsule, length=0.505, radius=0.505, shapeName=base_skirt_SideBack_r_01_capsule, rotation=<rot 0 -90 0>}]
 base_skirt_SideFront_l_01 [{shapeType=Capsule, length=0.505, radius=0.505, shapeName=base_skirt_SideFront_l_01_capsule, rotation=<rot 0 -90 0>}]
 base_skirt_SideFront_r_01 [{shapeType=Capsule, length=0.505, radius=0.505, shapeName=base_skirt_SideFront_r_01_capsule, rotation=<rot 0 -90 0>}]
 calf_l [
  {
   shapeType=Capsule
   length=38.9812
   radius=6.4738
   shapeName=calf_l_capsule
   center=<-23.3121 -2.0135 1.034>
   rotation=<rot 88.914 179.6173 -177.84>
  }
  {
   shapeType=Capsule
   length=6.7502
   radius=6.1042
   shapeName=calf_r_capsule
   center=<-4.3692 -1.2947 1.8236>
   rotation=<rot 81.4055 -104.1397 -106.914>
  }
 ]
 calf_r [
  {
   shapeType=Capsule
   length=38.9812
   radius=6.4738
   shapeName=calf_l_capsule
   center=<23.3121 2.0135 -1.034>
   rotation=<rot 88.914 179.6173 -177.84>
  }
  {
   shapeType=Capsule
   length=6.7502
   radius=6.1042
   shapeName=calf_r_capsule
   center=<4.3692 1.2947 -1.8236>
   rotation=<rot 81.4055 -104.1397 -106.914>
  }
 ]
 dyn_skirt_Back_l_01 [{shapeType=Capsule, length=13, radius=9, shapeName=dyn_skirt_Back_l_01_capsule, center=<4.3317 -7.7522 0>}]
 dyn_skirt_Back_l_02 [{shapeType=Capsule, length=16, radius=10, shapeName=dyn_skirt_Back_l_02_capsule, center=<4.455 -9.8887 0>}]
 dyn_skirt_Back_l_03 [{
   shapeType=Capsule
   length=19.4066
   radius=10
   shapeName=dyn_skirt_Back_l_03_capsule
   center=<4.2123 -8.1195 0.0831>
   rotation=<rot 0 0.0001 -4.3216>
  }]
 dyn_skirt_Back_l_04 [{
   shapeType=Capsule
   length=34.8737
   radius=10
   shapeName=dyn_skirt_Back_l_04_capsule
   center=<5.4513 -8.0711 -0.02>
   rotation=<rot 0 0.0002 -7.0516>
  }]
 dyn_skirt_Back_l_05 [{
   shapeType=Capsule
   length=39.808
   radius=11.0387
   shapeName=dyn_cape_b_l_04_capsule
   center=<6.8362 -8.2411 0.0755>
   rotation=<rot 0 0.0002 -5.1056>
  }]
 dyn_skirt_Back_r_01 [{shapeType=Capsule, length=13, radius=9, shapeName=dyn_skirt_Back_r_01_capsule, center=<4.3317 -7.7522 0>}]
 dyn_skirt_Back_r_02 [{shapeType=Capsule, length=16, radius=10, shapeName=dyn_skirt_Back_r_02_capsule, center=<4.455 -9.8887 0>}]
 dyn_skirt_Back_r_03 [{
   shapeType=Capsule
   length=19.4066
   radius=10
   shapeName=dyn_skirt_Back_r_03_capsule
   center=<4.2123 -8.1195 -0.0831>
   rotation=<rot 0 0.0001 4.3216>
  }]
 dyn_skirt_Back_r_04 [{
   shapeType=Capsule
   length=34.8737
   radius=10
   shapeName=dyn_skirt_Back_r_04_capsule
   center=<5.4513 -7.8106 -0.0122>
   rotation=<rot 0 0.0002 7.0516>
  }]
 dyn_skirt_Back_r_05 [{
   shapeType=Capsule
   length=39.808
   radius=11.0387
   shapeName=dyn_cape_b_r_05_capsule
   center=<6.8362 -7.8608 -0.1095>
   rotation=<rot -0.0003 0.001 5.1056>
  }]
 dyn_skirt_Front_l_01 [{shapeType=Capsule, length=13, radius=9, shapeName=dyn_skirt_Front_l_01_capsule, center=<4.3317 -8.3695 0>}]
 dyn_skirt_Front_l_02 [{shapeType=Capsule, length=16, radius=10, shapeName=dyn_skirt_Front_l_02_capsule, center=<4.455 -8.5184 0>}]
 dyn_skirt_Front_l_03 [{
   shapeType=Capsule
   length=21.783
   radius=10
   shapeName=dyn_skirt_Front_l_03_capsule
   center=<4.702 -6.5798 0.0408>
   rotation=<rot 0 -0.0001 -6.2358>
  }]
 dyn_skirt_Front_l_04 [{
   shapeType=Capsule
   length=23.4434
   radius=10
   shapeName=dyn_skirt_Front_l_04_capsule
   center=<5.4513 -4.872 -2.1417>
   rotation=<rot 1.6647 -0.0969 -2.5298>
  }]
 dyn_skirt_Front_l_05 [{
   shapeType=Capsule
   length=25.9311
   radius=10
   shapeName=dyn_skirt_Front_l_05_capsule
   center=<5.7145 -5.4607 -1.9666>
   rotation=<rot 1.7528 -0.0004 -2.6863>
  }]
 dyn_skirt_Front_r_01 [{shapeType=Capsule, length=13, radius=9, shapeName=dyn_skirt_Front_r_01_capsule, center=<4.3317 -8.3695 0>}]
 dyn_skirt_Front_r_02 [{
   shapeType=Capsule
   length=16
   radius=10
   shapeName=dyn_skirt_Front_r_02_capsule
   center=<4.455 -6.639 4.8509>
   rotation=<rot 0 -0.0002 19.0244>
  }]
 dyn_skirt_Front_r_03 [{
   shapeType=Capsule
   length=18.5345
   radius=10.8721
   shapeName=dyn_skirt_Front_r_03_capsule
   center=<4.2124 -5.2205 7.3982>
   rotation=<rot 0 -0.0003 23.2263>
  }]
 dyn_skirt_Front_r_04 [{
   shapeType=Capsule
   length=20.1735
   radius=10.8935
   shapeName=dyn_skirt_Front_r_04_capsule
   center=<5.4513 -2.2674 7.6066>
   rotation=<rot 0 -0.0003 22.8018>
  }]
 dyn_skirt_Front_r_05 [{
   shapeType=Capsule
   length=23.2045
   radius=10.3502
   shapeName=dyn_skirt_Front_r_05_capsule
   center=<6.3158 -1.0037 7.5436>
   rotation=<rot -0.0001 -0.0003 21.4321>
  }]
 dyn_skirt_SideBack_l_01 [{shapeType=Capsule, length=13, radius=9, shapeName=dyn_skirt_SideBack_l_01_capsule, center=<4.3317 -7.6585 0>}]
 dyn_skirt_SideBack_l_02 [{shapeType=Capsule, length=16, radius=10, shapeName=dyn_skirt_SideBack_l_02_capsule, center=<4.455 -9.0905 0>}]
 dyn_skirt_SideBack_l_03 [{shapeType=Capsule, length=19.4066, radius=10, shapeName=dyn_skirt_SideBack_l_03_capsule, center=<4.2123 -7.9467 0>}]
 dyn_skirt_SideBack_l_04 [{shapeType=Capsule, length=21.067, radius=10, shapeName=dyn_skirt_SideBack_l_04_capsule, center=<5.4513 -7.9948 0>}]
 dyn_skirt_SideBack_l_05 [{
   shapeType=Capsule
   length=26.1176
   radius=10
   shapeName=dyn_skirt_SideBack_r_05_capsule
   center=<5.6543 -6.9083 -2.8876>
   rotation=<rot 0.0002 0 -14.5466>
  }]
 dyn_skirt_SideBack_r_01 [{
   shapeType=Capsule
   length=13
   radius=9
   shapeName=dyn_skirt_SideBack_r_01_capsule
   center=<4.3317 -7.4988 -0.0294>
   rotation=<rot -0.0004 0.0002 10.4307>
  }]
 dyn_skirt_SideBack_r_02 [{
   shapeType=Capsule
   length=16
   radius=10
   shapeName=dyn_skirt_SideBack_r_02_capsule
   center=<4.455 -8.9308 -0.0294>
   rotation=<rot -0.0004 0.0002 10.4307>
  }]
 dyn_skirt_SideBack_r_03 [{
   shapeType=Capsule
   length=19.5608
   radius=10.7038
   shapeName=dyn_skirt_SideBack_r_03_capsule
   center=<4.2844 -6.4162 6.2642>
   rotation=<rot 3.5067 1.6829 30.8126>
  }]
 dyn_skirt_SideBack_r_04 [{
   shapeType=Capsule
   length=27.9194
   radius=10
   shapeName=dyn_skirt_SideBack_r_04_capsule
   center=<5.3171 -6.4489 6.7015>
   rotation=<rot 3.4353 1.9967 30.1774>
  }]
 dyn_skirt_SideBack_r_05 [{
   shapeType=Capsule
   length=29.1571
   radius=10
   shapeName=dyn_skirt_SideBack_r_05_capsule
   center=<5.6064 -4.6044 7.388>
   rotation=<rot 3.6175 2.2254 31.6216>
  }]
 dyn_skirt_SideFront_l_01 [{shapeType=Capsule, length=13, radius=9, shapeName=dyn_skirt_SideFront_l_01_capsule, center=<4.3317 -7.6205 0>}]
 dyn_skirt_SideFront_l_02 [{shapeType=Capsule, length=16, radius=10, shapeName=dyn_skirt_SideFront_l_02_capsule, center=<4.455 -8.912 0>}]
 dyn_skirt_SideFront_l_03 [{shapeType=Capsule, length=26.3461, radius=10, shapeName=dyn_skirt_SideFront_l_03_capsule, center=<4.6873 -7.721 0>}]
 dyn_skirt_SideFront_l_04 [{shapeType=Capsule, length=28.0065, radius=10, shapeName=dyn_skirt_SideFront_l_04_capsule, center=<5.4513 -7.2852 0>}]
 dyn_skirt_SideFront_l_05 [{shapeType=Capsule, length=33.0572, radius=10, shapeName=dyn_skirt_SideFront_l_05_capsule, center=<5.6543 -6.1099 0>}]
 dyn_skirt_SideFront_r_01 [{shapeType=Capsule, length=13, radius=9, shapeName=dyn_skirt_SideFront_r_01_capsule, center=<4.3317 -7.6205 0>}]
 dyn_skirt_SideFront_r_02 [{shapeType=Capsule, length=16, radius=10, shapeName=dyn_skirt_SideFront_r_02_capsule, center=<4.455 -8.912 0>}]
 dyn_skirt_SideFront_r_03 [{shapeType=Capsule, length=23.4397, radius=10.3961, shapeName=dyn_skirt_SideFront_r_03_capsule, center=<4.6796 -7.9422 1.8368>}]
 dyn_skirt_SideFront_r_04 [{
   shapeType=Capsule
   length=26.6794
   radius=10.725
   shapeName=dyn_skirt_SideFront_r_04_capsule
   center=<5.4513 -6.6562 5.4885>
   rotation=<rot -0.0001 0 2.6907>
  }]
 dyn_skirt_SideFront_r_05 [{
   shapeType=Capsule
   length=25.5759
   radius=10.5418
   shapeName=dyn_skirt_SideFront_r_05_capsule
   center=<6.1368 -5.8865 5.3455>
   rotation=<rot 0 0 2.8684>
  }]
 foot_l [{shapeType=Capsule, length=15, radius=12.75, shapeName=foot_l_capsule, center=<-4.0948 6.1521 -0.1893>, rotation=<rot 0 0 90>}]
 foot_r [{shapeType=Capsule, length=15, radius=12.25, shapeName=foot_r_capsule, center=<4.0948 -4.8644 0.1893>, rotation=<rot 0 0 -90>}]
 ground_mover [{shapeType=Capsule, length=0.5056, radius=337.6725, shapeName=floor_joint_capsule, center=<2.6954 -3.0549 -337.3088>, rotation=<rot 0 90 -90>}]
 pelvis [{shapeType=Capsule, length=5.6891, radius=7.982, shapeName=pelvis_capsule, center=<0 -0.1654 0>}]
 spine_01 [{shapeType=Capsule, length=7.2791, radius=5.4353, shapeName=spine_01_capsule, center=<0 1.1594 0>}]
 spine_02 [{shapeType=Capsule, length=4.2576, radius=6.0103, shapeName=spine_02_capsule, center=<0 1.6446 0>}]
 thigh_l [{
   shapeType=Capsule
   length=29.9054
   radius=6.2294
   shapeName=thigh_l_capsule
   center=<-19.8301 -0.0245 2.3975>
   rotation=<rot 89.4575 0.1673 -1.0885>
  }]
 thigh_r [{
   shapeType=Capsule
   length=35.4834
   radius=6.4923
   shapeName=thigh_l_capsule
   center=<21.6713 0.3711 -2.5283>
   rotation=<rot 89.4007 -179.6371 179.2638>
  }]
}
constraints [
 {bone1Name=dyn_skirt_Back_l_01, bone2Name=base_skirt_Back_l_01, swing1LimitDegrees=5, swing2LimitDegrees=2, twistLimitDegrees=1}
 {bone1Name=dyn_skirt_Back_l_02, bone2Name=dyn_skirt_Back_l_01, swing1LimitDegrees=7, swing2LimitDegrees=2, twistLimitDegrees=1}
 {bone1Name=dyn_skirt_Back_l_03, bone2Name=dyn_skirt_Back_l_02, swing1LimitDegrees=10, swing2LimitDegrees=3, twistLimitDegrees=1}
 {bone1Name=dyn_skirt_Back_l_04, bone2Name=dyn_skirt_Back_l_03, swing1LimitDegrees=10, swing2LimitDegrees=5, twistLimitDegrees=1}
 {bone1Name=dyn_skirt_Back_l_05, bone2Name=dyn_skirt_Back_l_04, swing1LimitDegrees=12, swing2LimitDegrees=5, twistLimitDegrees=1}
 {bone1Name=dyn_skirt_Back_r_01, bone2Name=base_skirt_Back_r_01, swing1LimitDegrees=5, swing2LimitDegrees=2, twistLimitDegrees=1}
 {bone1Name=dyn_skirt_Back_r_02, bone2Name=dyn_skirt_Back_r_01, swing1LimitDegrees=7, swing2LimitDegrees=2, twistLimitDegrees=1}
 {bone1Name=dyn_skirt_Back_r_03, bone2Name=dyn_skirt_Back_r_02, swing1LimitDegrees=10, swing2LimitDegrees=3, twistLimitDegrees=1}
 {bone1Name=dyn_skirt_Back_r_04, bone2Name=dyn_skirt_Back_r_03, swing1LimitDegrees=10, swing2LimitDegrees=5, twistLimitDegrees=1}
 {bone1Name=dyn_skirt_Back_r_05, bone2Name=dyn_skirt_Back_r_04, swing1LimitDegrees=12, swing2LimitDegrees=5, twistLimitDegrees=1}
 {bone1Name=dyn_skirt_Front_l_01, bone2Name=base_skirt_Front_l_01, swing1LimitDegrees=10, swing2LimitDegrees=2, twistLimitDegrees=1}
 {bone1Name=dyn_skirt_Front_l_02, bone2Name=dyn_skirt_Front_l_01, swing1LimitDegrees=14, swing2LimitDegrees=2, twistLimitDegrees=1}
 {bone1Name=dyn_skirt_Front_l_03, bone2Name=dyn_skirt_Front_l_02, swing1LimitDegrees=20, swing2LimitDegrees=3, twistLimitDegrees=1}
 {bone1Name=dyn_skirt_Front_l_04, bone2Name=dyn_skirt_Front_l_03, swing1LimitDegrees=20, swing2LimitDegrees=5, twistLimitDegrees=1}
 {bone1Name=dyn_skirt_Front_l_05, bone2Name=dyn_skirt_Front_l_04, swing1LimitDegrees=12, swing2LimitDegrees=5, twistLimitDegrees=1}
 {bone1Name=dyn_skirt_Front_r_01, bone2Name=base_skirt_Front_r_01, swing1LimitDegrees=10, swing2LimitDegrees=2, twistLimitDegrees=1}
 {bone1Name=dyn_skirt_Front_r_02, bone2Name=dyn_skirt_Front_r_01, swing1LimitDegrees=14, swing2LimitDegrees=2, twistLimitDegrees=1}
 {bone1Name=dyn_skirt_Front_r_03, bone2Name=dyn_skirt_Front_r_02, swing1LimitDegrees=20, swing2LimitDegrees=3, twistLimitDegrees=1}
 {bone1Name=dyn_skirt_Front_r_04, bone2Name=dyn_skirt_Front_r_03, swing1LimitDegrees=20, swing2LimitDegrees=5, twistLimitDegrees=1}
 {bone1Name=dyn_skirt_Front_r_05, bone2Name=dyn_skirt_Front_r_04, swing1LimitDegrees=20, swing2LimitDegrees=5, twistLimitDegrees=1}
 {bone1Name=dyn_skirt_SideBack_l_01, bone2Name=base_skirt_SideBack_l_01, swing1LimitDegrees=5, swing2LimitDegrees=2, twistLimitDegrees=1}
 {bone1Name=dyn_skirt_SideBack_l_02, bone2Name=dyn_skirt_SideBack_l_01, swing1LimitDegrees=7, swing2LimitDegrees=2, twistLimitDegrees=1}
 {bone1Name=dyn_skirt_SideBack_l_03, bone2Name=dyn_skirt_SideBack_l_02, swing1LimitDegrees=10, swing2LimitDegrees=3, twistLimitDegrees=1}
 {bone1Name=dyn_skirt_SideBack_l_04, bone2Name=dyn_skirt_SideBack_l_03, swing1LimitDegrees=10, swing2LimitDegrees=5, twistLimitDegrees=1}
 {bone1Name=dyn_skirt_SideBack_l_05, bone2Name=dyn_skirt_SideBack_l_04, swing1LimitDegrees=12, swing2LimitDegrees=5, twistLimitDegrees=1}
 {bone1Name=dyn_skirt_SideBack_r_01, bone2Name=base_skirt_SideBack_r_01, swing1LimitDegrees=5, swing2LimitDegrees=2, twistLimitDegrees=1}
 {bone1Name=dyn_skirt_SideBack_r_02, bone2Name=dyn_skirt_SideBack_r_01, swing1LimitDegrees=7, swing2LimitDegrees=2, twistLimitDegrees=1}
 {bone1Name=dyn_skirt_SideBack_r_03, bone2Name=dyn_skirt_SideBack_r_02, swing1LimitDegrees=10, swing2LimitDegrees=3, twistLimitDegrees=1}
 {bone1Name=dyn_skirt_SideBack_r_04, bone2Name=dyn_skirt_SideBack_r_03, swing1LimitDegrees=10, swing2LimitDegrees=5, twistLimitDegrees=1}
 {bone1Name=dyn_skirt_SideBack_r_05, bone2Name=dyn_skirt_SideBack_r_04, swing1LimitDegrees=12, swing2LimitDegrees=5, twistLimitDegrees=1}
 {bone1Name=dyn_skirt_SideFront_l_01, bone2Name=base_skirt_SideFront_l_01, swing1LimitDegrees=10, swing2LimitDegrees=2, twistLimitDegrees=1}
 {bone1Name=dyn_skirt_SideFront_l_02, bone2Name=dyn_skirt_SideFront_l_01, swing1LimitDegrees=14, swing2LimitDegrees=2, twistLimitDegrees=1}
 {bone1Name=dyn_skirt_SideFront_l_03, bone2Name=dyn_skirt_SideFront_l_02, swing1LimitDegrees=20, swing2LimitDegrees=3, twistLimitDegrees=1}
 {bone1Name=dyn_skirt_SideFront_l_04, bone2Name=dyn_skirt_SideFront_l_03, swing1LimitDegrees=20, swing2LimitDegrees=5, twistLimitDegrees=1}
 {bone1Name=dyn_skirt_SideFront_l_05, bone2Name=dyn_skirt_SideFront_l_04, swing1LimitDegrees=12, swing2LimitDegrees=5, twistLimitDegrees=1}
 {bone1Name=dyn_skirt_SideFront_r_01, bone2Name=base_skirt_SideFront_r_01, swing1LimitDegrees=10, swing2LimitDegrees=2, twistLimitDegrees=1}
 {bone1Name=dyn_skirt_SideFront_r_02, bone2Name=dyn_skirt_SideFront_r_01, swing1LimitDegrees=14, swing2LimitDegrees=2, twistLimitDegrees=1}
 {bone1Name=dyn_skirt_SideFront_r_03, bone2Name=dyn_skirt_SideFront_r_02, swing1LimitDegrees=20, swing2LimitDegrees=3, twistLimitDegrees=1}
 {bone1Name=dyn_skirt_SideFront_r_04, bone2Name=dyn_skirt_SideFront_r_03, swing1LimitDegrees=20, swing2LimitDegrees=5, twistLimitDegrees=1}
 {bone1Name=dyn_skirt_SideFront_r_05, bone2Name=dyn_skirt_SideFront_r_04, swing1LimitDegrees=20, swing2LimitDegrees=5, twistLimitDegrees=1}
 {
  bone1Name=base_skirt_Back_l_01
  bone2Name=spine_02
  swing1LimitDegrees=45
  swing1Motion=Free
  swing2LimitDegrees=45
  swing2Motion=Free
  twistLimitDegrees=45
  twistMotion=Free
 }
 {
  bone1Name=base_skirt_SideBack_l_01
  bone2Name=spine_02
  swing1LimitDegrees=45
  swing1Motion=Free
  swing2LimitDegrees=45
  swing2Motion=Free
  twistLimitDegrees=45
  twistMotion=Free
 }
 {
  bone1Name=base_skirt_SideBack_r_01
  bone2Name=spine_02
  swing1LimitDegrees=45
  swing1Motion=Free
  swing2LimitDegrees=45
  swing2Motion=Free
  twistLimitDegrees=45
  twistMotion=Free
 }
 {
  bone1Name=base_skirt_Back_r_01
  bone2Name=spine_02
  swing1LimitDegrees=45
  swing1Motion=Free
  swing2LimitDegrees=45
  swing2Motion=Free
  twistLimitDegrees=45
  twistMotion=Free
 }
 {
  bone1Name=base_skirt_SideFront_l_01
  bone2Name=spine_02
  swing1LimitDegrees=45
  swing1Motion=Free
  swing2LimitDegrees=45
  swing2Motion=Free
  twistLimitDegrees=45
  twistMotion=Free
 }
 {
  bone1Name=base_skirt_SideFront_r_01
  bone2Name=spine_02
  swing1LimitDegrees=45
  swing1Motion=Free
  swing2LimitDegrees=45
  swing2Motion=Free
  twistLimitDegrees=45
  twistMotion=Free
 }
 {
  bone1Name=base_skirt_Front_r_01
  bone2Name=spine_02
  swing1LimitDegrees=45
  swing1Motion=Free
  swing2LimitDegrees=45
  swing2Motion=Free
  twistLimitDegrees=45
  twistMotion=Free
 }
 {
  bone1Name=base_skirt_Front_l_01
  bone2Name=spine_02
  swing1LimitDegrees=45
  swing1Motion=Free
  swing2LimitDegrees=45
  swing2Motion=Free
  twistLimitDegrees=45
  twistMotion=Free
 }
 {
  bone1Name=spine_02
  bone2Name=spine_01
  swing1LimitDegrees=45
  swing1Motion=Free
  swing2LimitDegrees=45
  swing2Motion=Free
  twistLimitDegrees=45
  twistMotion=Free
 }
 {
  bone1Name=spine_01
  bone2Name=pelvis
  swing1LimitDegrees=45
  swing1Motion=Free
  swing2LimitDegrees=45
  swing2Motion=Free
  twistLimitDegrees=45
  twistMotion=Free
 }
 {
  bone1Name=thigh_l
  bone2Name=pelvis
  swing1LimitDegrees=45
  swing1Motion=Free
  swing2LimitDegrees=45
  swing2Motion=Free
  twistLimitDegrees=45
  twistMotion=Free
 }
 {
  bone1Name=calf_l
  bone2Name=thigh_l
  swing1LimitDegrees=45
  swing1Motion=Free
  swing2LimitDegrees=45
  swing2Motion=Free
  twistLimitDegrees=45
  twistMotion=Free
 }
 {
  bone1Name=foot_l
  bone2Name=calf_l
  swing1LimitDegrees=45
  swing1Motion=Free
  swing2LimitDegrees=45
  swing2Motion=Free
  twistLimitDegrees=45
  twistMotion=Free
 }
 {
  bone1Name=thigh_r
  bone2Name=pelvis
  swing1LimitDegrees=45
  swing1Motion=Free
  swing2LimitDegrees=45
  swing2Motion=Free
  twistLimitDegrees=45
  twistMotion=Free
 }
 {
  bone1Name=calf_r
  bone2Name=thigh_r
  swing1LimitDegrees=45
  swing1Motion=Free
  swing2LimitDegrees=45
  swing2Motion=Free
  twistLimitDegrees=45
  twistMotion=Free
 }
 {
  bone1Name=foot_r
  bone2Name=calf_r
  swing1LimitDegrees=45
  swing1Motion=Free
  swing2LimitDegrees=45
  swing2Motion=Free
  twistLimitDegrees=45
  twistMotion=Free
 }
 {
  bone1Name=dyn_skirt_Back_r_03
  bone2Name=dyn_skirt_Back_l_03
  swing1LimitDegrees=45
  swing1Motion=Free
  swing2LimitDegrees=45
  swing2Motion=Free
  twistLimitDegrees=45
  twistMotion=Free
 }
 {
  bone1Name=dyn_skirt_Back_r_04
  bone2Name=dyn_skirt_Back_l_04
  swing1LimitDegrees=45
  swing1Motion=Free
  swing2LimitDegrees=45
  swing2Motion=Free
  twistLimitDegrees=45
  twistMotion=Free
 }
 {
  bone1Name=dyn_skirt_SideFront_l_04
  bone2Name=dyn_skirt_Front_l_04
  swing1LimitDegrees=45
  swing1Motion=Free
  swing2LimitDegrees=45
  swing2Motion=Free
  twistLimitDegrees=45
  twistMotion=Free
 }
 {
  bone1Name=dyn_skirt_SideFront_r_04
  bone2Name=dyn_skirt_Front_r_04
  swing1LimitDegrees=45
  swing1Motion=Free
  swing2LimitDegrees=45
  swing2Motion=Free
  twistLimitDegrees=45
  twistMotion=Free
 }
]
