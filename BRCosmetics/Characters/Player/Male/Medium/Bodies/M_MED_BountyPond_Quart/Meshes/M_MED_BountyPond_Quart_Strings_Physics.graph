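# Physics asset for the Strings mesh of the male medium BountyPond Quart player cosmetic body, covering its belt strings and skirt.
# Two four-link belt-string chains with limited swing and tight twist, plus five five-link skirt chains with free rotation, hang from base bodies constrained to spine_02.
# Capsules on pelvis, spine, thighs and calves form the body, while an unconstrained, oversized ground_mover capsule approximates the floor.
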PhysicsAsset /BRCosmetics/Characters/Player/Male/Medium/Bodies/M_MED_BountyPond_Quart/Meshes/M_MED_BountyPond_Quart_Strings_Physics
bodies {
 base_beltString_1_01 [{shapeType=Capsule, length=0.505, radius=0.505, shapeName=base_beltString_1_01_capsule, rotation=<rot 0 -90 0>}]
 base_beltString_2_01 [{shapeType=Capsule, length=0.505, radius=0.505, shapeName=base_beltString_2_01_capsule, rotation=<rot 0 -90 0>}]
 base_skirt_Front_l_01 [{shapeType=Capsule, length=1, radius=2, shapeName=base_skirt_Front_l_01_capsule}]
 base_skirt_Front_r_01 [{shapeType=Capsule, length=1, radius=2, shapeName=base_skirt_Front_r_01_capsule}]
 base_skirt_SideBack_l_01 [{shapeType=Capsule, length=1, radius=2, shapeName=base_skirt_SideBack_l_01_capsule}]
 base_skirt_SideFront_l_01 [{shapeType=Capsule, length=1, radius=2, shapeName=base_skirt_SideFront_l_01_capsule}]
 base_skirt_SideFront_r_01 [{shapeType=Capsule, length=1, radius=2, shapeName=base_skirt_SideFront_r_01_capsule}]
 calf_l [{
   shapeType=Capsule
   length=38.3092
   radius=7.1459
   shapeName=calf_l_capsule
   center=<-24.0003 -1.4476 1.1357>
   rotation=<rot 88.914 179.6173 -177.84>
  }]
 calf_r [{
   shapeType=Capsule
   length=38.3092
   radius=7.1459
   shapeName=calf_l_capsule
   center=<23.2954 1.638 -1.0343>
   rotation=<rot 88.914 179.6173 -177.84>
  }]
 dyn_beltString_1_01 [{
   shapeType=Capsule
   length=14.8691
   radius=2.2745
   shapeName=dyn_beltString_1_01_capsule
   center=<8.994 1.5582 -1.8567>
   rotation=<rot 90 -3.4363 0>
  }]
 dyn_beltString_1_02 [{
   shapeType=Capsule
   length=14.8443
   radius=2.2993
   shapeName=dyn_beltString_1_02_capsule
   center=<8.2637 0.6863 -1.941>
   rotation=<rot 90 -0.7853 0>
  }]
 dyn_beltString_1_03 [{
   shapeType=Capsule
   length=14.7429
   radius=2.4007
   shapeName=dyn_beltString_1_03_capsule
   center=<8.2626 1.2449 -1.3968>
   rotation=<rot 90 0 0>
  }]
 dyn_beltString_1_04 [{
   shapeType=Capsule
   length=14.6938
   radius=2.4498
   shapeName=dyn_beltString_1_04_capsule
   center=<8.2626 1.6488 -0.919>
   rotation=<rot 90 0 0>
  }]
 dyn_beltString_2_01 [{
   shapeType=Capsule
   length=13.1051
   radius=2.1181
   shapeName=dyn_beltString_2_01_capsule
   center=<7.0024 1.7046 1.8481>
   rotation=<rot 90 -2.3698 0>
  }]
 dyn_beltString_2_02 [{
   shapeType=Capsule
   length=13.0804
   radius=2.1429
   shapeName=dyn_beltString_2_02_capsule
   center=<6.996 1.4461 1.0081>
   rotation=<rot 90 -0.4051 0>
  }]
 dyn_beltString_2_03 [{
   shapeType=Capsule
   length=12.9789
   radius=2.2443
   shapeName=dyn_beltString_2_03_capsule
   center=<6.9862 1.4063 1.1225>
   rotation=<rot 89.9 -179.4348 -179.9964>
  }]
 dyn_beltString_2_04 [{
   shapeType=Capsule
   length=12.9299
   radius=2.2934
   shapeName=dyn_beltString_2_04_capsule
   center=<6.9839 1.3736 0.8881>
   rotation=<rot 90 0.0624 0>
  }]
 dyn_skirt_Front_l_01 [{shapeType=Capsule, length=0.2907, radius=12.6322, shapeName=dyn_skirt_Front_l_01_capsule, center=<9.0256 10.365 3.2307>}]
 dyn_skirt_Front_l_02 [{shapeType=Capsule, length=1.3066, radius=12.6322, shapeName=dyn_skirt_Front_l_02_capsule, center=<9.0256 10.1723 -1.4174>}]
 dyn_skirt_Front_l_03 [{shapeType=Capsule, length=1.0758, radius=12.6322, shapeName=dyn_skirt_Front_l_03_capsule, center=<9.0256 10.1943 -0.5301>}]
 dyn_skirt_Front_l_04 [{shapeType=Capsule, length=2.6879, radius=12.6322, shapeName=dyn_skirt_Front_l_04_capsule, center=<9.0256 10.7787 -1.1268>}]
 dyn_skirt_Front_l_05 [{shapeType=Capsule, length=3.0482, radius=12.6322, shapeName=dyn_skirt_Front_l_05_capsule, center=<9.0256 10.365 0.8286>}]
 dyn_skirt_Front_r_01 [{shapeType=Capsule, length=1, radius=2, shapeName=dyn_skirt_Front_l_01_capsule, center=<2 0 0>, rotation=<rot 90 0 0>}]
 dyn_skirt_Front_r_02 [{shapeType=Capsule, length=1, radius=2, shapeName=dyn_skirt_Front_l_02_capsule, center=<2 0 0>, rotation=<rot 90 0 0>}]
 dyn_skirt_Front_r_03 [{shapeType=Capsule, length=1, radius=2, shapeName=dyn_skirt_Front_l_03_capsule, center=<2 0 0>, rotation=<rot 90 0 0>}]
 dyn_skirt_Front_r_04 [{shapeType=Capsule, length=1, radius=2, shapeName=dyn_skirt_Front_l_04_capsule, center=<2 0 0>, rotation=<rot 90 0 0>}]
 dyn_skirt_Front_r_05 [{shapeType=Capsule, length=1, radius=2, shapeName=dyn_skirt_Front_l_05_capsule, center=<2 0 0>, rotation=<rot 90 0 0>}]
 dyn_skirt_SideBack_l_01 [{
   shapeType=Capsule
   length=7.4138
   radius=12.1856
   shapeName=dyn_skirt_SideBack_l_01_capsule
   center=<6.8142 9.245 -3.6359>
   rotation=<rot 0 0 28.0851>
  }]
 dyn_skirt_SideBack_l_02 [{
   shapeType=Capsule
   length=7.4138
   radius=12.1856
   shapeName=dyn_skirt_SideBack_l_02_capsule
   center=<6.8142 6.1504 -4.6538>
   rotation=<rot 0.0001 0 46.141>
  }]
 dyn_skirt_SideBack_l_03 [{
   shapeType=Capsule
   length=7.4138
   radius=12.1856
   shapeName=dyn_skirt_SideBack_l_03_capsule
   center=<6.8142 8.7096 -4.1669>
   rotation=<rot 0 0 28.0851>
  }]
 dyn_skirt_SideBack_l_04 [{
   shapeType=Capsule
   length=7.4138
   radius=12.1856
   shapeName=dyn_skirt_SideBack_l_04_capsule
   center=<6.8142 9.2976 -3.6359>
   rotation=<rot 0 0 28.0851>
  }]
 dyn_skirt_SideBack_l_05 [{
   shapeType=Capsule
   length=7.4138
   radius=12.1856
   shapeName=dyn_skirt_SideBack_l_05_capsule
   center=<6.8142 9.245 -3.6359>
   rotation=<rot 0 0 28.0851>
  }]
 dyn_skirt_SideFront_l_01 [{shapeType=Capsule, length=0, radius=12.1856, shapeName=dyn_skirt_SideFront_l_01_capsule, center=<6.2693 8.5025 -2.2732>}]
 dyn_skirt_SideFront_l_02 [{shapeType=Capsule, length=0.7706, radius=12.1856, shapeName=dyn_skirt_SideFront_l_02_capsule, center=<6.2693 9.6494 -2.2996>}]
 dyn_skirt_SideFront_l_03 [{shapeType=Capsule, length=1.21, radius=12.1856, shapeName=dyn_skirt_SideFront_l_03_capsule, center=<6.2693 10.2764 -1.8423>}]
 dyn_skirt_SideFront_l_04 [{
   shapeType=Capsule
   length=2.24
   radius=12.1856
   shapeName=dyn_skirt_SideFront_l_04_capsule
   center=<6.2693 10.4416 -2.2646>
   rotation=<rot 0.0001 0 -5.1827>
  }]
 dyn_skirt_SideFront_l_05 [{shapeType=Capsule, length=1.7787, radius=12.1856, shapeName=dyn_skirt_SideFront_l_05_capsule, center=<6.2693 10.1787 -2.5291>}]
 dyn_skirt_SideFront_r_01 [{shapeType=Capsule, length=18.674, radius=12.1856, shapeName=dyn_skirt_SideFront_r_01_capsule, center=<0 11.042 0>}]
 dyn_skirt_SideFront_r_02 [{shapeType=Capsule, length=18.674, radius=12.1856, shapeName=dyn_skirt_SideFront_r_02_capsule, center=<0 11.042 0>}]
 dyn_skirt_SideFront_r_03 [{shapeType=Capsule, length=18.674, radius=12.1856, shapeName=dyn_skirt_SideFront_r_03_capsule, center=<0 11.042 0>}]
 dyn_skirt_SideFront_r_04 [{shapeType=Capsule, length=18.674, radius=12.1856, shapeName=dyn_skirt_SideFront_r_04_capsule, center=<0 11.042 0>}]
 dyn_skirt_SideFront_r_05 [{shapeType=Capsule, length=18.674, radius=12.1856, shapeName=dyn_skirt_SideFront_r_05_capsule, center=<0 11.042 0>}]
 ground_mover [{shapeType=Capsule, length=0.5056, radius=337.6725, shapeName=floor_joint_capsule, center=<2.6954 -3.0549 -346.3523>, rotation=<rot 0 90 -90>}]
 pelvis [{shapeType=Capsule, length=7.1872, radius=9.562, shapeName=pelvis_capsule, center=<0 -0.0227 0>}]
 spine_01 [{shapeType=Capsule, length=7.2791, radius=5.4353, shapeName=spine_01_capsule, center=<0 1.1594 0>}]
 spine_02 [{shapeType=Capsule, length=4.2576, radius=6.0103, shapeName=spine_02_capsule, center=<0 1.6446 0>}]
 thigh_l [{shapeType=Capsule, length=35, radius=6, shapeName=thigh_l_capsule, center=<-21.2061 -0.0547 2.4106>, rotation=<rot 89.4575 0.1673 -1.0885>}]
 thigh_r [{
   shapeType=Capsule
   length=35.4834
   radius=6.4923
   shapeName=thigh_l_capsule
   center=<21.6713 0.3711 -2.5283>
   rotation=<rot 89.4007 -179.6371 179.2638>
  }]
}
constraints [
 {
  bone1Name=spine_02
  bone2Name=spine_01
  swing1LimitDegrees=45
  swing1Motion=Free
  swing2LimitDegrees=45
  swing2Motion=Free
  twistLimitDegrees=45
  twistMotion=Free
 }
 {
  bone1Name=spine_01
  bone2Name=pelvis
  swing1LimitDegrees=45
  swing1Motion=Free
  swing2LimitDegrees=45
  swing2Motion=Free
  twistLimitDegrees=45
  twistMotion=Free
 }
 {
  bone1Name=thigh_l
  bone2Name=pelvis
  swing1LimitDegrees=45
  swing1Motion=Free
  swing2LimitDegrees=45
  swing2Motion=Free
  twistLimitDegrees=45
  twistMotion=Free
 }
 {
  bone1Name=calf_l
  bone2Name=thigh_l
  swing1LimitDegrees=45
  swing1Motion=Free
  swing2LimitDegrees=45
  swing2Motion=Free
  twistLimitDegrees=45
  twistMotion=Free
 }
 {
  bone1Name=thigh_r
  bone2Name=pelvis
  swing1LimitDegrees=45
  swing1Motion=Free
  swing2LimitDegrees=45
  swing2Motion=Free
  twistLimitDegrees=45
  twistMotion=Free
 }
 {
  bone1Name=calf_r
  bone2Name=thigh_r
  swing1LimitDegrees=45
  swing1Motion=Free
  swing2LimitDegrees=45
  swing2Motion=Free
  twistLimitDegrees=45
  twistMotion=Free
 }
 {
  bone1Name=base_beltString_1_01
  bone2Name=spine_02
  swing1LimitDegrees=45
  swing1Motion=Free
  swing2LimitDegrees=45
  swing2Motion=Free
  twistLimitDegrees=45
  twistMotion=Free
 }
 {bone1Name=dyn_beltString_1_01, bone2Name=base_beltString_1_01, swing1LimitDegrees=45, swing2LimitDegrees=20, twistLimitDegrees=1}
 {bone1Name=dyn_beltString_1_02, bone2Name=dyn_beltString_1_01, swing1LimitDegrees=45, swing2LimitDegrees=20, twistLimitDegrees=1}
 {bone1Name=dyn_beltString_1_03, bone2Name=dyn_beltString_1_02, swing1LimitDegrees=45, swing2LimitDegrees=20, twistLimitDegrees=1}
 {bone1Name=dyn_beltString_1_04, bone2Name=dyn_beltString_1_03, swing1LimitDegrees=45, swing2LimitDegrees=20, twistLimitDegrees=1}
 {
  bone1Name=base_beltString_2_01
  bone2Name=spine_02
  swing1LimitDegrees=45
  swing1Motion=Free
  swing2LimitDegrees=45
  swing2Motion=Free
  twistLimitDegrees=45
  twistMotion=Free
 }
 {bone1Name=dyn_beltString_2_01, bone2Name=base_beltString_2_01, swing1LimitDegrees=45, swing2LimitDegrees=20, twistLimitDegrees=1}
 {bone1Name=dyn_beltString_2_02, bone2Name=dyn_beltString_2_01, swing1LimitDegrees=45, swing2LimitDegrees=20, twistLimitDegrees=1}
 {bone1Name=dyn_beltString_2_03, bone2Name=dyn_beltString_2_02, swing1LimitDegrees=45, swing2LimitDegrees=20, twistLimitDegrees=1}
 {bone1Name=dyn_beltString_2_04, bone2Name=dyn_beltString_2_03, swing1LimitDegrees=45, swing2LimitDegrees=20, twistLimitDegrees=1}
 {
  bone1Name=base_skirt_Front_l_01
  bone2Name=spine_02
  swing1LimitDegrees=45
  swing1Motion=Free
  swing2LimitDegrees=45
  swing2Motion=Free
  twistLimitDegrees=45
  twistMotion=Free
 }
 {
  bone1Name=dyn_skirt_Front_l_01
  bone2Name=base_skirt_Front_l_01
  swing1LimitDegrees=45
  swing1Motion=Free
  swing2LimitDegrees=45
  swing2Motion=Free
  twistLimitDegrees=45
  twistMotion=Free
 }
 {
  bone1Name=dyn_skirt_Front_l_02
  bone2Name=dyn_skirt_Front_l_01
  swing1LimitDegrees=45
  swing1Motion=Free
  swing2LimitDegrees=45
  swing2Motion=Free
  twistLimitDegrees=45
  twistMotion=Free
 }
 {
  bone1Name=dyn_skirt_Front_l_03
  bone2Name=dyn_skirt_Front_l_02
  swing1LimitDegrees=45
  swing1Motion=Free
  swing2LimitDegrees=45
  swing2Motion=Free
  twistLimitDegrees=45
  twistMotion=Free
 }
 {
  bone1Name=dyn_skirt_Front_l_04
  bone2Name=dyn_skirt_Front_l_03
  swing1LimitDegrees=45
  swing1Motion=Free
  swing2LimitDegrees=45
  swing2Motion=Free
  twistLimitDegrees=45
  twistMotion=Free
 }
 {
  bone1Name=dyn_skirt_Front_l_05
  bone2Name=dyn_skirt_Front_l_04
  swing1LimitDegrees=45
  swing1Motion=Free
  swing2LimitDegrees=45
  swing2Motion=Free
  twistLimitDegrees=45
  twistMotion=Free
 }
 {
  bone1Name=base_skirt_Front_r_01
  bone2Name=spine_02
  swing1LimitDegrees=45
  swing1Motion=Free
  swing2LimitDegrees=45
  swing2Motion=Free
  twistLimitDegrees=45
  twistMotion=Free
 }
 {
  bone1Name=dyn_skirt_Front_r_01
  bone2Name=base_skirt_Front_r_01
  swing1LimitDegrees=45
  swing1Motion=Free
  swing2LimitDegrees=45
  swing2Motion=Free
  twistLimitDegrees=45
  twistMotion=Free
 }
 {
  bone1Name=dyn_skirt_Front_r_02
  bone2Name=dyn_skirt_Front_r_01
  swing1LimitDegrees=45
  swing1Motion=Free
  swing2LimitDegrees=45
  swing2Motion=Free
  twistLimitDegrees=45
  twistMotion=Free
 }
 {
  bone1Name=dyn_skirt_Front_r_03
  bone2Name=dyn_skirt_Front_r_02
  swing1LimitDegrees=45
  swing1Motion=Free
  swing2LimitDegrees=45
  swing2Motion=Free
  twistLimitDegrees=45
  twistMotion=Free
 }
 {
  bone1Name=dyn_skirt_Front_r_04
  bone2Name=dyn_skirt_Front_r_03
  swing1LimitDegrees=45
  swing1Motion=Free
  swing2LimitDegrees=45
  swing2Motion=Free
  twistLimitDegrees=45
  twistMotion=Free
 }
 {
  bone1Name=dyn_skirt_Front_r_05
  bone2Name=dyn_skirt_Front_r_04
  swing1LimitDegrees=45
  swing1Motion=Free
  swing2LimitDegrees=45
  swing2Motion=Free
  twistLimitDegrees=45
  twistMotion=Free
 }
 {
  bone1Name=base_skirt_SideBack_l_01
  bone2Name=spine_02
  swing1LimitDegrees=45
  swing1Motion=Free
  swing2LimitDegrees=45
  swing2Motion=Free
  twistLimitDegrees=45
  twistMotion=Free
 }
 {
  bone1Name=dyn_skirt_SideBack_l_01
  bone2Name=base_skirt_SideBack_l_01
  swing1LimitDegrees=45
  swing1Motion=Free
  swing2LimitDegrees=45
  swing2Motion=Free
  twistLimitDegrees=45
  twistMotion=Free
 }
 {
  bone1Name=dyn_skirt_SideBack_l_02
  bone2Name=dyn_skirt_SideBack_l_01
  swing1LimitDegrees=45
  swing1Motion=Free
  swing2LimitDegrees=45
  swing2Motion=Free
  twistLimitDegrees=45
  twistMotion=Free
 }
 {
  bone1Name=dyn_skirt_SideBack_l_03
  bone2Name=dyn_skirt_SideBack_l_02
  swing1LimitDegrees=45
  swing1Motion=Free
  swing2LimitDegrees=45
  swing2Motion=Free
  twistLimitDegrees=45
  twistMotion=Free
 }
 {
  bone1Name=dyn_skirt_SideBack_l_04
  bone2Name=dyn_skirt_SideBack_l_03
  swing1LimitDegrees=45
  swing1Motion=Free
  swing2LimitDegrees=45
  swing2Motion=Free
  twistLimitDegrees=45
  twistMotion=Free
 }
 {
  bone1Name=dyn_skirt_SideBack_l_05
  bone2Name=dyn_skirt_SideBack_l_04
  swing1LimitDegrees=45
  swing1Motion=Free
  swing2LimitDegrees=45
  swing2Motion=Free
  twistLimitDegrees=45
  twistMotion=Free
 }
 {
  bone1Name=base_skirt_SideFront_l_01
  bone2Name=spine_02
  swing1LimitDegrees=45
  swing1Motion=Free
  swing2LimitDegrees=45
  swing2Motion=Free
  twistLimitDegrees=45
  twistMotion=Free
 }
 {
  bone1Name=dyn_skirt_SideFront_l_01
  bone2Name=base_skirt_SideFront_l_01
  swing1LimitDegrees=45
  swing1Motion=Free
  swing2LimitDegrees=45
  swing2Motion=Free
  twistLimitDegrees=45
  twistMotion=Free
 }
 {
  bone1Name=dyn_skirt_SideFront_l_02
  bone2Name=dyn_skirt_SideFront_l_01
  swing1LimitDegrees=45
  swing1Motion=Free
  swing2LimitDegrees=45
  swing2Motion=Free
  twistLimitDegrees=45
  twistMotion=Free
 }
 {
  bone1Name=dyn_skirt_SideFront_l_03
  bone2Name=dyn_skirt_SideFront_l_02
  swing1LimitDegrees=45
  swing1Motion=Free
  swing2LimitDegrees=45
  swing2Motion=Free
  twistLimitDegrees=45
  twistMotion=Free
 }
 {
  bone1Name=dyn_skirt_SideFront_l_04
  bone2Name=dyn_skirt_SideFront_l_03
  swing1LimitDegrees=45
  swing1Motion=Free
  swing2LimitDegrees=45
  swing2Motion=Free
  twistLimitDegrees=45
  twistMotion=Free
 }
 {
  bone1Name=dyn_skirt_SideFront_l_05
  bone2Name=dyn_skirt_SideFront_l_04
  swing1LimitDegrees=45
  swing1Motion=Free
  swing2LimitDegrees=45
  swing2Motion=Free
  twistLimitDegrees=45
  twistMotion=Free
 }
 {
  bone1Name=base_skirt_SideFront_r_01
  bone2Name=spine_02
  swing1LimitDegrees=45
  swing1Motion=Free
  swing2LimitDegrees=45
  swing2Motion=Free
  twistLimitDegrees=45
  twistMotion=Free
 }
 {
  bone1Name=dyn_skirt_SideFront_r_01
  bone2Name=base_skirt_SideFront_r_01
  swing1LimitDegrees=45
  swing1Motion=Free
  swing2LimitDegrees=45
  swing2Motion=Free
  twistLimitDegrees=45
  twistMotion=Free
 }
 {
  bone1Name=dyn_skirt_SideFront_r_02
  bone2Name=dyn_skirt_SideFront_r_01
  swing1LimitDegrees=45
  swing1Motion=Free
  swing2LimitDegrees=45
  swing2Motion=Free
  twistLimitDegrees=45
  twistMotion=Free
 }
 {
  bone1Name=dyn_skirt_SideFront_r_03
  bone2Name=dyn_skirt_SideFront_r_02
  swing1LimitDegrees=45
  swing1Motion=Free
  swing2LimitDegrees=45
  swing2Motion=Free
  twistLimitDegrees=45
  twistMotion=Free
 }
 {
  bone1Name=dyn_skirt_SideFront_r_04
  bone2Name=dyn_skirt_SideFront_r_03
  swing1LimitDegrees=45
  swing1Motion=Free
  swing2LimitDegrees=45
  swing2Motion=Free
  twistLimitDegrees=45
  twistMotion=Free
 }
 {
  bone1Name=dyn_skirt_SideFront_r_05
  bone2Name=dyn_skirt_SideFront_r_04
  swing1LimitDegrees=45
  swing1Motion=Free
  swing2LimitDegrees=45
  swing2Motion=Free
  twistLimitDegrees=45
  twistMotion=Free
 }
]
