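# Animation Blueprint for the BountyPond Quart male body cosmetic, adding secondary motion to a pose copied from the attached parent mesh.
# A Control Rig fed by movement-state variables precedes additive bone offsets, sleeve AnimDynamics and two RigidBody simulations, one for strings.
# The event graph updates movement variables and flips RigidBody gravity upward during skydive floating or diving.
# It also lowers main RigidBody alpha when standing idle, except non-emote front end.
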AnimBlueprint /BRCosmetics/Characters/Player/Male/Medium/Bodies/M_MED_BountyPond_Quart/Meshes/M_MED_BountyPond_Quart_AnimBP
target_skeleton=&/M_MED_BountyPond_Quart_Skeleton
graphs [
 AnimGraph {nodes [
   n1 = Control Rig!(
    Source=Copy Pose From Mesh() {bUseAttachedParent=true}
    Alpha
    "is Moving 2D ":double=Get is Moving 2D CR() -> (is Moving 2D CR:double)
    is Knee Sliding:double=Get is Knee Sliding CR() -> (is Knee Sliding CR:double)
    is Baseball Sliding:double=Get is Baseball Sliding CR() -> (is Baseball Sliding CR:double)
    "is Crouching ":double=Get is Crouching CR() -> (is Crouching CR:double)
    is Sky Dive:double=Get is Sky Dive CR() -> (is Sky Dive CR:double)
    is Standing Moving:double=Get is Standing Moving CR() -> (is Standing Moving CR:double)
    Forward Start:double=Get Forward() -> (Forward:double)
    is Sprinting:bool=Get is Idle/Start Walk() -> (is Idle/Start Walk:bool)
   ) -> (Pose) {controlRigAssetReference {blueprintRigClass=&/M_MED_BountyPond_Quart_CtrlRig.M_MED_BountyPond_Quart_CtrlRig_C}}
   n2 = "Transform (Modify) Bone - Bone: ground_mover"(ComponentPose=Local To Component(LocalPose=@n1), Alpha=1, Translation="0,0,-15", Rotation="0,0,0", Scale="1,1,1") {
    rotationMode=BMM_Additive
    rotationSpace=BCS_BoneSpace
    translationMode=BMM_Additive
    translationSpace=BCS_BoneSpace
    alphaScaleBiasClamp {bInterpResult=true}
    boneToModify {boneName=ground_mover}
    translation {z=-15}
   }
   n3 = "Transform (Modify) Bone - Bone: ground_mover"(ComponentPose=@n2, Alpha=1, Translation="0.000,0.000,9.319", Rotation="0,0,0", Scale="1,1,1") {
    rotationMode=BMM_Additive
    rotationSpace=BCS_BoneSpace
    translationMode=BMM_Additive
    translationSpace=BCS_BoneSpace
    boneToModify {boneName=ground_mover}
    translation {z=9.319}
   }
   n4 = Anim Dynamics - Bone: dyn_sleeve_01_r(ComponentPose=@n3, Alpha=1, ExternalForce="0,0,0") {
    angularDampingOverride=1
    angularSpringConstant=20000
    bAngularSpring=true
    bOverrideAngularDamping=true
    bUseGravityOverride=true
    gravityScale=0.35
    simulationSpace=RootRelative
    boundBone {boneName=dyn_sleeve_01_r}
    componentLinearAccScale=<1 1 1>
    componentLinearVelScale=<1 1 1>
    gravityOverride {z=-980}
    physicsBodyDefinitions [{
      boundBone {boneName=dyn_sleeve_01_r}
      boxExtents=<30 30 30>
      constraintSetup {angularLimitsMax=<0 5 5>, angularLimitsMin=<0 -5 -15>, angularTarget=<1 0 0>}
      localJointOffset=<3 0 0>
     }]
    relativeSpaceBone {boneName=lowerarm_r}
   }
   n5 = Anim Dynamics - Bone: dyn_sleeve_02_r(ComponentPose=@n4, Alpha=1, ExternalForce="0,0,0") {
    angularDampingOverride=1
    angularSpringConstant=20000
    bAngularSpring=true
    bOverrideAngularDamping=true
    bUseGravityOverride=true
    gravityScale=0.35
    simulationSpace=RootRelative
    boundBone {boneName=dyn_sleeve_02_r}
    componentLinearAccScale=<1 1 1>
    componentLinearVelScale=<1 1 1>
    gravityOverride {z=-980}
    physicsBodyDefinitions [{
      boundBone {boneName=dyn_sleeve_02_r}
      boxExtents=<30 30 30>
      constraintSetup {angularLimitsMax=<0 5 10>, angularLimitsMin=<0 -5 -15>, angularTarget=<1 0 0>}
      localJointOffset=<4 0 0>
     }]
    relativeSpaceBone {boneName=lowerarm_twist_02_r}
   }
   n6 = Anim Dynamics - Bone: dyn_sleeve_03_r(ComponentPose=@n5, Alpha=1, ExternalForce="0,0,0") {
    angularDampingOverride=1
    angularSpringConstant=20000
    bAngularSpring=true
    bOverrideAngularDamping=true
    bUseGravityOverride=true
    gravityScale=0.35
    simulationSpace=RootRelative
    boundBone {boneName=dyn_sleeve_03_r}
    componentLinearAccScale=<1 1 1>
    componentLinearVelScale=<1 1 1>
    gravityOverride {z=-980}
    physicsBodyDefinitions [{
      boundBone {boneName=dyn_sleeve_03_r}
      boxExtents=<30 30 30>
      constraintSetup {angularLimitsMax=<0 5 5>, angularLimitsMin=<0 -5 -15>, angularTarget=<1 0 0>}
      localJointOffset=<4 0 0>
     }]
    relativeSpaceBone {boneName=lowerarm_twist_02_r}
   }
   n7 = Anim Dynamics - Bone: dyn_sleeve_01_l(ComponentPose=@n6, Alpha=1, ExternalForce="0,0,0") {
    angularDampingOverride=1
    angularSpringConstant=20000
    bAngularSpring=true
    bOverrideAngularDamping=true
    bUseGravityOverride=true
    gravityScale=0.35
    simulationSpace=BoneRelative
    boundBone {boneName=dyn_sleeve_01_l}
    componentLinearAccScale=<1 1 1>
    componentLinearVelScale=<1 1 1>
    gravityOverride {z=-980}
    physicsBodyDefinitions [{
      boundBone {boneName=dyn_sleeve_01_l}
      boxExtents=<30 30 30>
      constraintSetup {angularLimitsMax=<0 5 15>, angularLimitsMin=<0 -5 -7>, angularTarget=<1 0 0>}
      localJointOffset=<4 0 0>
     }]
   }
   n8 = Anim Dynamics - Bone: dyn_sleeve_02_l(ComponentPose=@n7, Alpha=1, ExternalForce="0,0,0") {
    angularDampingOverride=1
    angularSpringConstant=20000
    bAngularSpring=true
    bOverrideAngularDamping=true
    bUseGravityOverride=true
    gravityScale=0.35
    simulationSpace=BoneRelative
    boundBone {boneName=dyn_sleeve_02_l}
    componentLinearAccScale=<1 1 1>
    componentLinearVelScale=<1 1 1>
    gravityOverride {z=-980}
    physicsBodyDefinitions [{
      boundBone {boneName=dyn_sleeve_02_l}
      boxExtents=<30 30 30>
      constraintSetup {angularLimitsMax=<0 5 15>, angularLimitsMin=<0 -5 -7>, angularTarget=<1 0 0>}
      localJointOffset=<3 0 0>
     }]
   }
   n9 = Anim Dynamics - Bone: dyn_sleeve_03_l(ComponentPose=@n8, Alpha=1, ExternalForce="0,0,0") {
    angularDampingOverride=1
    angularSpringConstant=20000
    bAngularSpring=true
    bOverrideAngularDamping=true
    bUseGravityOverride=true
    gravityScale=0.35
    simulationSpace=BoneRelative
    boundBone {boneName=dyn_sleeve_03_l}
    componentLinearAccScale=<1 1 1>
    componentLinearVelScale=<1 1 1>
    gravityOverride {z=-980}
    physicsBodyDefinitions [{
      boundBone {boneName=dyn_sleeve_03_l}
      boxExtents=<30 30 30>
      constraintSetup {angularLimitsMax=<0 5 15>, angularLimitsMin=<0 -5 -5>, angularTarget=<1 0 0>}
      localJointOffset=<3 0 0>
     }]
   }
   n10 = Get GravityOverride()
   n11 = RigidBody!(ComponentPose=@n9, Alpha=Get Alpha RBAN AB() -> (Alpha RBAN AB:double), OverrideWorldGravity:Vector=@n10, ExternalForce) -> (Pose) {
    bClampLinearTranslationLimitToRefPose=true
    bOverrideWorldGravity=true
    overridePhysicsAsset=&/M_MED_BountyPond_Quart_Physics
    simulationSpace=BaseBoneSpace
    alphaScaleBiasClamp {bInterpResult=true}
    baseBoneRef {boneName=spine_02}
    componentAppliedLinearAccClamp=<0 0 0>
    simSpaceSettings {
     dampingAlpha=0
     maxAngularAcceleration=16
     maxAngularVelocity=8
     maxLinearAcceleration=2800
     maxLinearVelocity=700
     velocityScaleZ=0.3
     worldAlpha=0.8
     externalLinearDragV=<0.2 0.2 0.2>
    }
   }
   n12 = "Transform (Modify) Bone - Bone: beltStrings_Sub"(ComponentPose=@n11, Alpha=1, Translation="0,0,0", Rotation="10,0,0", Scale="1,1,1") {
    rotationMode=BMM_Additive
    rotationSpace=BCS_BoneSpace
    translationMode=BMM_Additive
    translationSpace=BCS_BoneSpace
    alphaScaleBiasClamp {bInterpResult=true}
    boneToModify {boneName=beltStrings_Sub}
    rotation {pitch=10}
   }
   n13 = RigidBody(ComponentPose=@n12, OverrideWorldGravity=@n10) {
    bClampLinearTranslationLimitToRefPose=true
    bOverrideWorldGravity=true
    overridePhysicsAsset=&/M_MED_BountyPond_Quart_Strings_Physics
    simulationSpace=BaseBoneSpace
    baseBoneRef {boneName=spine_02}
    componentAppliedLinearAccClamp=<0 0 0>
    simSpaceSettings {
     dampingAlpha=0
     maxAngularAcceleration=10
     maxAngularVelocity=8
     maxLinearAcceleration=2800
     maxLinearVelocity=700
     velocityScaleZ=0.4
     worldAlpha=1
     externalLinearDragV=<0.2 0.2 0.2>
    }
   }
   n14 = "Transform (Modify) Bone - Bone: base_skirt_Back_r_01"(ComponentPose=@n13, Alpha=1, Translation="0,0,0", Rotation="0,-4,0", Scale="1,1,1") {
    rotationMode=BMM_Additive
    rotationSpace=BCS_BoneSpace
    translationMode=BMM_Additive
    translationSpace=BCS_BoneSpace
    alphaScaleBiasClamp {bInterpResult=true}
    boneToModify {boneName=base_skirt_Back_r_01}
    rotation {yaw=-4}
   }
   n15 = "Transform (Modify) Bone - Bone: base_skirt_Back_l_01"(ComponentPose=@n14, Alpha=1, Translation="0,0,0", Rotation="0,-4,0", Scale="1,1,1") {
    rotationMode=BMM_Additive
    rotationSpace=BCS_BoneSpace
    translationMode=BMM_Additive
    translationSpace=BCS_BoneSpace
    alphaScaleBiasClamp {bInterpResult=true}
    boneToModify {boneName=base_skirt_Back_l_01}
    rotation {yaw=-4}
   }
   Output Pose(Result=Component To Local(ComponentPose=@n15)) {layerGroup=None}
   Get bIsSprinting()
  ]}
 EventGraph {nodes [
   Event BlueprintUpdateAnimation()
   > n1 = Sequence!(execute) -> (then_0, then_1, then_2:Exec)
   n2 = NOT Boolean(A=Get bIsJumping())
   n3 = NOT Boolean(A=Get bIsFalling())
   n4 = Get bIsMoving2D()
   > Set is Moving 2D CR(is Moving 2D CR:double="To Float (Boolean)"(InBool=AND Boolean!(A=@n4, B=@n2, C:bool=@n3) -> (ReturnValue))) -> (Output_Get:double)
   > Set is Knee Sliding CR(is Knee Sliding CR:double="To Float (Boolean)"(InBool=Get bIsKneeSliding())) -> (Output_Get:double)
   > Set is Baseball Sliding CR(is Baseball Sliding CR:double="To Float (Boolean)"(InBool=Get bIsBaseballSliding())) -> (Output_Get:double)
   > Set is Crouching CR(is Crouching CR:double="To Float (Boolean)"(InBool=Get bIsCrouching())) -> (Output_Get:double)
   > Set is Sky Dive CR(is Sky Dive CR:double="To Float (Boolean)"(InBool=Get bIsSkyDiving())) -> (Output_Get:double)
   n5 = NOT Boolean(A=Get bIsCrouching())
   > Set is Standing Moving CR(is Standing Moving CR:double="To Float (Boolean)"(InBool=AND Boolean!(A=@n4, B=@n5) -> (ReturnValue))) -> (Output_Get:double)
   n6 = Get bIsParachuteOpened()
   > Set Parachuting Forward(Parachuting Forward:double=float * float(A="To Float (Boolean)"(InBool=@n6), B=Get Forward()) {is=title K2Node_PromotableOperator}) -> (Output_Get:double)
   > n7 = Branch(Condition=OR Boolean(A=Get bIsSkydiveFloating(), B=Get bIsSkydiveDiveMode()))
   > Set GravityOverride(GravityOverride="0,0,500")
   Set GravityOverride(execute=@n7.else, GravityOverride="0,0,-980")
   n8 = OR Boolean(A=Get bIsJumping(), B=Get bIsFalling())
   n9 = Get bIsTacticalSprinting()
   n10 = Branch(execute=@n1.then_1, Condition=OR Boolean(A=Get bIsCrouching(), B=AND Boolean(A=@n9, B=@n8)))
   > Set Standing Fixes Coat(Standing Fixes Coat:double=0) -> (Output_Get:double)
   Set Standing Fixes Coat(execute=@n10.else, Standing Fixes Coat=1)
   n11 = PlayerMovement(execute=@n1.then_2, AnimInstance=Self-Reference(), XDotProductCoefficient=100, YDotProductCoefficient=100, ZDotProductCoefficient=100)
   n12 = BreakVector(InVec=MakeVector(X=@n11, Y=@n11.YMovement))
   > Set Left(Left:double="To Float (Boolean)"(InBool="float < float"(A=@n12, B=-50))) -> (Output_Get:double)
   > Set Right(Right:double="To Float (Boolean)"(InBool="float > float"(A=@n12, B=50))) -> (Output_Get:double)
   n13 = "float > float"(A=@n12.Y, B=40)
   > Set Forward(Forward:double="To Float (Boolean)"(InBool=@n13)) -> (Output_Get:double)
   > Set Backward(Backward:double="To Float (Boolean)"(InBool="float < float"(A=@n12.Y, B=-40))) -> (Output_Get:double)
   > Set Moving Forward(Moving Forward:bool=@n13) -> (Output_Get:bool)
   > Set Alpha RBAN(Alpha RBAN:double=float - float(A=1, B=GetCurveValue(CurveName=Alpha RBAN))) -> (Output_Get:double)
   n14 = Get bIsJumping()
   n15 = Get bIsMoving2D()
   n16 = Get bIsCrouching()
   n17 = Get bIsFalling()
   > n18 = Branch(Condition=AND Boolean!(A=NOT Boolean(A=@n15), B=NOT Boolean(A=@n16), C:bool=NOT Boolean(A=@n14), D:bool=NOT Boolean(A=@n17)) -> (ReturnValue))
   > Set Alpha RBAN AB(Alpha RBAN AB:double=0.55) -> (Output_Get:double)
   n19 = Get bIsPlayingEmote()
   > n20 = Branch(Condition=AND Boolean!(A=Get bIsInFrontEnd(), B=NOT Boolean(A=@n19)) -> (ReturnValue))
   > Set Alpha RBAN AB(Alpha RBAN AB=1)
   Set Alpha RBAN AB(execute=@n20.else, Alpha RBAN AB=0.55)
   Set Alpha RBAN AB(execute=@n18.else, Alpha RBAN AB=1)
   "To Float (Boolean)"(InBool="float > float"(A=@n12.Y, B=40))
  ]}
]
variables [
 Alpha Foward {type=number}
 Alpha RBAN {type=number}
 Alpha RBAN AB {type=number}
 Backward {type=number}
 crouching RBAN disabled {type=number}
 Forward {type=number}
 Forward Start {type=number}
 GravityOverride {type=Vector}
 is Baseball Sliding CR {type=number}
 is Crouching CR {type=number}
 is Idle/Start Walk {type=boolean}
 is Knee Sliding CR {type=number}
 is Moving 2D CR {type=number}
 is Sky Dive CR {type=number}
 is Standing Moving CR {type=number}
 Left {type=number}
 Moving Forward {type=boolean}
 Parachuting Forward {type=number}
 Right {type=number}
 Standing Fixes Coat {type=number}
]
