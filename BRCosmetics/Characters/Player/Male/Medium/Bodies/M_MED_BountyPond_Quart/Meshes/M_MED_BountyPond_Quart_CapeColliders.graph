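# Physics asset of cape colliders for the male medium BountyPond Quart player body.
# Capsules cover pelvis, spine, neck and clavicles, with a head sphere; arm and leg bodies have no shapes.
# Capsules also follow left and right dyn_skirt chains, doubled on some lower back segments, while Front chains have only their first bone.
# Constraints link each chain bone to its predecessor and roots to spine_02, with swing and twist free.
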PhysicsAsset /BRCosmetics/Characters/Player/Male/Medium/Bodies/M_MED_BountyPond_Quart/Meshes/M_MED_BountyPond_Quart_CapeColliders
bodies {
 calf_l []
 calf_r []
 clavicle_l [{shapeType=Capsule, length=9.0192, radius=9.8871, center=<9.3226 0.2008 -2.4717>, rotation=<rot 69.3297 1.4476 176.9462>}]
 clavicle_r [{shapeType=Capsule, length=9.0192, radius=9.8871, center=<-9.3025 -0.4675 2.4667>, rotation=<rot 69.3297 1.4476 -3.0538>}]
 dyn_skirt_Back_l_01 [{shapeType=Capsule, length=6.0386, radius=3.5915, shapeName=dyn_skirt_Back_l_01_capsule}]
 dyn_skirt_Back_l_02 [{shapeType=Capsule, length=20.028, radius=9.9843, shapeName=dyn_skirt_Back_l_02_capsule, center=<8.1517 2.4554 0>, rotation=<rot 90 0 0>}]
 dyn_skirt_Back_l_03 [
  {shapeType=Capsule, length=17.8947, radius=10.0994, shapeName=dyn_skirt_Back_l_03_capsule, center=<6.7022 2.7133 0>, rotation=<rot 90 0 0>}
  {shapeType=Capsule, length=13.1028, radius=11.7099, shapeName=dyn_skirt_Back_l_05_capsule, center=<9.4449 8.0012 0>}
 ]
 dyn_skirt_Back_l_04 [
  {shapeType=Capsule, length=14.618, radius=14.7751, shapeName=dyn_skirt_Back_l_04_capsule, center=<5.3204 5.9521 0>, rotation=<rot 90 0 0>}
  {shapeType=Capsule, length=15, radius=13.1697, shapeName=dyn_skirt_Back_l_05_capsule, center=<9.4449 8.0012 0>}
 ]
 dyn_skirt_Back_l_05 [
  {shapeType=Capsule, length=16.4574, radius=16.2372, shapeName=dyn_skirt_Back_l_05_capsule, center=<4.7457 8.5676 0>, rotation=<rot 90 0 0>}
  {shapeType=Capsule, length=15, radius=12.0077, shapeName=dyn_skirt_Back_l_05_capsule, center=<9.4449 8.0012 0>}
 ]
 dyn_skirt_Back_r_01 [{shapeType=Capsule, length=6.0386, radius=3.5915, shapeName=dyn_skirt_Back_r_01_capsule}]
 dyn_skirt_Back_r_02 [{shapeType=Capsule, length=20.028, radius=9.9843, shapeName=dyn_skirt_Back_r_02_capsule, center=<8.1517 2.4554 0>, rotation=<rot 90 0 0>}]
 dyn_skirt_Back_r_03 [
  {shapeType=Capsule, length=17.8947, radius=10.0994, shapeName=dyn_skirt_Back_r_03_capsule, center=<6.7022 2.7133 0>, rotation=<rot 90 0 0>}
  {shapeType=Capsule, length=13.1028, radius=11.7099, shapeName=dyn_skirt_Back_l_05_capsule, center=<9.4449 8.0012 0>}
 ]
 dyn_skirt_Back_r_04 [
  {shapeType=Capsule, length=14.618, radius=14.7751, shapeName=dyn_skirt_Back_l_04_capsule, center=<5.3204 5.9521 0>, rotation=<rot 90 0 0>}
  {shapeType=Capsule, length=15, radius=13.1697, shapeName=dyn_skirt_Back_l_05_capsule, center=<9.4449 8.0012 0>}
 ]
 dyn_skirt_Back_r_05 [
  {shapeType=Capsule, length=16.4574, radius=16.2372, shapeName=dyn_skirt_Back_r_05_capsule, center=<4.7457 8.5676 0>, rotation=<rot 90 0 0>}
  {shapeType=Capsule, length=15, radius=12.0077, shapeName=dyn_skirt_Back_r_05_capsule, center=<9.4449 8.0012 0>}
 ]
 dyn_skirt_Front_l_01 [{shapeType=Capsule, length=6.0386, radius=3.5915, shapeName=dyn_skirt_Front_l_01_capsule}]
 dyn_skirt_Front_r_01 [{shapeType=Capsule, length=6.0386, radius=3.5915, shapeName=dyn_skirt_Front_r_01_capsule}]
 dyn_skirt_SideBack_l_01 [{shapeType=Capsule, length=6.0386, radius=3.5915, shapeName=dyn_skirt_SideBack_l_01_capsule}]
 dyn_skirt_SideBack_l_02 [{shapeType=Capsule, length=20.028, radius=9.9843, shapeName=dyn_skirt_SideBack_l_02_capsule, center=<8.1517 2.4554 0>, rotation=<rot 90 0 0>}]
 dyn_skirt_SideBack_l_03 [{
   shapeType=Capsule
   length=17.8947
   radius=10.0994
   shapeName=dyn_skirt_SideBack_l_03_capsule
   center=<6.7022 2.7133 0>
   rotation=<rot 90 0 0>
  }]
 dyn_skirt_SideBack_l_04 [
  {
   shapeType=Capsule
   length=16.5096
   radius=10.8672
   shapeName=dyn_skirt_SideBack_l_04_capsule
   center=<5.3204 3.2501 0>
   rotation=<rot 90 0 0>
  }
  {
   shapeType=Capsule
   length=12.7781
   radius=15.3916
   shapeName=dyn_skirt_Back_l_05_capsule
   center=<9.4449 8.0012 0>
   rotation=<rot 0.0002 0.0001 -6.0776>
  }
 ]
 dyn_skirt_SideBack_l_05 [
  {
   shapeType=Capsule
   length=14.6669
   radius=12.9335
   shapeName=dyn_skirt_SideBack_r_05_capsule
   center=<4.7457 5.6631 0>
   rotation=<rot 90 0 0>
  }
  {shapeType=Capsule, length=12.7781, radius=14.2296, shapeName=dyn_skirt_Back_r_05_capsule, center=<9.4449 8.0012 0>}
 ]
 dyn_skirt_SideBack_r_01 [{shapeType=Capsule, length=6.0386, radius=3.5915, shapeName=dyn_skirt_SideBack_r_01_capsule}]
 dyn_skirt_SideBack_r_02 [{shapeType=Capsule, length=20.028, radius=9.9843, shapeName=dyn_skirt_SideBack_r_02_capsule, center=<8.1517 2.4554 0>, rotation=<rot 90 0 0>}]
 dyn_skirt_SideBack_r_03 [{
   shapeType=Capsule
   length=17.8947
   radius=10.0994
   shapeName=dyn_skirt_SideBack_r_03_capsule
   center=<6.7022 2.7133 0>
   rotation=<rot 90 0 0>
  }]
 dyn_skirt_SideBack_r_04 [
  {
   shapeType=Capsule
   length=14.3555
   radius=13.0213
   shapeName=dyn_skirt_SideBack_r_04_capsule
   center=<5.3204 5.3987 0>
   rotation=<rot 90 0 0>
  }
  {
   shapeType=Capsule
   length=12.7781
   radius=15.3916
   shapeName=dyn_skirt_Back_l_05_capsule
   center=<9.4449 8.0012 0>
   rotation=<rot -0.0003 0.0001 6.1129>
  }
 ]
 dyn_skirt_SideBack_r_05 [
  {
   shapeType=Capsule
   length=14.6669
   radius=12.9335
   shapeName=dyn_skirt_SideBack_r_05_capsule
   center=<4.7457 5.6631 0>
   rotation=<rot 90 0 0>
  }
  {shapeType=Capsule, length=12.7781, radius=14.2296, shapeName=dyn_skirt_Back_r_05_capsule, center=<9.4449 8.0012 0>}
 ]
 dyn_skirt_SideFront_l_01 [{shapeType=Capsule, length=6.0386, radius=3.5915, shapeName=dyn_skirt_SideFront_l_01_capsule}]
 dyn_skirt_SideFront_l_02 [{shapeType=Capsule, length=20.028, radius=9.9843, shapeName=dyn_skirt_SideFront_l_02_capsule, center=<8.1517 2.4554 0>, rotation=<rot 90 0 0>}]
 dyn_skirt_SideFront_l_03 [{
   shapeType=Capsule
   length=17.8947
   radius=10.0994
   shapeName=dyn_skirt_SideFront_l_03_capsule
   center=<6.7022 2.7133 0>
   rotation=<rot 90 0 0>
  }]
 dyn_skirt_SideFront_l_04 [{
   shapeType=Capsule
   length=16.5096
   radius=10.8672
   shapeName=dyn_skirt_SideFront_l_04_capsule
   center=<5.3204 3.2501 0>
   rotation=<rot 90 0 0>
  }]
 dyn_skirt_SideFront_l_05 [{
   shapeType=Capsule
   length=14.6669
   radius=12.9335
   shapeName=dyn_skirt_SideFront_l_05_capsule
   center=<4.7457 5.6631 0>
   rotation=<rot 90 0 0>
  }]
 dyn_skirt_SideFront_r_01 [{shapeType=Capsule, length=6.0386, radius=3.5915, shapeName=dyn_skirt_SideFront_r_01_capsule}]
 dyn_skirt_SideFront_r_02 [{shapeType=Capsule, length=20.028, radius=9.9843, shapeName=dyn_skirt_SideFront_r_02_capsule, center=<8.1517 2.4554 0>, rotation=<rot 90 0 0>}]
 dyn_skirt_SideFront_r_03 [{
   shapeType=Capsule
   length=17.8947
   radius=10.0994
   shapeName=dyn_skirt_SideFront_r_03_capsule
   center=<6.7022 2.7133 0>
   rotation=<rot 90 0 0>
  }]
 dyn_skirt_SideFront_r_04 [{
   shapeType=Capsule
   length=16.5096
   radius=10.8672
   shapeName=dyn_skirt_SideFront_r_04_capsule
   center=<5.3204 3.2501 0>
   rotation=<rot 90 0 0>
  }]
 dyn_skirt_SideFront_r_05 [{
   shapeType=Capsule
   length=14.6669
   radius=12.9335
   shapeName=dyn_skirt_SideFront_r_05_capsule
   center=<4.7457 5.6631 0>
   rotation=<rot 90 0 0>
  }]
 foot_l []
 foot_r []
 head [{shapeType=Sphere, radius=17.3125, center=<11.028 -0.174 0>}]
 lowerarm_l []
 lowerarm_r []
 neck_02 [{shapeType=Capsule, length=13.4375, radius=7.3175, center=<-0.2642 -0.3408 0>, rotation=<rot 90 -171.1193 -175.3978>}]
 pelvis [{shapeType=Capsule, length=8.6456, radius=14.5388, center=<-2.4887 0.1881 0>}]
 spine_01 [{shapeType=Capsule, length=7.9438, radius=13.5586, center=<-1.7965 2.0776 0>}]
 spine_02 [
  {shapeType=Capsule, length=7.3264, radius=13.6322, center=<-1.711 2.1027 0.0001>}
  {shapeType=Capsule, length=29.3956, radius=11.5807, center=<2.5024 0.8889 3.9825>, rotation=<rot 90 0 0>}
  {shapeType=Capsule, length=29.875, radius=11.1013, center=<2.4704 0.8888 -3.9825>, rotation=<rot 90 0 0>}
 ]
 spine_03 [{shapeType=Capsule, length=8.2107, radius=13.7031, center=<0.1445 2.6533 0.0001>}]
 spine_04 [{shapeType=Capsule, length=9.4404, radius=14.1325, center=<2.766 2.8219 0.0002>}]
 spine_05 [
  {shapeType=Capsule, length=8.0226, radius=15.695, center=<-2.2795 4.0847 0>}
  {shapeType=Capsule, length=13.9518, radius=13.9425, center=<-0.2916 2.3903 0>}
 ]
 thigh_l []
 thigh_r []
 upperarm_l []
 upperarm_r []
}
constraints [
 {
  bone1Name=dyn_skirt_Back_l_01
  bone2Name=spine_02
  swing1LimitDegrees=45
  swing1Motion=Free
  swing2LimitDegrees=45
  swing2Motion=Free
  twistLimitDegrees=45
  twistMotion=Free
 }
 {
  bone1Name=dyn_skirt_Back_l_02
  bone2Name=dyn_skirt_Back_l_01
  swing1LimitDegrees=45
  swing1Motion=Free
  swing2LimitDegrees=45
  swing2Motion=Free
  twistLimitDegrees=45
  twistMotion=Free
 }
 {
  bone1Name=dyn_skirt_Back_l_03
  bone2Name=dyn_skirt_Back_l_02
  swing1LimitDegrees=45
  swing1Motion=Free
  swing2LimitDegrees=45
  swing2Motion=Free
  twistLimitDegrees=45
  twistMotion=Free
 }
 {
  bone1Name=dyn_skirt_Back_l_04
  bone2Name=dyn_skirt_Back_l_03
  swing1LimitDegrees=45
  swing1Motion=Free
  swing2LimitDegrees=45
  swing2Motion=Free
  twistLimitDegrees=45
  twistMotion=Free
 }
 {
  bone1Name=dyn_skirt_Back_l_05
  bone2Name=dyn_skirt_Back_l_04
  swing1LimitDegrees=45
  swing1Motion=Free
  swing2LimitDegrees=45
  swing2Motion=Free
  twistLimitDegrees=45
  twistMotion=Free
 }
 {
  bone1Name=dyn_skirt_Back_r_01
  bone2Name=spine_02
  swing1LimitDegrees=45
  swing1Motion=Free
  swing2LimitDegrees=45
  swing2Motion=Free
  twistLimitDegrees=45
  twistMotion=Free
 }
 {
  bone1Name=dyn_skirt_Back_r_02
  bone2Name=dyn_skirt_Back_r_01
  swing1LimitDegrees=45
  swing1Motion=Free
  swing2LimitDegrees=45
  swing2Motion=Free
  twistLimitDegrees=45
  twistMotion=Free
 }
 {
  bone1Name=dyn_skirt_Back_r_03
  bone2Name=dyn_skirt_Back_r_02
  swing1LimitDegrees=45
  swing1Motion=Free
  swing2LimitDegrees=45
  swing2Motion=Free
  twistLimitDegrees=45
  twistMotion=Free
 }
 {
  bone1Name=dyn_skirt_Back_r_04
  bone2Name=dyn_skirt_Back_r_03
  swing1LimitDegrees=45
  swing1Motion=Free
  swing2LimitDegrees=45
  swing2Motion=Free
  twistLimitDegrees=45
  twistMotion=Free
 }
 {
  bone1Name=dyn_skirt_Back_r_05
  bone2Name=dyn_skirt_Back_r_04
  swing1LimitDegrees=45
  swing1Motion=Free
  swing2LimitDegrees=45
  swing2Motion=Free
  twistLimitDegrees=45
  twistMotion=Free
 }
 {
  bone1Name=dyn_skirt_Front_l_01
  bone2Name=spine_02
  swing1LimitDegrees=45
  swing1Motion=Free
  swing2LimitDegrees=45
  swing2Motion=Free
  twistLimitDegrees=45
  twistMotion=Free
 }
 {
  bone1Name=dyn_skirt_Front_r_01
  bone2Name=spine_02
  swing1LimitDegrees=45
  swing1Motion=Free
  swing2LimitDegrees=45
  swing2Motion=Free
  twistLimitDegrees=45
  twistMotion=Free
 }
 {
  bone1Name=dyn_skirt_SideBack_l_01
  bone2Name=spine_02
  swing1LimitDegrees=45
  swing1Motion=Free
  swing2LimitDegrees=45
  swing2Motion=Free
  twistLimitDegrees=45
  twistMotion=Free
 }
 {
  bone1Name=dyn_skirt_SideBack_l_02
  bone2Name=dyn_skirt_SideBack_l_01
  swing1LimitDegrees=45
  swing1Motion=Free
  swing2LimitDegrees=45
  swing2Motion=Free
  twistLimitDegrees=45
  twistMotion=Free
 }
 {
  bone1Name=dyn_skirt_SideBack_l_03
  bone2Name=dyn_skirt_SideBack_l_02
  swing1LimitDegrees=45
  swing1Motion=Free
  swing2LimitDegrees=45
  swing2Motion=Free
  twistLimitDegrees=45
  twistMotion=Free
 }
 {
  bone1Name=dyn_skirt_SideBack_l_04
  bone2Name=dyn_skirt_SideBack_l_03
  swing1LimitDegrees=45
  swing1Motion=Free
  swing2LimitDegrees=45
  swing2Motion=Free
  twistLimitDegrees=45
  twistMotion=Free
 }
 {
  bone1Name=dyn_skirt_SideBack_l_05
  bone2Name=dyn_skirt_SideBack_l_04
  swing1LimitDegrees=45
  swing1Motion=Free
  swing2LimitDegrees=45
  swing2Motion=Free
  twistLimitDegrees=45
  twistMotion=Free
 }
 {
  bone1Name=dyn_skirt_SideBack_r_01
  bone2Name=spine_02
  swing1LimitDegrees=45
  swing1Motion=Free
  swing2LimitDegrees=45
  swing2Motion=Free
  twistLimitDegrees=45
  twistMotion=Free
 }
 {
  bone1Name=dyn_skirt_SideBack_r_02
  bone2Name=dyn_skirt_SideBack_r_01
  swing1LimitDegrees=45
  swing1Motion=Free
  swing2LimitDegrees=45
  swing2Motion=Free
  twistLimitDegrees=45
  twistMotion=Free
 }
 {
  bone1Name=dyn_skirt_SideBack_r_03
  bone2Name=dyn_skirt_SideBack_r_02
  swing1LimitDegrees=45
  swing1Motion=Free
  swing2LimitDegrees=45
  swing2Motion=Free
  twistLimitDegrees=45
  twistMotion=Free
 }
 {
  bone1Name=dyn_skirt_SideBack_r_04
  bone2Name=dyn_skirt_SideBack_r_03
  swing1LimitDegrees=45
  swing1Motion=Free
  swing2LimitDegrees=45
  swing2Motion=Free
  twistLimitDegrees=45
  twistMotion=Free
 }
 {
  bone1Name=dyn_skirt_SideBack_r_05
  bone2Name=dyn_skirt_SideBack_r_04
  swing1LimitDegrees=45
  swing1Motion=Free
  swing2LimitDegrees=45
  swing2Motion=Free
  twistLimitDegrees=45
  twistMotion=Free
 }
 {
  bone1Name=dyn_skirt_SideFront_l_01
  bone2Name=spine_02
  swing1LimitDegrees=45
  swing1Motion=Free
  swing2LimitDegrees=45
  swing2Motion=Free
  twistLimitDegrees=45
  twistMotion=Free
 }
 {
  bone1Name=dyn_skirt_SideFront_l_02
  bone2Name=dyn_skirt_SideFront_l_01
  swing1LimitDegrees=45
  swing1Motion=Free
  swing2LimitDegrees=45
  swing2Motion=Free
  twistLimitDegrees=45
  twistMotion=Free
 }
 {
  bone1Name=dyn_skirt_SideFront_l_03
  bone2Name=dyn_skirt_SideFront_l_02
  swing1LimitDegrees=45
  swing1Motion=Free
  swing2LimitDegrees=45
  swing2Motion=Free
  twistLimitDegrees=45
  twistMotion=Free
 }
 {
  bone1Name=dyn_skirt_SideFront_l_04
  bone2Name=dyn_skirt_SideFront_l_03
  swing1LimitDegrees=45
  swing1Motion=Free
  swing2LimitDegrees=45
  swing2Motion=Free
  twistLimitDegrees=45
  twistMotion=Free
 }
 {
  bone1Name=dyn_skirt_SideFront_l_05
  bone2Name=dyn_skirt_SideFront_l_04
  swing1LimitDegrees=45
  swing1Motion=Free
  swing2LimitDegrees=45
  swing2Motion=Free
  twistLimitDegrees=45
  twistMotion=Free
 }
 {
  bone1Name=dyn_skirt_SideFront_r_01
  bone2Name=spine_02
  swing1LimitDegrees=45
  swing1Motion=Free
  swing2LimitDegrees=45
  swing2Motion=Free
  twistLimitDegrees=45
  twistMotion=Free
 }
 {
  bone1Name=dyn_skirt_SideFront_r_02
  bone2Name=dyn_skirt_SideFront_r_01
  swing1LimitDegrees=45
  swing1Motion=Free
  swing2LimitDegrees=45
  swing2Motion=Free
  twistLimitDegrees=45
  twistMotion=Free
 }
 {
  bone1Name=dyn_skirt_SideFront_r_03
  bone2Name=dyn_skirt_SideFront_r_02
  swing1LimitDegrees=45
  swing1Motion=Free
  swing2LimitDegrees=45
  swing2Motion=Free
  twistLimitDegrees=45
  twistMotion=Free
 }
 {
  bone1Name=dyn_skirt_SideFront_r_04
  bone2Name=dyn_skirt_SideFront_r_03
  swing1LimitDegrees=45
  swing1Motion=Free
  swing2LimitDegrees=45
  swing2Motion=Free
  twistLimitDegrees=45
  twistMotion=Free
 }
 {
  bone1Name=dyn_skirt_SideFront_r_05
  bone2Name=dyn_skirt_SideFront_r_04
  swing1LimitDegrees=45
  swing1Motion=Free
  swing2LimitDegrees=45
  swing2Motion=Free
  twistLimitDegrees=45
  twistMotion=Free
 }
]
