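# Post-simulation Control Rig for the BountyPond_Quart male body, built to keep dynamic skirt bones out of the calves.
# A reusable function spatially clamps a skirt bone to an offset capsule around a calf, overriding its transform when within max radius, with debug drawing enabled.
# Separate test chains add bone nudges and a distance-driven, interpolated skirt-base rotation.
# Begin Execution connects to none of these chains, and some left-side collision nodes are unlinked.
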ControlRigBlueprint /BRCosmetics/Characters/Player/Male/Medium/Bodies/M_MED_BountyPond_Quart/Meshes/M_MED_BountyPond_Quart_PostSim_CtrlRig
graphs [
 New Function_ContainedGraph {nodes [
   n1 = RigVMFunctionEntryNode()
   RigVMFunction_MathQuaternionFromRotatorV2 = /Script/RigVM.RigVMFunction_MathQuaternionFromRotatorV2(Value=@n1.Target Offset Rotation) {is=struct RigVMUnitNode, template="Cast::Execute(in Value,out Result)"}
   Make Transform = RigVMDispatchNode(
    Value=(Rotation=(X=0,Y=0,Z=0,W=1),Translation=(X=0,Y=0,Z=0),Scale3D=(X=1,Y=1,Z=1))
    Value.Rotation=@RigVMFunction_MathQuaternionFromRotatorV2.Result
    Value.Translation=@n1.Target Offset
   ) {template="DISPATCH_RigVMDispatch_Constant(io Value)"}
   n2 = &RigUnit_GetTransform(bInitial=False, Item=@n1.Source)
   Make Transform_1 = RigVMDispatchNode(Value=(Rotation=(X=0,Y=0,Z=0,W=1),Translation=(X=0,Y=0,Z=0),Scale3D=(X=1,Y=1,Z=1)), Value.Translation=@n1.Source Offset) {template="DISPATCH_RigVMDispatch_Constant(io Value)"}
   ClampSpatially = /Script/RigVM.RigVMFunction_MathTransformClampSpatially(
    Axis=X
    DebugColor=(R=1,G=0,B=0,A=1)
    DebugThickness=1
    Type=Capsule
    Space=&RigVMFunction_MathTransformMul(B=&RigUnit_GetTransform(bInitial=False, Item=@n1.Target).Transform, A=@"Make Transform".Value) {template="Multiply::Execute(in A,in B,out Result)"}.Result
    Value=&RigVMFunction_MathTransformMul(B=@n2.Transform, A=@"Make Transform_1".Value) {template="Multiply::Execute(in A,in B,out Result)"}.Result
    Maximum=@n1.Max Radius
    Minimum=@n1.Min Radius
    bDrawDebug=@n1.Debug
   ) {
    is=struct RigVMUnitNode
    template="ClampSpatially::Execute(in Value,in Axis,in Type,in Minimum,in Maximum,in Space,in bDrawDebug,in DebugColor,in DebugThickness,out Result)"
   }
   LessEqual = /Script/RigVM.RigVMFunction_MathDoubleLessEqual(B=@n1.Max Radius, A=&RigVMFunction_MathVectorDistance(B=@ClampSpatially.Result.Translation, A=@n2.Transform.Translation).Result) {is=struct RigVMUnitNode, template="LessEqual::Execute(in A,in B,out Result)"}
   n3 = &RigVMFunction_ControlFlowBranch(ExecuteContext=@n1.ExecuteContext, Condition=@LessEqual.Result)
   RigVMFunctionReturnNode(isColliding=@LessEqual.Result, ExecuteContext=@n3.Completed)
   &RigUnit_ModifyTransforms(
    ItemToModify=((Item=(Type=Bone,Name="None"),Transform=(Rotation=(X=0,Y=0,Z=0,W=1),Translation=(X=0,Y=0,Z=0),Scale3D=(X=1,Y=1,Z=1))))
    Mode=OverrideGlobal
    WeightMinimum=0
    ItemToModify.0.Item=@n1.Source
    ExecutePin=@n3.True
    Weight=@n1.Weight
    ItemToModify.0.Transform=@ClampSpatially.Result
   )
  ]}
 RigVMFunctionLibrary {nodes [Collision Leg Spherical Collisions Edited = RigVMCollapseNode(
    Debug=False
    ExecuteContext=()
    Source=(Type=None,Name="None")
    Source Offset=(X=0,Y=0,Z=0)
    Target=(Type=None,Name="None")
    Target Offset=(X=0,Y=0,Z=0)
    Target Offset Rotation=(Pitch=0,Yaw=0,Roll=0)
   )]}
 RigVMFunction_Sequence_1_ContainedGraph {nodes [
   n1 = RigVMFunctionEntryNode()
   RigVMFunction_Sequence = &RigVMFunction_Sequence(ExecuteContext=@n1.ExecuteContext)
   n2 = &RigVMFunction_Sequence(ExecuteContext=@RigVMFunction_Sequence.B)
   RigVMFunctionReturnNode(A=@RigVMFunction_Sequence.A, B=@n2.A, C=@n2.B)
  ]}
 RigVMModel {nodes [
   Make Transform_1 = RigVMDispatchNode(Value=(Rotation=(X=0,Y=0,Z=0,W=1),Translation=(X=0,Y=4,Z=-6.5),Scale3D=(X=1,Y=1,Z=1))) {template="DISPATCH_RigVMDispatch_Constant(io Value)"}
   n1 = &RigVMFunction_MathTransformMul(
    B=&RigUnit_GetTransform(Item=(Type=Bone,Name="dyn_skirt_Front_l_04"), bInitial=False).Transform
    A=@"Make Transform_1".Value
   ) {template="Multiply::Execute(in A,in B,out Result)"}
   n2 = &RigVMFunction_MathVectorDistance(
    A=&RigUnit_GetTransform(Item=(Type=Control,Name="ref_calf_ctrl"), bInitial=False).Transform.Translation
    B=@n1.Result.Translation
   )
   RigVMDispatchNode(Enabled=True, ExecuteContext=(), ScreenColor=(R=1,G=1,B=1,A=1), ScreenDuration=0.05, Value=@n2.Result) {template="DISPATCH_RigVMDispatch_Print(in Prefix,in Value,in Enabled,in ScreenDuration,in ScreenColor)"}
   > RigVMFunction_DebugTransformMutableNoSpace = /Script/RigVM.RigVMFunction_DebugTransformMutableNoSpace(
    Color=(R=1,G=0,B=0,A=1)
    DebugDrawSettings=(DepthPriority=SDPG_Foreground,Lifetime=-1)
    Mode=Axes
    Scale=10
    Thickness=0
    WorldOffset=(Rotation=(X=0,Y=0,Z=0,W=1),Translation=(X=0,Y=0,Z=0),Scale3D=(X=1,Y=1,Z=1))
    bEnabled=True
    Transform=@n1.Result
   ) {is=struct RigVMUnitNode}
   Less = /Script/RigVM.RigVMFunction_MathDoubleLess(B=5, A=@n2.Result) {is=struct RigVMUnitNode, template="Less::Execute(in A,in B,out Result)"}
   RigVMFunction_MathBoolToFloat = &RigVMFunction_MathBoolToFloat(Value=@Less.Result)
   AlphaInterp = &RigVMFunction_AlphaInterp(InterpSpeedDecreasing=10, InterpSpeedIncreasing=10, Scale=1, bClampResult=true, Value=@RigVMFunction_MathBoolToFloat.Result)
   > &RigUnit_ModifyTransforms(
    ItemToModify="((Item=(Type=Bone,Name=\"base_skirt_Front_l_01\"),Transform=(Rotation=(X=0,Y=0,Z=-0.0872,W=0.9962),Translation=(X=0,Y=0,Z=0),Scale3D=(X=1,Y=1,Z=1))),(Item=(Type=Bone,Name=\"base_skirt_SideFront_l_01\"),Transform=(Rotation=(X=0,Y=0,Z=-0.0872,W=0.9962),Translati ...(truncated)"
    WeightMinimum=0
    Weight=@AlphaInterp.Result
   )
   RigVMFunction_Sequence = RigVMAggregateNode(ExecuteContext=())
   Collision Leg Spherical Collisions Edited = RigVMFunctionReferenceNode(
    Debug=true
    Max Radius=50
    Min Radius=9
    Source=(Type=Bone,Name="dyn_skirt_Front_l_04")
    Source Offset=(X=0,Y=0,Z=0)
    Target=(Type=Bone,Name="calf_l")
    Target Offset=(X=-18,Y=-2,Z=1.5)
    Target Offset Rotation=(Pitch=1,Yaw=0,Roll=0)
    Weight=1
    ExecuteContext=@RigVMFunction_Sequence.A
   )
   > Collision Leg Spherical Collisions Edited_3 = RigVMFunctionReferenceNode(
    Debug=true
    Max Radius=50
    Min Radius=9
    Source=(Type=Bone,Name="dyn_skirt_Front_l_05")
    Source Offset=(X=0,Y=0,Z=0)
    Target=(Type=Bone,Name="calf_l")
    Target Offset=(X=-18,Y=-2,Z=1.5)
    Target Offset Rotation=(Pitch=1,Yaw=0,Roll=0)
    Weight=1
   )
   > Collision Leg Spherical Collisions Edited_6 = RigVMFunctionReferenceNode(
    Debug=true
    Max Radius=50
    Min Radius=9
    Source=(Type=Bone,Name="dyn_skirt_Front_r_04")
    Source Offset=(X=0,Y=0,Z=0)
    Target=(Type=Bone,Name="calf_r")
    Target Offset=(X=18,Y=2,Z=-2)
    Target Offset Rotation=(Pitch=1,Yaw=0,Roll=0)
    Weight=1
   )
   > Collision Leg Spherical Collisions Edited_7 = RigVMFunctionReferenceNode(
    Debug=true
    Max Radius=50
    Min Radius=9
    Source=(Type=Bone,Name="dyn_skirt_SideFront_r_04")
    Source Offset=(X=0,Y=0,Z=0)
    Target=(Type=Bone,Name="calf_r")
    Target Offset=(X=18,Y=2,Z=-2)
    Target Offset Rotation=(Pitch=1,Yaw=0,Roll=0)
    Weight=1
   )
   > Collision Leg Spherical Collisions Edited_8 = RigVMFunctionReferenceNode(
    Debug=true
    Max Radius=50
    Min Radius=9
    Source=(Type=Bone,Name="dyn_skirt_SideBack_r_04")
    Source Offset=(X=0,Y=0,Z=0)
    Target=(Type=Bone,Name="calf_r")
    Target Offset=(X=18,Y=2,Z=-2)
    Target Offset Rotation=(Pitch=1,Yaw=0,Roll=0)
    Weight=1
   )
   > Collision Leg Spherical Collisions Edited_9 = RigVMFunctionReferenceNode(
    Debug=true
    Max Radius=50
    Min Radius=9
    Source=(Type=Bone,Name="dyn_skirt_Front_r_05")
    Source Offset=(X=0,Y=0,Z=0)
    Target=(Type=Bone,Name="calf_r")
    Target Offset=(X=18,Y=2,Z=-2)
    Target Offset Rotation=(Pitch=1,Yaw=0,Roll=0)
    Weight=1
   )
   > Collision Leg Spherical Collisions Edited_10 = RigVMFunctionReferenceNode(
    Debug=true
    Max Radius=50
    Min Radius=9
    Source=(Type=Bone,Name="dyn_skirt_SideFront_r_05")
    Source Offset=(X=0,Y=0,Z=0)
    Target=(Type=Bone,Name="calf_r")
    Target Offset=(X=18,Y=2,Z=-2)
    Target Offset Rotation=(Pitch=1,Yaw=0,Roll=0)
    Weight=1
   )
   > Collision Leg Spherical Collisions Edited_11 = RigVMFunctionReferenceNode(
    Debug=true
    Max Radius=50
    Min Radius=9
    Source=(Type=Bone,Name="dyn_skirt_SideBack_r_05")
    Source Offset=(X=0,Y=0,Z=0)
    Target=(Type=Bone,Name="calf_r")
    Target Offset=(X=18,Y=2,Z=-2)
    Target Offset Rotation=(Pitch=1,Yaw=0,Roll=0)
    Weight=1
   )
   RigVMFunction_ControlFlowBranch = &RigVMFunction_ControlFlowBranch(Condition=true, ExecuteContext=@RigVMFunction_Sequence.B)
   &RigUnit_ModifyTransforms(
    ItemToModify=((Item=(Type=Bone,Name="dyn_skirt_SideFront_r_03"),Transform=(Rotation=(X=0,Y=0,Z=0,W=1),Translation=(X=0,Y=-5,Z=0),Scale3D=(X=1,Y=1,Z=1))))
    Weight=1
    WeightMinimum=0
    ExecutePin=@RigVMFunction_ControlFlowBranch.True
   )
   n3 = &RigVMFunction_ControlFlowBranch(Condition=@"Collision Leg Spherical Collisions Edited".isColliding, ExecuteContext=@RigVMFunction_Sequence.C)
   &RigUnit_ModifyTransforms(
    ItemToModify=((Item=(Type=Bone,Name="dyn_skirt_Front_l_04"),Transform=(Rotation=(X=0,Y=0,Z=0,W=1),Translation=(X=0,Y=0,Z=0),Scale3D=(X=1,Y=1,Z=1))))
    Weight=1
    WeightMinimum=0
    ExecutePin=@n3.True
   )
   Collision Leg Spherical Collisions Edited_1 = RigVMFunctionReferenceNode(
    Debug=true
    ExecuteContext=()
    Max Radius=50
    Min Radius=9
    Source=(Type=Bone,Name="dyn_skirt_SideFront_l_04")
    Source Offset=(X=0,Y=0,Z=0)
    Target=(Type=Bone,Name="calf_l")
    Target Offset=(X=-18,Y=-2,Z=1.5)
    Target Offset Rotation=(Pitch=1,Yaw=0,Roll=0)
    Weight=1
   )
   Collision Leg Spherical Collisions Edited_4 = RigVMFunctionReferenceNode(
    Debug=true
    ExecuteContext=()
    Max Radius=50
    Min Radius=9
    Source=(Type=Bone,Name="dyn_skirt_SideFront_l_05")
    Source Offset=(X=0,Y=0,Z=0)
    Target=(Type=Bone,Name="calf_l")
    Target Offset=(X=-18,Y=-2,Z=1.5)
    Target Offset Rotation=(Pitch=1,Yaw=0,Roll=0)
    Weight=1
   )
   Collision Leg Spherical Collisions Edited_5 = RigVMFunctionReferenceNode(
    Debug=true
    ExecuteContext=()
    Max Radius=50
    Min Radius=9
    Source=(Type=Bone,Name="dyn_skirt_SideBack_l_05")
    Source Offset=(X=0,Y=0,Z=0)
    Target=(Type=Bone,Name="calf_l")
    Target Offset=(X=-18,Y=-2,Z=1.5)
    Target Offset Rotation=(Pitch=1,Yaw=0,Roll=0)
    Weight=1
   )
   &RigUnit_ModifyTransforms(
    ExecutePin=()
    ItemToModify="((Item=(Type=Bone,Name=\"dyn_skirt_Front_l_03\"),Transform=(Rotation=(X=0,Y=0,Z=0,W=1),Translation=(X=0,Y=0,Z=0),Scale3D=(X=1,Y=1,Z=1))),(Item=(Type=Bone,Name=\"dyn_skirt_SideFront_l_03\"),Transform=(Rotation=(X=0,Y=0,Z=0,W=1),Translation=(X=0,Y=0,Z=0),Scale3D ...(truncated)"
    Weight=1
    WeightMinimum=0
   )
   &RigUnit_PrepareForExecution()
   RigUnit_BeginExecution = &RigUnit_BeginExecution()
  ]}
]
hierarchy [
 root {loc=<0 0 0>, rot=<0 0 0 1>}
 pelvis {parent=root, loc=<0 -0.7071 93.1191>, rot=<0.0127 0.707 -0.0127 -0.707>}
 spine_01 {parent=pelvis, loc=<6.7494 -0.6388 0>, rot=<0 0 0.034 -0.9994>}
 spine_02 {parent=spine_01, loc=<7.2518 -0.035 0>, rot=<0 0 0.0238 -0.9997>}
 spine_03 {parent=spine_02, loc=<6.0104 0 0>, rot=<0 0 0.0149 -0.9999>}
 spine_04 {parent=spine_03, loc=<8.1024 0 0>, rot=<0 0 -0.0132 -0.9999>}
 spine_05 {parent=spine_04, loc=<10.0146 0 0>, rot=<0 0 -0.0015 -1>}
 neck_01 {parent=spine_05, loc=<11.3138 0.1529 0>, rot=<0 0 -0.2003 -0.9797>}
 neck_02 {parent=neck_01, loc=<4.5627 0 0>, rot=<0 0 0.0504 -0.9987>}
 head {parent=neck_02, loc=<5.3914 0 0>, rot=<0 0 0.115 -0.9934>}
 attach_cape {parent=spine_05, loc=<9.017 -11.714 0>, rot=<0.0616 -0.7044 -0.0616 -0.7044>}
 attach_backpack {parent=spine_05, loc=<-0.123 -13.473 0>, rot=<-0.0093 -0.707 0.0093 -0.707>}
 clavicle_l {parent=spine_05, loc=<4.3575 -0.1466 -2.3223>, rot=<0.0465 -0.7289 -0.009 -0.6829>}
 upperarm_l {parent=clavicle_l, loc=<15.5111 0 0>, rot=<0.0226 -0.3234 -0.0088 -0.9459>}
 lowerarm_l {parent=upperarm_l, loc=<26.6514 0 0>, rot=<0.0391 -0.0099 -0.2453 -0.9686>}
 hand_l {parent=lowerarm_l, loc=<22.3491 0 0>, rot=<-0.7081 -0.0507 0.0975 -0.6975>}
 thumb_01_l {parent=hand_l, loc=<1.3118 -0.3191 -2.459>, rot=<0.6546 -0.4042 0.0664 -0.6354>}
 thumb_02_l {parent=thumb_01_l, loc=<5.9336 -0.0448 0.0401>, rot=<-0.0037 0.0213 0.298 -0.9543>}
 thumb_03_l {parent=thumb_02_l, loc=<2.2223 0.3682 -0.0907>, rot=<0.0118 -0.013 0.0066 -0.9998>}
 index_metacarpal_l {parent=hand_l, loc=<1.6535 0.0183 -2.96>, rot=<0.0158 -0.0124 -0.0173 -0.9996>}
 index_01_l {parent=index_metacarpal_l, loc=<8.3008 0.0357 -0.0524>, rot=<-0.0044 0.0485 0.0924 -0.9945>}
 index_02_l {parent=index_01_l, loc=<3.3587 -0.3004 0.0452>, rot=<-0.002 0.0027 0.3225 -0.9466>}
 index_03_l {parent=index_02_l, loc=<2.5578 -0.123 -0.0252>, rot=<-0.002 0.0067 0.2558 -0.9667>}
 middle_metacarpal_l {parent=hand_l, loc=<2.1856 -0.1725 -0.5984>, rot=<0.006 0.0168 -0.0329 -0.9993>}
 middle_01_l {parent=middle_metacarpal_l, loc=<7.3275 -0.2447 0.0129>, rot=<-0.0208 0.0042 0.0858 -0.9961>}
 middle_02_l {parent=middle_01_l, loc=<4.0123 0.1401 0.1175>, rot=<-0.0218 0.0258 0.3512 -0.9357>}
 middle_03_l {parent=middle_02_l, loc=<2.6077 0.0534 0.1373>, rot=<0.0082 0.0188 0.1708 -0.9851>}
 ring_metacarpal_l {parent=hand_l, loc=<2.2074 -0.3625 1.3638>, rot=<-0.0284 0.0722 -0.0458 -0.9959>}
 ring_01_l {parent=ring_metacarpal_l, loc=<7.0295 -0.6197 -0.1446>, rot=<-0.0489 -0.0252 0.1892 -0.9804>}
 ring_02_l {parent=ring_01_l, loc=<3.6174 0.4647 -0.1027>, rot=<0 0 0.2806 -0.9598>}
 ring_03_l {parent=ring_02_l, loc=<2.4152 0.114 0.0199>, rot=<0 0 0.1189 -0.9929>}
 pinky_metacarpal_l {parent=hand_l, loc=<1.7616 -0.9753 3.3814>, rot=<-0.0397 0.0814 -0.069 -0.9935>}
 pinky_01_l {parent=pinky_metacarpal_l, loc=<7.1709 -0.4831 -0.1094>, rot=<-0.0654 -0.0035 0.2242 -0.9723>}
 pinky_02_l {parent=pinky_01_l, loc=<2.8274 0.0893 -0.1043>, rot=<0.0037 0.0011 0.2854 -0.9584>}
 pinky_03_l {parent=pinky_02_l, loc=<2.2046 0.0307 -0.0078>, rot=<-0.0002 0.0098 0.1445 -0.9895>}
 deform_wrist_dn_base_l {parent=hand_l, loc=<0 -2.2342 0>, rot=<0 0 0 -1>}
 deform_wrist_dn_l {parent=deform_wrist_dn_base_l, loc=<0 0 0>, rot=<0 0 0 1>}
 deform_wrist_up_base_l {parent=hand_l, loc=<0 2.2682 0>, rot=<0 0 0 -1>}
 deform_wrist_up_l {parent=deform_wrist_up_base_l, loc=<0 0 0>, rot=<0 0 0 1>}
 lowerarm_twist_02_l {parent=lowerarm_l, loc=<9.6199 0 0>, rot=<0 0 0 -1>}
 base_sleeve_01_l {parent=lowerarm_twist_02_l, loc=<3.9055 0.8402 3.8173>, rot=<-0.0597 -0.586 0.8057 0.0614>}
 dyn_sleeve_01_l {parent=base_sleeve_01_l, loc=<0 0 0>, rot=<0 0 0 -1>}
 base_sleeve_02_l {parent=lowerarm_twist_02_l, loc=<3.4144 -4.6071 -1.7269>, rot=<-0.1034 -0.9912 0.0702 0.0446>}
 dyn_sleeve_02_l {parent=base_sleeve_02_l, loc=<0 0 0>, rot=<0 0 0 1>}
 base_sleeve_03_l {parent=lowerarm_twist_02_l, loc=<5.2871 0.9873 -4.7514>, rot=<0.1079 0.4178 0.9013 0.0389>}
 dyn_sleeve_03_l {parent=base_sleeve_03_l, loc=<0 0 0>, rot=<0 0 0 1>}
 lowerarm_twist_01_l {parent=lowerarm_l, loc=<17.8549 0 0>, rot=<0 0 0 -1>}
 deform_elbow_in_base_l {parent=lowerarm_l, loc=<0.2112 5.259 0.8274>, rot=<0 0 0 1>}
 deform_elbow_in_l {parent=deform_elbow_in_base_l, loc=<0 0 0>, rot=<0 0 0 1>}
 deform_elbow_out_base_l {parent=lowerarm_l, loc=<0 -4 0>, rot=<0 0 0 1>}
 deform_elbow_out_l {parent=deform_elbow_out_base_l, loc=<0 0 0>, rot=<0 0 0 1>}
 upperarm_twist_01_l {parent=upperarm_l, loc=<0.0041 0 0>, rot=<0 0 0 -1>}
 deform_upperarm_bk_base_l {parent=upperarm_twist_01_l, loc=<1.0813 -7.2193 -1.9453>, rot=<0 0 0 -1>}
 deform_upperarm_bk_l {parent=deform_upperarm_bk_base_l, loc=<0 0 0>, rot=<0 0 0 1>}
 deform_upperarm_fr_base_l {parent=upperarm_twist_01_l, loc=<-0.3691 7.0506 -2.765>, rot=<0 0 0 -1>}
 deform_upperarm_fr_l {parent=deform_upperarm_fr_base_l, loc=<0 0 0>, rot=<0 0 0 1>}
 deform_upperarm_up_base_l {parent=upperarm_twist_01_l, loc=<-3.2216 -1.1391 3.3311>, rot=<0 0 0 1>}
 deform_upperarm_up_l {parent=deform_upperarm_up_base_l, loc=<0 0 0>, rot=<0 0 0 1>}
 upperarm_twist_02_l {parent=upperarm_l, loc=<13.7941 0 0>, rot=<0 0 0 -1>}
 clavicle_r {parent=spine_05, loc=<4.3575 -0.1466 2.3223>, rot=<0.7289 0.0465 -0.6829 0.009>}
 upperarm_r {parent=clavicle_r, loc=<-15.5111 0 0>, rot=<0.0226 -0.3234 -0.0088 -0.9459>}
 lowerarm_r {parent=upperarm_r, loc=<-26.6514 0 0>, rot=<0.0391 -0.0099 -0.2453 -0.9686>}
 hand_r {parent=lowerarm_r, loc=<-22.3491 0 0>, rot=<-0.7081 -0.0507 0.0975 -0.6975>}
 thumb_01_r {parent=hand_r, loc=<-1.3118 0.3191 2.459>, rot=<0.6546 -0.4042 0.0664 -0.6354>}
 thumb_02_r {parent=thumb_01_r, loc=<-5.9336 0.0448 -0.0401>, rot=<-0.0037 0.0213 0.298 -0.9543>}
 thumb_03_r {parent=thumb_02_r, loc=<-2.2223 -0.3682 0.0907>, rot=<0.0118 -0.013 0.0066 -0.9998>}
 index_metacarpal_r {parent=hand_r, loc=<-1.6535 -0.0183 2.96>, rot=<0.0158 -0.0124 -0.0173 -0.9996>}
 index_01_r {parent=index_metacarpal_r, loc=<-8.3008 -0.0357 0.0524>, rot=<-0.0044 0.0485 0.0924 -0.9945>}
 index_02_r {parent=index_01_r, loc=<-3.3587 0.3004 -0.0452>, rot=<-0.002 0.0027 0.3225 -0.9466>}
 index_03_r {parent=index_02_r, loc=<-2.5578 0.123 0.0252>, rot=<-0.002 0.0067 0.2558 -0.9667>}
 middle_metacarpal_r {parent=hand_r, loc=<-2.1856 0.1725 0.5984>, rot=<0.006 0.0168 -0.0329 -0.9993>}
 middle_01_r {parent=middle_metacarpal_r, loc=<-7.3275 0.2447 -0.0129>, rot=<-0.0208 0.0042 0.0858 -0.9961>}
 middle_02_r {parent=middle_01_r, loc=<-4.0123 -0.1401 -0.1175>, rot=<-0.0218 0.0258 0.3512 -0.9357>}
 middle_03_r {parent=middle_02_r, loc=<-2.6077 -0.0534 -0.1373>, rot=<0.0082 0.0188 0.1708 -0.9851>}
 ring_metacarpal_r {parent=hand_r, loc=<-2.2074 0.3625 -1.3638>, rot=<-0.0284 0.0722 -0.0458 -0.9959>}
 ring_01_r {parent=ring_metacarpal_r, loc=<-7.0295 0.6197 0.1446>, rot=<-0.0489 -0.0252 0.1892 -0.9804>}
 ring_02_r {parent=ring_01_r, loc=<-3.6174 -0.4647 0.1027>, rot=<0 0 0.2806 -0.9598>}
 ring_03_r {parent=ring_02_r, loc=<-2.4152 -0.114 -0.0199>, rot=<0 0 0.1189 -0.9929>}
 pinky_metacarpal_r {parent=hand_r, loc=<-1.7616 0.9753 -3.3814>, rot=<-0.0397 0.0814 -0.069 -0.9935>}
 pinky_01_r {parent=pinky_metacarpal_r, loc=<-7.1709 0.4831 0.1094>, rot=<-0.0654 -0.0035 0.2242 -0.9723>}
 pinky_02_r {parent=pinky_01_r, loc=<-2.8274 -0.0893 0.1043>, rot=<0.0037 0.0011 0.2854 -0.9584>}
 pinky_03_r {parent=pinky_02_r, loc=<-2.2046 -0.0307 0.0078>, rot=<-0.0002 0.0098 0.1445 -0.9895>}
 deform_wrist_dn_base_r {parent=hand_r, loc=<-0.0004 2.2345 -0.0001>, rot=<0 0 0 -1>}
 deform_wrist_dn_r {parent=deform_wrist_dn_base_r, loc=<0 0 0>, rot=<0 0 0 1>}
 deform_wrist_up_base_r {parent=hand_r, loc=<0.0003 -2.2684 0.0001>, rot=<0 0 0 -1>}
 deform_wrist_up_r {parent=deform_wrist_up_base_r, loc=<0 0 0>, rot=<0 0 0 1>}
 lowerarm_twist_02_r {parent=lowerarm_r, loc=<-9.6199 0 0>, rot=<0 0 0 -1>}
 base_sleeve_01_r {parent=lowerarm_twist_02_r, loc=<-3.1127 1.7432 -4.2253>, rot=<0.0689 0.6173 0.7806 0.0696>}
 dyn_sleeve_01_r {parent=base_sleeve_01_r, loc=<0 0 0>, rot=<0 0 0 1>}
 base_sleeve_02_r {parent=lowerarm_twist_02_r, loc=<-2.4921 4.5314 1.0447>, rot=<0.0325 -0.0604 0.993 0.0962>}
 dyn_sleeve_02_r {parent=base_sleeve_02_r, loc=<0 0 0>, rot=<0 0 0 1>}
 base_sleeve_03_r {parent=lowerarm_twist_02_r, loc=<-5.5611 -4.4998 0.8281>, rot=<0.0358 -0.9991 -0.0074 0.0224>}
 dyn_sleeve_03_r {parent=base_sleeve_03_r, loc=<0 0 0>, rot=<0 0 0 1>}
 lowerarm_twist_01_r {parent=lowerarm_r, loc=<-17.8549 0 0>, rot=<0 0 0 -1>}
 deform_elbow_in_base_r {parent=lowerarm_r, loc=<-0.2111 -5.2591 -0.8275>, rot=<0 0 0 -1>}
 deform_elbow_in_r {parent=deform_elbow_in_base_r, loc=<-0.0001 0 0.0002>, rot=<0 0 0 1>}
 deform_elbow_out_base_r {parent=lowerarm_r, loc=<0 4 -0.0001>, rot=<0 0 0 -1>}
 deform_elbow_out_r {parent=deform_elbow_out_base_r, loc=<0 0 0>, rot=<0 0 0 1>}
 upperarm_twist_01_r {parent=upperarm_r, loc=<-0.0041 0 0>, rot=<0 0 0 -1>}
 deform_upperarm_bk_base_r {parent=upperarm_twist_01_r, loc=<-1.0814 7.2192 1.9455>, rot=<0 0 0 1>}
 deform_upperarm_bk_r {parent=deform_upperarm_bk_base_r, loc=<0 0 0>, rot=<0 0 0 1>}
 deform_upperarm_fr_base_r {parent=upperarm_twist_01_r, loc=<0.3689 -7.0506 2.7652>, rot=<0 0 0 1>}
 deform_upperarm_fr_r {parent=deform_upperarm_fr_base_r, loc=<0 0 0>, rot=<0 0 0 1>}
 deform_upperarm_up_base_r {parent=upperarm_twist_01_r, loc=<3.2214 1.1391 -3.3308>, rot=<0 0 0 1>}
 deform_upperarm_up_r {parent=deform_upperarm_up_base_r, loc=<0 0 0>, rot=<0 0 0 -1>}
 upperarm_twist_02_r {parent=upperarm_r, loc=<-13.7941 0 0>, rot=<0 0 0 -1>}
 deform_pec_aim_base_l {parent=spine_04, loc=<7.8732 12.6388 -2.4123>, rot=<0.0102 -0.4602 0.3979 -0.7936>}
 deform_pec_aim_l {parent=deform_pec_aim_base_l, loc=<0 0 0>, rot=<0 0 0 1>}
 deform_pec_base_l {parent=deform_pec_aim_l, loc=<10.766 0 0>, rot=<0 0 0 1>}
 deform_pec_l {parent=deform_pec_base_l, loc=<0.055 0 0>, rot=<0 0 0 1>}
 deform_pec_aim_base_r {parent=spine_04, loc=<7.8727 12.6388 2.4123>, rot=<-0.4602 -0.0102 0.7936 0.3979>}
 deform_pec_aim_r {parent=deform_pec_aim_base_r, loc=<0 0 0>, rot=<0 0 0 1>}
 deform_pec_base_r {parent=deform_pec_aim_r, loc=<-10.7662 -0.0006 -0.0007>, rot=<0 0 0 1>}
 deform_pec_r {parent=deform_pec_base_r, loc=<0 0 0>, rot=<0 0 0 1>}
 beltStrings_Master {parent=spine_03, loc=<-3.8292 9.5929 -4.5917>, rot=<-0.0617 -0.9935 0.0908 -0.0305>}
 beltStrings_Sub {parent=beltStrings_Master, loc=<0 0 0>, rot=<0 0 0 1>}
 base_beltString_2_01 {parent=beltStrings_Sub, loc=<0.5212 0.128 0.2939>, rot=<0 0 0 1>}
 dyn_beltString_2_01 {parent=base_beltString_2_01, loc=<0 0 0>, rot=<0 0 0 1>}
 dyn_beltString_2_02 {parent=dyn_beltString_2_01, loc=<14.8064 0 0>, rot=<0 0 0.0679 -0.9977>}
 dyn_beltString_2_03 {parent=dyn_beltString_2_02, loc=<15.331 0.0081 -0.1165>, rot=<-0.0346 -0.0002 0.0062 -0.9994>}
 dyn_beltString_2_04 {parent=dyn_beltString_2_03, loc=<14.0343 0.0025 0.0302>, rot=<-0.0143 -0.023 -0.0003 -0.9996>}
 base_beltString_1_01 {parent=beltStrings_Sub, loc=<0.4272 -0.1991 -0.2169>, rot=<-0.0094 -0.0185 0.0272 -0.9994>}
 dyn_beltString_1_01 {parent=base_beltString_1_01, loc=<0 0 0>, rot=<0 0 0 1>}
 dyn_beltString_1_02 {parent=dyn_beltString_1_01, loc=<16.8741 0 0>, rot=<0 0 0.0638 -0.998>}
 dyn_beltString_1_03 {parent=dyn_beltString_1_02, loc=<16.6675 0 0>, rot=<0.1001 -0.0001 -0.0006 -0.995>}
 dyn_beltString_1_04 {parent=dyn_beltString_1_03, loc=<16.6843 0 0>, rot=<0 0 0.0042 -1>}
 Skirt_Master {parent=spine_02, loc=<6.0104 0 0>, rot=<0 0 0.0149 -0.9999>}
 Skirt_MasterOffset {parent=Skirt_Master, loc=<0 0 0>, rot=<0 0 0 1>}
 base_skirt_Back_l_01 {parent=Skirt_MasterOffset, loc=<-0.942 -5.1698 -3.565>, rot=<0.1092 -0.9852 -0.1321 -0.0072>}
 dyn_skirt_Back_l_01 {parent=base_skirt_Back_l_01, loc=<0 0 0>, rot=<0 0 0 1>}
 dyn_skirt_Back_l_02 {parent=dyn_skirt_Back_l_01, loc=<19.3916 0 0>, rot=<0 0 0 1>}
 dyn_skirt_Back_l_03 {parent=dyn_skirt_Back_l_02, loc=<18.6381 0 0>, rot=<0 0 0 1>}
 dyn_skirt_Back_l_04 {parent=dyn_skirt_Back_l_03, loc=<18.957 0 0>, rot=<0 0 0 1>}
 dyn_skirt_Back_l_05 {parent=dyn_skirt_Back_l_04, loc=<18.957 0 0>, rot=<0 0 0 1>}
 base_skirt_SideBack_l_01 {parent=Skirt_MasterOffset, loc=<-1.3755 -1.8114 -8.2799>, rot=<0.1204 -0.8348 -0.537 -0.0181>}
 dyn_skirt_SideBack_l_01 {parent=base_skirt_SideBack_l_01, loc=<0 0 0>, rot=<0 0 0 1>}
 dyn_skirt_SideBack_l_02 {parent=dyn_skirt_SideBack_l_01, loc=<19.3916 0 0>, rot=<0 0 0 1>}
 dyn_skirt_SideBack_l_03 {parent=dyn_skirt_SideBack_l_02, loc=<18.6381 0 0>, rot=<0 0 0 1>}
 dyn_skirt_SideBack_l_04 {parent=dyn_skirt_SideBack_l_03, loc=<18.957 0 0>, rot=<0 0 0 1>}
 dyn_skirt_SideBack_l_05 {parent=dyn_skirt_SideBack_l_04, loc=<18.957 0 0>, rot=<0 0 0 1>}
 base_skirt_SideBack_r_01 {parent=Skirt_MasterOffset, loc=<-1.3696 -1.8643 8.1131>, rot=<0.1204 -0.8348 0.537 0.0181>}
 dyn_skirt_SideBack_r_01 {parent=base_skirt_SideBack_r_01, loc=<0 0 0>, rot=<0 0 0 1>}
 dyn_skirt_SideBack_r_02 {parent=dyn_skirt_SideBack_r_01, loc=<19.3916 0 0>, rot=<0 0 0 1>}
 dyn_skirt_SideBack_r_03 {parent=dyn_skirt_SideBack_r_02, loc=<18.6381 0 0>, rot=<0 0 0 1>}
 dyn_skirt_SideBack_r_04 {parent=dyn_skirt_SideBack_r_03, loc=<18.957 0 0>, rot=<0 0 0 1>}
 dyn_skirt_SideBack_r_05 {parent=dyn_skirt_SideBack_r_04, loc=<18.957 0 0>, rot=<0 0 0 1>}
 base_skirt_Back_r_01 {parent=Skirt_MasterOffset, loc=<-0.942 -5.1698 3.1532>, rot=<0.1092 -0.9852 0.1321 0.0072>}
 dyn_skirt_Back_r_01 {parent=base_skirt_Back_r_01, loc=<0 0 0>, rot=<0 0 0 1>}
 dyn_skirt_Back_r_02 {parent=dyn_skirt_Back_r_01, loc=<19.3916 0 0>, rot=<0 0 0 1>}
 dyn_skirt_Back_r_03 {parent=dyn_skirt_Back_r_02, loc=<18.6381 0 0>, rot=<0 0 0 1>}
 dyn_skirt_Back_r_04 {parent=dyn_skirt_Back_r_03, loc=<18.957 0 0>, rot=<0 0 0 1>}
 dyn_skirt_Back_r_05 {parent=dyn_skirt_Back_r_04, loc=<18.957 0 0>, rot=<0 0 0 1>}
 base_skirt_SideFront_l_01 {parent=Skirt_MasterOffset, loc=<-1.9535 4.0289 -8.9924>, rot=<-0.088 0.6036 0.7923 0.0117>}
 dyn_skirt_SideFront_l_01 {parent=base_skirt_SideFront_l_01, loc=<0 0 0>, rot=<0 0 0 1>}
 dyn_skirt_SideFront_l_02 {parent=dyn_skirt_SideFront_l_01, loc=<19.3916 0 0>, rot=<0 0 0 1>}
 dyn_skirt_SideFront_l_03 {parent=dyn_skirt_SideFront_l_02, loc=<18.6381 0 0>, rot=<0 0 0 1>}
 dyn_skirt_SideFront_l_04 {parent=dyn_skirt_SideFront_l_03, loc=<18.957 0 0>, rot=<0 0 0 1>}
 dyn_skirt_SideFront_l_05 {parent=dyn_skirt_SideFront_l_04, loc=<18.957 0 0>, rot=<0 0 0 1>}
 base_skirt_SideFront_r_01 {parent=Skirt_MasterOffset, loc=<-1.9545 4.0384 9.2938>, rot=<0.088 -0.6036 0.7923 0.0117>}
 dyn_skirt_SideFront_r_01 {parent=base_skirt_SideFront_r_01, loc=<0 0 0>, rot=<0 0 0 1>}
 dyn_skirt_SideFront_r_02 {parent=dyn_skirt_SideFront_r_01, loc=<19.3916 0 0>, rot=<0 0 0 1>}
 dyn_skirt_SideFront_r_03 {parent=dyn_skirt_SideFront_r_02, loc=<18.6381 0 0>, rot=<0 0 0 1>}
 dyn_skirt_SideFront_r_04 {parent=dyn_skirt_SideFront_r_03, loc=<18.957 0 0>, rot=<0 0 0 1>}
 dyn_skirt_SideFront_r_05 {parent=dyn_skirt_SideFront_r_04, loc=<18.957 0 0>, rot=<0 0 0 1>}
 base_skirt_Front_r_01 {parent=Skirt_MasterOffset, loc=<-2.4524 8.566 4.722>, rot=<0.0459 -0.0832 0.9953 -0.0191>}
 dyn_skirt_Front_r_01 {parent=base_skirt_Front_r_01, loc=<0 0 0>, rot=<0 0 0 1>}
 dyn_skirt_Front_r_02 {parent=dyn_skirt_Front_r_01, loc=<19.3916 0 0>, rot=<0 0 0 1>}
 dyn_skirt_Front_r_03 {parent=dyn_skirt_Front_r_02, loc=<18.6381 0 0>, rot=<0 0 0 1>}
 dyn_skirt_Front_r_04 {parent=dyn_skirt_Front_r_03, loc=<18.957 0 0>, rot=<0 0 0 1>}
 dyn_skirt_Front_r_05 {parent=dyn_skirt_Front_r_04, loc=<18.957 0 0>, rot=<0 0 0 1>}
 base_skirt_Front_l_01 {parent=Skirt_MasterOffset, loc=<-2.4154 8.2294 -4.7217>, rot=<-0.0459 0.0832 0.9953 -0.0191>}
 dyn_skirt_Front_l_01 {parent=base_skirt_Front_l_01, loc=<0 0 0>, rot=<0 0 0 1>}
 dyn_skirt_Front_l_02 {parent=dyn_skirt_Front_l_01, loc=<19.3916 0 0>, rot=<0 0 0 1>}
 dyn_skirt_Front_l_03 {parent=dyn_skirt_Front_l_02, loc=<18.6381 0 0>, rot=<0 0 0 1>}
 dyn_skirt_Front_l_04 {parent=dyn_skirt_Front_l_03, loc=<18.957 0 0>, rot=<0 0 0 1>}
 dyn_skirt_Front_l_05 {parent=dyn_skirt_Front_l_04, loc=<18.957 0 0>, rot=<0 0 0 1>}
 thigh_l {parent=pelvis, loc=<-2.5693 -0.0109 -7.7501>, rot=<0.0012 0.0572 0.0155 -0.9982>}
 calf_l {parent=thigh_l, loc=<-39.0484 0 0>, rot=<0.0006 -0.0161 -0.0391 -0.9991>}
 foot_l {parent=calf_l, loc=<-42.561 0.1077 0.4292>, rot=<0.0418 -0.0455 0.0343 -0.9975>}
 ball_l {parent=foot_l, loc=<-6.1599 13.8264 0.6806>, rot=<0.0001 -0.0001 -0.7063 -0.7079>}
 calf_twist_02_l {parent=calf_l, loc=<-18.8064 -0.0967 0.0397>, rot=<0.0237 0.0095 -0.0035 -0.9997>}
 calf_twist_01_l {parent=calf_l, loc=<-35.813 -0.0967 0.0397>, rot=<0.0237 0.0095 -0.0035 -0.9997>}
 deform_knee_in_base_l {parent=calf_l, loc=<1.298 -5.2957 -0.0496>, rot=<0 0 0 -1>}
 deform_knee_in_l {parent=deform_knee_in_base_l, loc=<0 0 0>, rot=<0 0 0 1>}
 deform_knee_out_base_l {parent=calf_l, loc=<0 4 0>, rot=<0 0 0 -1>}
 deform_knee_out_l {parent=deform_knee_out_base_l, loc=<0 0 0>, rot=<0 0 0 1>}
 thigh_twist_01_l {parent=thigh_l, loc=<-21.6303 0 0>, rot=<0.0013 0 0 -1>}
 deform_groin_base_l {parent=thigh_l, loc=<-2.1908 8.1142 -1.7442>, rot=<0 0 0 -1>}
 deform_groin_l {parent=deform_groin_base_l, loc=<0 0 0>, rot=<0 0 0 1>}
 thigh_r {parent=pelvis, loc=<-2.5693 -0.0109 7.7501>, rot=<0.0572 -0.0012 0.9982 0.0155>}
 calf_r {parent=thigh_r, loc=<39.0484 0 0>, rot=<0.0006 -0.0161 -0.0391 -0.9991>}
 foot_r {parent=calf_r, loc=<42.561 -0.1077 -0.4292>, rot=<0.0418 -0.0455 0.0343 -0.9975>}
 ball_r {parent=foot_r, loc=<6.1599 -13.8264 -0.6806>, rot=<0.0001 -0.0001 -0.7063 -0.7079>}
 calf_twist_02_r {parent=calf_r, loc=<18.8064 0.0967 -0.0397>, rot=<0.0237 0.0095 -0.0035 -0.9997>}
 calf_twist_01_r {parent=calf_r, loc=<35.813 0.0967 -0.0397>, rot=<0.0237 0.0095 -0.0035 -0.9997>}
 deform_knee_in_base_r {parent=calf_r, loc=<-1.298 5.2957 0.0496>, rot=<0 0 0 1>}
 deform_knee_in_r {parent=deform_knee_in_base_r, loc=<0 0 0>, rot=<0 0 0 1>}
 deform_knee_out_base_r {parent=calf_r, loc=<0 -4 0>, rot=<0 0 0 -1>}
 deform_knee_out_r {parent=deform_knee_out_base_r, loc=<0 0 0>, rot=<0 0 0 1>}
 thigh_twist_01_r {parent=thigh_r, loc=<21.6303 0 0>, rot=<0.0013 0 0 -1>}
 deform_groin_base_r {parent=thigh_r, loc=<2.1908 -8.1142 1.7442>, rot=<0 0 0 -1>}
 deform_groin_r {parent=deform_groin_base_r, loc=<0 0 0>, rot=<0 0 0 1>}
 deform_glute_base_l {parent=pelvis, loc=<-8.8064 -8.8222 -8.4287>, rot=<0.0275 0.0501 -0.4472 -0.8926>}
 deform_glute_l {parent=deform_glute_base_l, loc=<0 0 0>, rot=<0 0 0 1>}
 deform_glute_base_r {parent=pelvis, loc=<-8.8062 -8.8222 8.4287>, rot=<0.0501 -0.0275 0.8926 -0.4472>}
 deform_glute_r {parent=deform_glute_base_r, loc=<0 0 0>, rot=<0 0 0 1>}
 ground_mover {parent=root, loc=<0 0 0>, rot=<0 0 0 1>}
]
