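# Physics asset for the BountyPond_Quart male face accessory, covering a left and right hat cord.
# Each cord is a five-capsule chain from a base body at the neck, with limited swing and near-locked twist between segments.
# Capsules on the spine, neck, clavicles, upper arms and left pec supply body shapes, joined by constraints with free angular motion.
# The two cord tips are also constrained to each other in both directions with free angular motion.
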PhysicsAsset /BRCosmetics/Characters/Player/Male/Medium/Bodies/M_MED_BountyPond_Quart/Meshes/Parts/M_MED_BountyPond_Quart_FaceAcc_Physics
bodies {
 base_hatCord_L_01 [{shapeType=Capsule, length=0.505, radius=0.505, shapeName=base_hatCord_L_01_capsule, rotation=<rot 0 -90 0>}]
 base_hatCord_R_01 [{shapeType=Capsule, length=0.505, radius=0.505, shapeName=base_hatCord_R_01_capsule, rotation=<rot 0 -90 0>}]
 clavicle_l [{
   shapeType=Capsule
   length=17.3811
   radius=7.6317
   shapeName=clavicle_l_capsule
   center=<4.5767 -1.1926 1.1886>
   rotation=<rot 72.9107 -4.3073 -18.0368>
  }]
 clavicle_r [{
   shapeType=Capsule
   length=17.3811
   radius=7.6317
   shapeName=clavicle_l_capsule
   center=<-4.416 0.4351 -1.254>
   rotation=<rot 72.9107 -4.3073 -18.0368>
  }]
 deform_pec_l [{
   shapeType=Capsule
   length=6.4487
   radius=4.1546
   shapeName=deform_pec_l_capsule
   center=<-1.8576 -0.111 0.0124>
   rotation=<rot 47.063 -156.5432 173.2123>
  }]
 dyn_hatCord_L_01 [{shapeType=Capsule, length=7.6973, radius=1.1315, shapeName=dyn_hatCord_L_01_capsule, center=<3.4924 0 0>, rotation=<rot 90 0 0>}]
 dyn_hatCord_L_02 [{shapeType=Capsule, length=7.6973, radius=1.1315, shapeName=dyn_hatCord_L_02_capsule, center=<3.4924 0.0403 0>, rotation=<rot 90 0 0>}]
 dyn_hatCord_L_03 [{shapeType=Capsule, length=7.6973, radius=1.1315, shapeName=dyn_hatCord_L_03_capsule, center=<3.4924 0.477 0>, rotation=<rot 90 0 0>}]
 dyn_hatCord_L_04 [{shapeType=Capsule, length=7.6973, radius=1.1315, shapeName=dyn_hatCord_L_04_capsule, center=<3.4924 0.416 0>, rotation=<rot 90 0 0>}]
 dyn_hatCord_L_05 [{
   shapeType=Capsule
   length=6.9785
   radius=1.8504
   shapeName=dyn_hatCord_L_05_capsule
   center=<4.242 1.19 -0.9331>
   rotation=<rot 61.4463 0.5384 -3.9284>
  }]
 dyn_hatCord_R_01 [{shapeType=Capsule, length=7.6973, radius=1.1315, shapeName=dyn_hatCord_R_01_capsule, center=<3.4924 0 0>, rotation=<rot 90 0 0>}]
 dyn_hatCord_R_02 [{shapeType=Capsule, length=7.6973, radius=1.1315, shapeName=dyn_hatCord_R_02_capsule, center=<3.4924 0.0403 0>, rotation=<rot 90 0 0>}]
 dyn_hatCord_R_03 [{shapeType=Capsule, length=7.6973, radius=1.1315, shapeName=dyn_hatCord_R_03_capsule, center=<3.4924 0.477 0>, rotation=<rot 90 0 0>}]
 dyn_hatCord_R_04 [{shapeType=Capsule, length=7.6973, radius=1.1315, shapeName=dyn_hatCord_R_04_capsule, center=<3.4924 0.3832 0>, rotation=<rot 90 0 0>}]
 dyn_hatCord_R_05 [{
   shapeType=Capsule
   length=6.8917
   radius=1.9371
   shapeName=dyn_hatCord_R_05_capsule
   center=<4.2157 1.2418 0.5584>
   rotation=<rot 66.969 -179.8464 173.6233>
  }]
 neck_01 [{shapeType=Capsule, length=7.6576, radius=5.7057, shapeName=neck_01_capsule, center=<6.8629 0 0>, rotation=<rot 90 0 0>}]
 spine_04 [{shapeType=Capsule, length=11.3518, radius=11.3695, shapeName=spine_04_capsule, center=<1.1283 2.5065 0>, rotation=<rot 0 -90 0>}]
 spine_05 [
  {shapeType=Capsule, length=13.8978, radius=10.4175, shapeName=spine_05_capsule, center=<-2.898 2.1135 0>, rotation=<rot 0 -90 0>}
  {
   shapeType=Capsule
   length=5.4447
   radius=7.172
   shapeName=spine_05_capsule
   center=<0.0817 2.6376 -12.738>
   rotation=<rot 69.2686 -0.0002 17.6866>
  }
  {
   shapeType=Capsule
   length=5.4447
   radius=7.172
   shapeName=spine_05_capsule
   center=<-0.1663 2.9212 12.77>
   rotation=<rot -69.2686 -0.0002 -17.6866>
  }
 ]
 upperarm_l [{
   shapeType=Capsule
   length=23.6899
   radius=6.3102
   shapeName=upperarm_l_capsule
   center=<11.9148 0.6145 -2.0843>
   rotation=<rot 89.689 -179.9792 -179.9793>
  }]
 upperarm_r [{
   shapeType=Capsule
   length=23.6899
   radius=6.3102
   shapeName=upperarm_l_capsule
   center=<-11.9166 2.5429 2.4133>
   rotation=<rot 89.689 -179.9736 169.0179>
  }]
}
constraints [
 {bone1Name=dyn_hatCord_L_01, bone2Name=base_hatCord_L_01, swing1LimitDegrees=45, swing2LimitDegrees=30, twistLimitDegrees=1}
 {bone1Name=dyn_hatCord_L_02, bone2Name=dyn_hatCord_L_01, swing1LimitDegrees=45, swing2LimitDegrees=45, twistLimitDegrees=1}
 {bone1Name=dyn_hatCord_L_03, bone2Name=dyn_hatCord_L_02, swing1LimitDegrees=50, swing2LimitDegrees=50, twistLimitDegrees=1}
 {bone1Name=dyn_hatCord_L_04, bone2Name=dyn_hatCord_L_03, swing1LimitDegrees=50, swing2LimitDegrees=50, twistLimitDegrees=1}
 {bone1Name=dyn_hatCord_L_05, bone2Name=dyn_hatCord_L_04, swing1LimitDegrees=40, swing2LimitDegrees=40, twistLimitDegrees=1}
 {bone1Name=dyn_hatCord_R_01, bone2Name=base_hatCord_R_01, swing1LimitDegrees=45, swing2LimitDegrees=30, twistLimitDegrees=1}
 {bone1Name=dyn_hatCord_R_02, bone2Name=dyn_hatCord_R_01, swing1LimitDegrees=45, swing2LimitDegrees=45, twistLimitDegrees=1}
 {bone1Name=dyn_hatCord_R_03, bone2Name=dyn_hatCord_R_02, swing1LimitDegrees=50, swing2LimitDegrees=50, twistLimitDegrees=1}
 {bone1Name=dyn_hatCord_R_04, bone2Name=dyn_hatCord_R_03, swing1LimitDegrees=50, swing2LimitDegrees=50, twistLimitDegrees=1}
 {bone1Name=dyn_hatCord_R_05, bone2Name=dyn_hatCord_R_04, swing1LimitDegrees=40, swing2LimitDegrees=40, twistLimitDegrees=1}
 {
  bone1Name=base_hatCord_L_01
  bone2Name=neck_01
  swing1LimitDegrees=45
  swing1Motion=Free
  swing2LimitDegrees=45
  swing2Motion=Free
  twistLimitDegrees=45
  twistMotion=Free
 }
 {
  bone1Name=base_hatCord_R_01
  bone2Name=neck_01
  swing1LimitDegrees=45
  swing1Motion=Free
  swing2LimitDegrees=45
  swing2Motion=Free
  twistLimitDegrees=45
  twistMotion=Free
 }
 {
  bone1Name=spine_05
  bone2Name=spine_04
  swing1LimitDegrees=45
  swing1Motion=Free
  swing2LimitDegrees=45
  swing2Motion=Free
  twistLimitDegrees=45
  twistMotion=Free
 }
 {
  bone1Name=clavicle_l
  bone2Name=spine_05
  swing1LimitDegrees=45
  swing1Motion=Free
  swing2LimitDegrees=45
  swing2Motion=Free
  twistLimitDegrees=45
  twistMotion=Free
 }
 {
  bone1Name=upperarm_l
  bone2Name=clavicle_l
  swing1LimitDegrees=45
  swing1Motion=Free
  swing2LimitDegrees=45
  swing2Motion=Free
  twistLimitDegrees=45
  twistMotion=Free
 }
 {
  bone1Name=clavicle_r
  bone2Name=spine_05
  swing1LimitDegrees=45
  swing1Motion=Free
  swing2LimitDegrees=45
  swing2Motion=Free
  twistLimitDegrees=45
  twistMotion=Free
 }
 {
  bone1Name=upperarm_r
  bone2Name=clavicle_r
  swing1LimitDegrees=45
  swing1Motion=Free
  swing2LimitDegrees=45
  swing2Motion=Free
  twistLimitDegrees=45
  twistMotion=Free
 }
 {
  bone1Name=dyn_hatCord_L_05
  bone2Name=dyn_hatCord_R_05
  swing1LimitDegrees=45
  swing1Motion=Free
  swing2LimitDegrees=45
  swing2Motion=Free
  twistLimitDegrees=45
  twistMotion=Free
 }
 {
  bone1Name=dyn_hatCord_R_05
  bone2Name=dyn_hatCord_L_05
  swing1LimitDegrees=45
  swing1Motion=Free
  swing2LimitDegrees=45
  swing2Motion=Free
  twistLimitDegrees=45
  twistMotion=Free
 }
 {
  bone1Name=neck_01
  bone2Name=spine_05
  swing1LimitDegrees=45
  swing1Motion=Free
  swing2LimitDegrees=45
  swing2Motion=Free
  twistLimitDegrees=45
  twistMotion=Free
 }
 {
  bone1Name=deform_pec_l
  bone2Name=spine_04
  swing1LimitDegrees=45
  swing1Motion=Free
  swing2LimitDegrees=45
  swing2Motion=Free
  twistLimitDegrees=45
  twistMotion=Free
 }
]
